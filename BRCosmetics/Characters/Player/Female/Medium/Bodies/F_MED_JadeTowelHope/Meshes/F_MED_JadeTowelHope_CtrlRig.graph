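# Corrective Control Rig for the F_MED_JadeTowelHope female body cosmetic, adjusting skirt, straps, belt, sleeves and tie as the character moves.
# Delta Rotation turns thigh swing/twist into weights for additive skirt offsets, including combination correctives scaled by crouch and emote variables.
# Spherical pose readers and deformRig anim attributes drive further skirt, belt and sleeve fixes.
# Parent constraints link skirt bases to thighs, and a debug toggle drives bones from controls.
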ControlRigBlueprint /BRCosmetics/Characters/Player/Female/Medium/Bodies/F_MED_JadeTowelHope/Meshes/F_MED_JadeTowelHope_CtrlRig
graphs [
 Multiply_4_1_ContainedGraph {nodes [
   n1 = RigVMFunctionEntryNode()
   RigVMFunctionReturnNode(Result=&RigVMFunction_MathDoubleMul(B=@n1.C, A=&RigVMFunction_MathDoubleMul(A=@n1.A, B=@n1.B).Result).Result)
  ]}
 Multiply_7_1_ContainedGraph {nodes [
   n1 = RigVMFunctionEntryNode()
   RigVMFunctionReturnNode(Result=&RigVMFunction_MathDoubleMul(B=@n1.C, A=&RigVMFunction_MathDoubleMul(A=@n1.A, B=@n1.B).Result).Result)
  ]}
 Multiply_9_1_ContainedGraph {nodes [
   n1 = RigVMFunctionEntryNode()
   RigVMFunctionReturnNode(Result=&RigVMFunction_MathDoubleMul(B=@n1.C, A=&RigVMFunction_MathDoubleMul(A=@n1.A, B=@n1.B).Result).Result)
  ]}
 New Function_ContainedGraph {nodes [
   n1 = RigVMFunctionEntryNode()
   RigVMFunction_ControlFlowBranch = &RigVMFunction_ControlFlowBranch(Condition=@n1.DebugControls, ExecuteContext=@n1.ExecuteContext)
   RigVMFunctionReturnNode(ExecuteContext=@RigVMFunction_ControlFlowBranch.Completed)
   RigUnit_GetTransform = &RigUnit_GetTransform(Item=(Type=Control,Name="thigh_l_ctrl"), bInitial=False)
   RigUnit_SetTransform = &RigUnit_SetTransform(
    Item=(Type=Bone,Name="thigh_l")
    Weight=1
    bInitial=False
    Value=@RigUnit_GetTransform.Transform
    ExecutePin=@RigVMFunction_ControlFlowBranch.True
   )
   > &RigUnit_SetTransform(
    Item=(Type=Bone,Name="thigh_r")
    Weight=1
    bInitial=False
    Value=&RigUnit_GetTransform(Item=(Type=Control,Name="thigh_r_ctrl"), bInitial=False).Transform
   )
   > &RigUnit_SetTransform(
    Item=(Type=Bone,Name="calf_l")
    Weight=1
    bInitial=False
    Value=&RigUnit_GetTransform(Item=(Type=Control,Name="calf_l_ctrl"), bInitial=False).Transform
   )
   > &RigUnit_SetTransform(
    Item=(Type=Bone,Name="calf_r")
    Weight=1
    bInitial=False
    Value=&RigUnit_GetTransform(Item=(Type=Control,Name="calf_r_ctrl"), bInitial=False).Transform
   )
   > &RigUnit_SetTransform(
    Item=(Type=Bone,Name="base_tie_02")
    Weight=1
    bInitial=False
    Value=&RigUnit_GetTransform(Item=(Type=Control,Name="base_tie_02_ctrl"), bInitial=False).Transform
   )
   > &RigUnit_SetTransform(
    Item=(Type=Bone,Name="upperarm_l")
    Weight=1
    bInitial=False
    Value=&RigUnit_GetTransform(Item=(Type=Control,Name="upperarm_l_ctrl"), bInitial=False).Transform
   )
   > &RigUnit_SetTransform(
    Item=(Type=Bone,Name="upperarm_r")
    Weight=1
    bInitial=False
    Value=&RigUnit_GetTransform(Item=(Type=Control,Name="upperarm_r_ctrl"), bInitial=False).Transform
   )
   > &RigUnit_SetTransform(
    Item=(Type=Bone,Name="spine_01")
    Weight=1
    bInitial=False
    Value=&RigUnit_GetTransform(Item=(Type=Control,Name="spine_01_ctrl"), bInitial=False).Transform
   )
  ]}
 New Function_ContainedGraph {nodes [
   n1 = RigVMFunctionEntryNode()
   Clamp = &RigVMFunction_MathIntClamp(Value=@n1.primAxis)
   RigVMFunction_MathBoolToInteger = &RigVMFunction_MathBoolToInteger(Value=@n1.onlyTwist)
   RigVMFunction_MathQuaternionSwingTwist = &RigVMFunction_MathQuaternionSwingTwist(
    Input=&RigUnit_GetRelativeTransformForItem(
     Child=(Type=Bone,Name="None")
     Parent=(Type=Bone,Name="None")
     bChildInitial=False
     bParentInitial=False
     Parent.Type=@n1.ParentType
     Child.Name=@n1.child
     Parent.Name=@n1.parent
    ).RelativeTransform.Rotation
    TwistAxis=@n1.twistAxis
   )
   RigVMFunction_MathQuaternionToEuler = &RigVMFunction_MathQuaternionToEuler(
    Value=RigVMDispatchNode(
     Values=((),())
     Index=@RigVMFunction_MathBoolToInteger.Result
     Values.0=@RigVMFunction_MathQuaternionSwingTwist.Swing
     Values.1=@RigVMFunction_MathQuaternionSwingTwist.Twist
    ) {template="DISPATCH_RigVMDispatch_SelectInt32(in Index,in Values,out Result)"}.Result
    RotationOrder=@n1.rotationOrder
   )
   DISPATCH_RigVMDispatch_SelectInt32 = RigVMDispatchNode(
    Values=(0,0,0)
    Index=@Clamp.Result
    Values.0=@RigVMFunction_MathQuaternionToEuler.Result.X
    Values.1=@RigVMFunction_MathQuaternionToEuler.Result.Y
    Values.2=@RigVMFunction_MathQuaternionToEuler.Result.Z
   ) {template="DISPATCH_RigVMDispatch_SelectInt32(in Index,in Values,out Result)"}
   Remap = &RigVMFunction_MathDoubleRemap(
    TargetMaximum=1
    TargetMinimum=0
    bClamp=False
    Value=@DISPATCH_RigVMDispatch_SelectInt32.Result
    SourceMinimum=@n1.min
    SourceMaximum=@n1.max
   )
   > RigVMFunctionReturnNode(ouputWeight=&RigVMFunction_MathDoubleClamp(Value=@Remap.Result).Result)
  ]}
 New Function_ContainedGraph {nodes [
   RigVMFunctionEntryNode()
   > RigVMFunction_Sequence = RigVMAggregateNode()
   Delta Rotation = RigVMFunctionReferenceNode(
    ParentType=Null
    child=thigh_l
    debugRotation=False
    max=-90
    onlyTwist=False
    parent=thigh_l_null
    primAxis=2
    rotationOrder=XYZ
    twistAxis=(X=1,Y=0,Z=0)
    ExecuteContext=@RigVMFunction_Sequence.A
   )
   > RigVMVariableNode(Value=@"Delta Rotation".ouputWeight) {variable=L_Leg_Rot_Fr}
   > &RigUnit_ModifyTransforms(
    ItemToModify="((Item=(Type=Bone,Name=\"base_skirt_bk_l\"),Transform=(Rotation=(X=0,Y=0,Z=0.7169,W=0.6972),Translation=(X=0,Y=0,Z=0),Scale3D=(X=1,Y=1,Z=1))),(Item=(Type=Bone,Name=\"base_skirt_sd_l\"),Transform=(Rotation=(X=0.0087,Y=-0.0871,Z=0.4554,W=0.886),Translation=(X=1. ...(truncated)"
    WeightMaximum=0.7
    WeightMinimum=0
    Weight=@"Delta Rotation".ouputWeight
   )
   > n1 = &RigUnit_ModifyTransforms(
    ItemToModify="((Item=(Type=Bone,Name=\"base_skirt_fr_l\"),Transform=(Rotation=(X=-0.0745,Y=-0.0079,Z=0.1929,W=0.9783),Translation=(X=0.0002,Y=2.0004,Z=0.6842),Scale3D=(X=1,Y=1,Z=1))),(Item=(Type=Bone,Name=\"dyn_skirt_fr_l_03\"),Transform=(Rotation=(X=0,Y=0,Z=0,W=1),Translat ...(truncated)"
    WeightMinimum=0
    Weight=&RigVMFunction_MathDoubleMul(B=RigVMVariableNode() {variable=Crouching}.Value, A=@"Delta Rotation".ouputWeight).Result
   )
   n2 = &RigVMFunction_ControlFlowBranch(ExecuteContext=@n1.ExecutePin, Condition=RigVMVariableNode() {variable=Sides}.Value)
   &RigUnit_ModifyTransforms(
    ItemToModify="((Item=(Type=Bone,Name=\"base_megaphone\"),Transform=(Rotation=(X=0.2139,Y=-0.2218,Z=0.36,W=0.8806),Translation=(X=2.5,Y=0,Z=-1),Scale3D=(X=1,Y=1,Z=1))),(Item=(Type=Bone,Name=\"base_nade\"),Transform=(Rotation=(X=-0.5124,Y=-0.1616,Z=0.2536,W=0.8044),Translatio ...(truncated)"
    WeightMinimum=0
    ExecutePin=@n2.Completed
    Weight=@"Delta Rotation".ouputWeight
   )
   n3 = &RigVMFunction_MathDoubleMul(B=RigVMVariableNode() {variable=CrouchingMoving2D}.Value, A=@"Delta Rotation".ouputWeight)
   > &RigUnit_ModifyTransforms(
    ItemToModify="((Item=(Type=Bone,Name=\"base_skirt_bk_l\"),Transform=(Rotation=(X=0,Y=0,Z=-0.2588,W=0.9659),Translation=(X=0,Y=0,Z=0),Scale3D=(X=1,Y=1,Z=1))),(Item=(Type=Bone,Name=\"base_skirt_sd_l\"),Transform=(Rotation=(X=0,Y=0.0872,Z=0,W=0.9962),Translation=(X=0,Y=0,Z=0), ...(truncated)"
    WeightMinimum=0
    Weight=@n3.Result
   )
   > Delta Rotation_1 = RigVMFunctionReferenceNode(
    ParentType=Null
    child=thigh_l
    debugRotation=False
    max=90
    onlyTwist=False
    parent=thigh_l_null
    primAxis=2
    rotationOrder=XYZ
    twistAxis=(X=1,Y=0,Z=0)
   )
   > &RigUnit_ModifyTransforms(
    ItemToModify="((Item=(Type=Bone,Name=\"base_skirt_fr_l\"),Transform=(Rotation=(X=0,Y=0,Z=-0.7146,W=0.6995),Translation=(X=0,Y=0,Z=0),Scale3D=(X=1,Y=1,Z=1))),(Item=(Type=Bone,Name=\"base_skirt_sd_l\"),Transform=(Rotation=(X=-0.0075,Y=-0.0856,Z=-0.3207,W=0.9433),Translation=( ...(truncated)"
    WeightMinimum=0
    Weight=@"Delta Rotation_1".ouputWeight
   )
   > Delta Rotation_1_1 = RigVMFunctionReferenceNode(
    ParentType=Null
    child=thigh_l
    debugRotation=False
    max=-90
    onlyTwist=False
    parent=thigh_l_null
    primAxis=1
    rotationOrder=XYZ
    twistAxis=(X=1,Y=0,Z=0)
   )
   > RigVMVariableNode(Value=@"Delta Rotation_1_1".ouputWeight) {variable=L_leg_Rot_Sd_Out}
   > &RigUnit_ModifyTransforms(
    ItemToModify="((Item=(Type=Bone,Name=\"base_skirt_fr_c\"),Transform=(Rotation=(X=0,Y=0.3606,Z=0,W=0.9327),Translation=(X=0,Y=0,Z=0),Scale3D=(X=1,Y=1,Z=1))),(Item=(Type=Bone,Name=\"base_skirt_fr_l\"),Transform=(Rotation=(X=0.0307,Y=0.3565,Z=-0.1383,W=0.9235),Translation=(X=0 ...(truncated)"
    WeightMinimum=0
    Weight=@"Delta Rotation_1_1".ouputWeight
   )
   > &RigUnit_ModifyTransforms(
    ItemToModify=((Item=(Type=Bone,Name="dyn_skirt_bk_l_01"),Transform=(Rotation=(X=0.1506,Y=0.1899,Z=-0.2534,W=0.9365),Translation=(X=0,Y=0,Z=0),Scale3D=(X=1,Y=1,Z=1))))
    WeightMinimum=0
    Weight=&RigVMFunction_MathDoubleMul(A=@"Delta Rotation_1_1".ouputWeight, B=RigVMVariableNode() {variable=Crouching}.Value).Result
   )
   > Delta Rotation_1_1_2 = RigVMFunctionReferenceNode(
    ParentType=Null
    child=thigh_l
    debugRotation=False
    max=60
    onlyTwist=False
    parent=thigh_l_null
    primAxis=1
    rotationOrder=XYZ
    twistAxis=(X=1,Y=0,Z=0)
   )
   > RigVMVariableNode(Value=@"Delta Rotation_1_1_2".ouputWeight) {variable=L_Leg_Rot_In}
   > &RigUnit_ModifyTransforms(
    ItemToModify="((Item=(Type=Bone,Name=\"base_skirt_bk_c\"),Transform=(Rotation=(X=0,Y=0,Z=-0.2563,W=0.9666),Translation=(X=0,Y=0,Z=0),Scale3D=(X=1,Y=1,Z=1))),(Item=(Type=Bone,Name=\"base_skirt_fr_c\"),Transform=(Rotation=(X=0,Y=-0.2595,Z=0,W=0.9658),Translation=(X=0,Y=0,Z=0) ...(truncated)"
    WeightMinimum=0
    Weight=@"Delta Rotation_1_1_2".ouputWeight
   )
   > Delta Rotation_1_1_2_2 = RigVMFunctionReferenceNode(
    ParentType=Null
    child=thigh_l
    debugRotation=False
    max=-45
    onlyTwist=true
    parent=thigh_l_null
    primAxis=0
    rotationOrder=XYZ
    twistAxis=(X=1,Y=0,Z=0)
   )
   > RigVMVariableNode(Value=@"Delta Rotation_1_1_2_2".ouputWeight) {variable=L_leg_Twist_Out}
   > &RigUnit_ModifyTransforms(
    ItemToModify="((Item=(Type=Bone,Name=\"base_skirt_fr_l\"),Transform=(Rotation=(X=0.6775,Y=-0.005,Z=-0.0293,W=0.7349),Translation=(X=0,Y=0,Z=0),Scale3D=(X=1,Y=1,Z=1))),(Item=(Type=Bone,Name=\"base_skirt_sd_l\"),Transform=(Rotation=(X=0.6043,Y=-0.0118,Z=-0.0217,W=0.7964),Tran ...(truncated)"
    WeightMinimum=0
    Weight=@"Delta Rotation_1_1_2_2".ouputWeight
   )
   > &RigUnit_ModifyTransforms(
    ItemToModify="((Item=(Type=Bone,Name=\"base_strap_fr_l\"),Transform=(Rotation=(X=-0.0002,Y=0.1129,Z=-0.0019,W=0.9936),Translation=(X=0,Y=0,Z=0),Scale3D=(X=1,Y=1,Z=1))),(Item=(Type=Bone,Name=\"dyn_strap_fr_l_02\"),Transform=(Rotation=(X=0,Y=0.2966,Z=0,W=0.955),Translation=(X ...(truncated)"
    WeightMinimum=0
    Weight=&RigVMFunction_MathDoubleMul(B=@"Delta Rotation_1_1_2_2".ouputWeight, A=RigVMVariableNode() {variable=Crouching}.Value).Result
   )
   > Delta Rotation_1_1_2_2_2 = RigVMFunctionReferenceNode(
    ParentType=Null
    child=thigh_l
    debugRotation=False
    max=45
    onlyTwist=true
    parent=thigh_l_null
    primAxis=0
    rotationOrder=XYZ
    twistAxis=(X=1,Y=0,Z=0)
   )
   > RigVMVariableNode(Value=@"Delta Rotation_1_1_2_2_2".ouputWeight) {variable=L_leg_Twist_In}
   > &RigUnit_ModifyTransforms(
    ItemToModify="((Item=(Type=Bone,Name=\"base_skirt_fr_c\"),Transform=(Rotation=(X=-0.2322,Y=0,Z=0,W=0.9727),Translation=(X=0,Y=0,Z=0),Scale3D=(X=1,Y=1,Z=1))),(Item=(Type=Bone,Name=\"base_skirt_fr_l\"),Transform=(Rotation=(X=-0.428,Y=0,Z=0,W=0.9038),Translation=(X=0,Y=0,Z=0), ...(truncated)"
    WeightMinimum=0
    Weight=@"Delta Rotation_1_1_2_2_2".ouputWeight
   )
   Multiply = &RigVMFunction_MathDoubleMul(A=RigVMVariableNode() {variable=L_Leg_Rot_Fr}.Value, B=RigVMVariableNode() {variable=L_leg_Rot_Sd_Out}.Value)
   > &RigUnit_ModifyTransforms(
    ItemToModify="((Item=(Type=Bone,Name=\"base_skirt_sd_l\"),Transform=(Rotation=(X=-0.0069,Y=-0.0475,Z=-0.1427,W=0.9886),Translation=(X=0.1388,Y=0.6238,Z=0),Scale3D=(X=1,Y=1,Z=1))),(Item=(Type=Bone,Name=\"base_skirt_fr_l\"),Transform=(Rotation=(X=0.2827,Y=0.1472,Z=0.3709,W=0. ...(truncated)"
    WeightMinimum=0
    Weight=@Multiply.Result
   )
   > &RigUnit_ModifyTransforms(
    ItemToModify="((Item=(Type=Bone,Name=\"base_skirt_fr_c\"),Transform=(Rotation=(X=0.0579,Y=0.1023,Z=-0.489,W=0.8643),Translation=(X=0,Y=0,Z=0),Scale3D=(X=1,Y=1,Z=1))),(Item=(Type=Bone,Name=\"base_skirt_fr_r\"),Transform=(Rotation=(X=0.3336,Y=-0.1807,Z=-0.7842,W=0.491),Transl ...(truncated)"
    WeightMinimum=0
    Weight=&RigVMFunction_MathDoubleMul(B=RigVMVariableNode() {variable=L_Leg_Rot_Fr}.Value, A=RigVMVariableNode() {variable=L_Leg_Rot_In}.Value).Result
   )
   Multiply_7 = RigVMAggregateNode(
    A=&RigVMFunction_MathDoubleClamp(Minimum=0.8, Value=RigVMVariableNode() {variable=L_Leg_Rot_In}.Value).Result
    C=RigVMVariableNode() {variable=L_leg_Twist_Out}.Value
    B=RigVMVariableNode() {variable=CrouchingMoving2D}.Value
   )
   > &RigUnit_ModifyTransforms(
    ItemToModify="((Item=(Type=Bone,Name=\"base_skirt_bk_l\"),Transform=(Rotation=(X=-0.2144,Y=-0.2181,Z=0.0492,W=0.9508),Translation=(X=0,Y=0,Z=0),Scale3D=(X=1,Y=1,Z=1))),(Item=(Type=Bone,Name=\"dyn_strap_bk_l_01\"),Transform=(Rotation=(X=0.0339,Y=-0.4092,Z=0.0152,W=0.9117),Tr ...(truncated)"
    WeightMinimum=0
    Weight=@Multiply_7.Result
   )
   Multiply_9 = RigVMAggregateNode(
    B=RigVMVariableNode() {variable=L_Leg_Rot_In}.Value
    A=RigVMVariableNode() {variable=L_leg_Twist_Out}.Value
    C=RigVMVariableNode() {variable=Crouching}.Value
   )
   > &RigUnit_ModifyTransforms(
    ItemToModify=((Item=(Type=Bone,Name="dyn_skirt_bk_l_01"),Transform=(Rotation=(X=0,Y=0,Z=0,W=1),Translation=(X=0,Y=0,Z=0),Scale3D=(X=1,Y=1,Z=1))))
    WeightMinimum=0
    Weight=@Multiply_9.Result
   )
   n4 = &RigVMFunction_AlphaInterp(InterpSpeedDecreasing=10, InterpSpeedIncreasing=10, Scale=1, bClampResult=true, Value=RigVMVariableNode() {variable=EmoteGoldenGlory}.Value)
   > &RigUnit_ModifyTransforms(
    ItemToModify="((Item=(Type=Bone,Name=\"base_skirt_fr_l\"),Transform=(Rotation=(X=0.0338,Y=0.0277,Z=-0.0504,W=-0.9978),Translation=(X=0,Y=0,Z=0),Scale3D=(X=1,Y=1,Z=1))),(Item=(Type=Bone,Name=\"base_skirt_sd_l\"),Transform=(Rotation=(X=-0.0252,Y=0.1469,Z=-0.2631,W=-0.9532),Tr ...(truncated)"
    WeightMinimum=0
    Weight=&RigVMFunction_MathDoubleMul(B=@n4.Result, A=RigVMVariableNode() {variable=L_Leg_Rot_Fr}.Value).Result
   )
   &RigUnit_ModifyTransforms(
    ItemToModify="((Item=(Type=Bone,Name=\"dyn_skirt_fr_l_01\"),Transform=(Rotation=(X=0,Y=0,Z=0,W=1),Translation=(X=0,Y=0,Z=0),Scale3D=(X=1,Y=1,Z=1))),(Item=(Type=Bone,Name=\"dyn_skirt_fr_c_01\"),Transform=(Rotation=(X=0,Y=0,Z=0,W=1),Translation=(X=0,Y=0,Z=0),Scale3D=(X=1,Y=1, ...(truncated)"
    WeightMinimum=0
    ExecutePin=@n2.False
    Weight=@n3.Result
   )
   &RigUnit_ModifyTransforms(
    ItemToModify="((Item=(Type=Bone,Name=\"dyn_skirt_fr_l_01\"),Transform=(Rotation=(X=-0.0046,Y=-0.0435,Z=-0.0202,W=0.9988),Translation=(X=0,Y=0,Z=0),Scale3D=(X=1,Y=1,Z=1))),(Item=(Type=Bone,Name=\"dyn_skirt_fr_c_01\"),Transform=(Rotation=(X=0.0012,Y=0.0305,Z=-0.0377,W=0.9988) ...(truncated)"
    WeightMinimum=0
    Weight=@n3.Result
    ExecutePin=@n2.True
   )
   Delta Rotation_2 = RigVMFunctionReferenceNode(
    ParentType=Null
    child=thigh_r
    debugRotation=False
    max=-90
    onlyTwist=False
    parent=thigh_r_null
    primAxis=2
    rotationOrder=ZYX
    twistAxis=(X=-1,Y=0,Z=0)
    ExecuteContext=@RigVMFunction_Sequence.B
   )
   > RigVMVariableNode(Value=@"Delta Rotation_2".ouputWeight) {variable=R_Leg_Rot_Fr}
   > &RigUnit_ModifyTransforms(
    ItemToModify="((Item=(Type=Bone,Name=\"base_skirt_bk_c\"),Transform=(Rotation=(X=0,Y=-0.3668,Z=0,W=0.9303),Translation=(X=0,Y=0,Z=0),Scale3D=(X=1,Y=1,Z=1))),(Item=(Type=Bone,Name=\"base_skirt_bk_r\"),Transform=(Rotation=(X=0,Y=0,Z=0.7303,W=0.6832),Translation=(X=0,Y=0,Z=0), ...(truncated)"
    WeightMinimum=0
    Weight=@"Delta Rotation_2".ouputWeight
   )
   > n5 = &RigUnit_ModifyTransforms(
    ItemToModify="((Item=(Type=Bone,Name=\"base_skirt_bk_r\"),Transform=(Rotation=(X=0,Y=0,Z=-0.2459,W=0.9693),Translation=(X=0,Y=0,Z=0),Scale3D=(X=1,Y=1,Z=1))),(Item=(Type=Bone,Name=\"base_skirt_bk_c\"),Transform=(Rotation=(X=-0.0285,Y=0.0931,Z=0.0617,W=0.9933),Translation=(X= ...(truncated)"
    WeightMinimum=0
    Weight=&RigVMFunction_MathDoubleMul(A=@"Delta Rotation_2".ouputWeight, B=RigVMVariableNode() {variable=Crouching}.Value).Result
   )
   RigVMFunction_ControlFlowBranch = &RigVMFunction_ControlFlowBranch(ExecuteContext=@n5.ExecutePin, Condition=RigVMVariableNode() {variable=Sides}.Value)
   Delta Rotation_1_2 = RigVMFunctionReferenceNode(
    ParentType=Null
    child=thigh_r
    debugRotation=False
    max=90
    onlyTwist=False
    parent=thigh_r_null
    primAxis=2
    rotationOrder=ZYX
    twistAxis=(X=-1,Y=0,Z=0)
    ExecuteContext=@RigVMFunction_ControlFlowBranch.Completed
   )
   > &RigUnit_ModifyTransforms(
    ItemToModify="((Item=(Type=Bone,Name=\"base_skirt_fr_c\"),Transform=(Rotation=(X=0,Y=0,Z=-0.4027,W=0.9153),Translation=(X=0,Y=0,Z=0),Scale3D=(X=1,Y=1,Z=1))),(Item=(Type=Bone,Name=\"base_skirt_fr_r\"),Transform=(Rotation=(X=0,Y=0,Z=-0.6807,W=0.7325),Translation=(X=0,Y=0,Z=0) ...(truncated)"
    WeightMinimum=0
    Weight=@"Delta Rotation_1_2".ouputWeight
   )
   > Delta Rotation_1_1_1 = RigVMFunctionReferenceNode(
    ParentType=Null
    child=thigh_r
    debugRotation=False
    max=-90
    onlyTwist=False
    parent=thigh_r_null
    primAxis=1
    rotationOrder=ZYX
    twistAxis=(X=1,Y=0,Z=0)
   )
   > RigVMVariableNode(Value=@"Delta Rotation_1_1_1".ouputWeight) {variable=R_leg_Rot_Sd_Out}
   > &RigUnit_ModifyTransforms(
    ItemToModify="((Item=(Type=Bone,Name=\"base_skirt_fr_c\"),Transform=(Rotation=(X=0,Y=-0.2681,Z=0,W=0.9634),Translation=(X=0,Y=0,Z=0),Scale3D=(X=1,Y=1,Z=1))),(Item=(Type=Bone,Name=\"base_skirt_fr_r\"),Transform=(Rotation=(X=-0.0432,Y=-0.2115,Z=-0.1338,W=0.9672),Translation=( ...(truncated)"
    WeightMinimum=0
    Weight=@"Delta Rotation_1_1_1".ouputWeight
   )
   > &RigUnit_ModifyTransforms(
    ItemToModify=((Item=(Type=Bone,Name="base_skirt_bk_r"),Transform=(Rotation=(X=-0.0062,Y=-0.6289,Z=0.0789,W=0.7735),Translation=(X=0,Y=0,Z=0),Scale3D=(X=1,Y=1,Z=1))))
    WeightMinimum=0
    Weight=&RigVMFunction_MathDoubleMul(A=@"Delta Rotation_1_1_1".ouputWeight, B=RigVMVariableNode() {variable=Crouching}.Value).Result
   )
   > Delta Rotation_1_1_2_1 = RigVMFunctionReferenceNode(
    ParentType=Null
    child=thigh_r
    debugRotation=False
    max=60
    min=0
    onlyTwist=False
    parent=thigh_r_null
    primAxis=1
    rotationOrder=XYZ
    twistAxis=(X=1,Y=0,Z=0)
   )
   > RigVMVariableNode(Value=@"Delta Rotation_1_1_2_1".ouputWeight) {variable=R_Leg_Rot_In}
   > &RigUnit_ModifyTransforms(
    ItemToModify="((Item=(Type=Bone,Name=\"base_skirt_fr_c\"),Transform=(Rotation=(X=0,Y=0.2698,Z=0,W=0.9629),Translation=(X=0,Y=0,Z=0),Scale3D=(X=1,Y=1,Z=1))),(Item=(Type=Bone,Name=\"base_skirt_bk_c\"),Transform=(Rotation=(X=0,Y=0,Z=0.2418,W=0.9703),Translation=(X=0,Y=0,Z=0),S ...(truncated)"
    WeightMinimum=0
    Weight=@"Delta Rotation_1_1_2_1".ouputWeight
   )
   > Delta Rotation_1_1_2_2_1 = RigVMFunctionReferenceNode(
    ParentType=Null
    child=thigh_r
    debugRotation=False
    max=-45
    onlyTwist=true
    parent=thigh_r_null
    primAxis=0
    rotationOrder=YZX
    twistAxis=(X=1,Y=0,Z=0)
   )
   > RigVMVariableNode(Value=@"Delta Rotation_1_1_2_2_1".ouputWeight) {variable=R_leg_Twist_Out}
   > &RigUnit_ModifyTransforms(
    ItemToModify="((Item=(Type=Bone,Name=\"base_skirt_fr_c\"),Transform=(Rotation=(X=-0.2207,Y=0,Z=0,W=0.9754),Translation=(X=0,Y=0,Z=0),Scale3D=(X=1,Y=1,Z=1))),(Item=(Type=Bone,Name=\"base_skirt_fr_r\"),Transform=(Rotation=(X=-0.6915,Y=-0.0118,Z=0.0535,W=0.7203),Translation=(X ...(truncated)"
    WeightMinimum=0
    Weight=@"Delta Rotation_1_1_2_2_1".ouputWeight
   )
   > Delta Rotation_1_1_2_2_2_1 = RigVMFunctionReferenceNode(
    ParentType=Null
    child=thigh_r
    debugRotation=False
    max=45
    onlyTwist=true
    parent=thigh_r_null
    primAxis=0
    rotationOrder=XYZ
    twistAxis=(X=1,Y=0,Z=0)
   )
   > &RigUnit_ModifyTransforms(
    ItemToModify="((Item=(Type=Bone,Name=\"base_skirt_fr_c\"),Transform=(Rotation=(X=0.2705,Y=0,Z=0,W=0.9627),Translation=(X=0,Y=0,Z=0),Scale3D=(X=1,Y=1,Z=1))),(Item=(Type=Bone,Name=\"base_skirt_fr_r\"),Transform=(Rotation=(X=0.347,Y=0,Z=0,W=0.9379),Translation=(X=0,Y=0,Z=0),Sc ...(truncated)"
    WeightMinimum=0
    Weight=@"Delta Rotation_1_1_2_2_2_1".ouputWeight
   )
   > &RigUnit_ModifyTransforms(
    ItemToModify="((Item=(Type=Bone,Name=\"base_skirt_sd_r\"),Transform=(Rotation=(X=-0.0635,Y=0.0548,Z=-0.2858,W=0.9546),Translation=(X=-0.5877,Y=-0.5197,Z=2.6155),Scale3D=(X=1,Y=1,Z=1))),(Item=(Type=Bone,Name=\"base_skirt_hanging_fr_r\"),Transform=(Rotation=(X=-0.525,Y=-0.413 ...(truncated)"
    WeightMinimum=0
    Weight=&RigVMFunction_MathDoubleMul(B=RigVMVariableNode() {variable=R_leg_Rot_Sd_Out}.Value, A=RigVMVariableNode() {variable=R_Leg_Rot_Fr}.Value).Result
   )
   > &RigUnit_ModifyTransforms(
    ItemToModify="((Item=(Type=Bone,Name=\"dyn_skirt_fr_c_01\"),Transform=(Rotation=(X=-0.0544,Y=-0.4871,Z=-0.3457,W=0.8002),Translation=(X=0,Y=0,Z=0),Scale3D=(X=1,Y=1,Z=1))),(Item=(Type=Bone,Name=\"dyn_skirt_fr_l_01\"),Transform=(Rotation=(X=0.4309,Y=-0.6205,Z=0.2386,W=0.6103) ...(truncated)"
    WeightMinimum=0
    Weight=&RigVMFunction_MathDoubleMul(B=RigVMVariableNode() {variable=R_Leg_Rot_Fr}.Value, A=RigVMVariableNode() {variable=R_Leg_Rot_In}.Value).Result
   )
   Clamp = &RigVMFunction_MathDoubleClamp(Minimum=0.5, Value=RigVMVariableNode() {variable=R_Leg_Rot_Fr}.Value)
   Multiply_4 = RigVMAggregateNode(B=@Clamp.Result, A=RigVMVariableNode() {variable=R_leg_Twist_Out}.Value, C=RigVMVariableNode() {variable=PlayingEmote}.Value)
   RigVMFunction_AnimEvalRichCurve = &RigVMFunction_AnimEvalRichCurve(
    Curve=(EditorCurveData=(Keys=((),(InterpMode=RCIM_Cubic,TangentMode=RCTM_User,Time=0.4179,Value=0.0769,ArriveTangent=0.3928,LeaveTangent=0.3928),(Time=1,Value=1))))
    SourceMinimum=0
    TargetMinimum=0
    Value=@Multiply_4.Result
   )
   > &RigUnit_ModifyTransforms(
    ItemToModify="((Item=(Type=Bone,Name=\"base_skirt_bk_r\"),Transform=(Rotation=(X=-0.0167,Y=0.084,Z=0.5417,W=0.8362),Translation=(X=0,Y=0,Z=0),Scale3D=(X=1,Y=1,Z=1))),(Item=(Type=Bone,Name=\"dyn_skirt_bk_c_01\"),Transform=(Rotation=(X=0,Y=-0.2506,Z=0,W=0.9681),Translation=(X ...(truncated)"
    WeightMinimum=0
    Weight=@RigVMFunction_AnimEvalRichCurve.Result
   )
   AlphaInterp = &RigVMFunction_AlphaInterp(InterpSpeedDecreasing=10, InterpSpeedIncreasing=10, Scale=1, bClampResult=true, Value=RigVMVariableNode() {variable=EmoteGoldenGlory}.Value)
   > &RigUnit_ModifyTransforms(
    ItemToModify="((Item=(Type=Bone,Name=\"base_skirt_fr_r\"),Transform=(Rotation=(X=-0.1518,Y=0.1406,Z=-0.0482,W=0.9772),Translation=(X=0,Y=0,Z=0),Scale3D=(X=1,Y=1,Z=1))),(Item=(Type=Bone,Name=\"base_skirt_sd_r\"),Transform=(Rotation=(X=0.0068,Y=0.0216,Z=-0.0001,W=0.9997),Tran ...(truncated)"
    WeightMinimum=0
    Weight=&RigVMFunction_MathDoubleMul(B=@AlphaInterp.Result, A=RigVMVariableNode() {variable=R_Leg_Rot_Fr}.Value).Result
   )
   n6 = &RigVMFunction_MathDoubleMul(A=@"Delta Rotation_2".ouputWeight, B=RigVMVariableNode() {variable=CrouchingMoving2D}.Value)
   &RigUnit_ModifyTransforms(
    ItemToModify="((Item=(Type=Bone,Name=\"base_skirt_fr_r\"),Transform=(Rotation=(X=0,Y=0,Z=0,W=1),Translation=(X=0,Y=0,Z=0),Scale3D=(X=1,Y=1,Z=1))),(Item=(Type=Bone,Name=\"base_skirt_fr_c\"),Transform=(Rotation=(X=0,Y=0,Z=0,W=1),Translation=(X=0,Y=0,Z=0),Scale3D=(X=1,Y=1,Z=1) ...(truncated)"
    WeightMinimum=0
    ExecutePin=@RigVMFunction_ControlFlowBranch.False
    Weight=@n6.Result
   )
   &RigUnit_ModifyTransforms(
    ItemToModify="((Item=(Type=Bone,Name=\"base_skirt_fr_r\"),Transform=(Rotation=(X=-0.0027,Y=-0.0435,Z=-0.061,W=0.9972),Translation=(X=0,Y=0,Z=0),Scale3D=(X=1,Y=1,Z=1))),(Item=(Type=Bone,Name=\"base_skirt_fr_c\"),Transform=(Rotation=(X=0,Y=0,Z=0.0963,W=0.9954),Translation=(X= ...(truncated)"
    WeightMinimum=0
    Weight=@n6.Result
    ExecutePin=@RigVMFunction_ControlFlowBranch.True
   )
   n7 = RigVMDispatchNode(BoneName=root, Default=1, Name=deformRig_l_upperarm_up_alpha) {template="DISPATCH_RigDispatch_GetAnimAttribute(in Name,in BoneName,in Default,out Value,out Found)"}
   n8 = &RigUnit_ModifyTransforms(
    ItemToModify=((Item=(Type=Bone,Name="sleeves_low_l"),Transform=(Rotation=(X=0,Y=0,Z=0,W=1),Translation=(X=0,Y=-2,Z=3),Scale3D=(X=1,Y=1,Z=1))))
    WeightMinimum=0
    Weight=@n7.Value
    ExecutePin=@RigVMFunction_Sequence.D
   )
   n9 = &RigVMFunction_ControlFlowBranch(ExecuteContext=@n8.ExecutePin, Condition=RigVMVariableNode() {variable=CrouchingTargeting}.Value)
   &RigUnit_ModifyTransforms(
    ItemToModify=((Item=(Type=Bone,Name="base_sleeves_bk_l"),Transform=(Rotation=(X=0,Y=0,Z=0,W=1),Translation=(X=0,Y=0,Z=10),Scale3D=(X=1,Y=1,Z=1))))
    WeightMinimum=0
    Weight=@n7.Value
    ExecutePin=@n9.True
   )
   n10 = RigVMDispatchNode(BoneName=root, Default=1, Name=deformRig_r_upperarm_up_alpha) {template="DISPATCH_RigDispatch_GetAnimAttribute(in Name,in BoneName,in Default,out Value,out Found)"}
   n11 = &RigUnit_ModifyTransforms(
    ItemToModify=((Item=(Type=Bone,Name="sleeves_low_r"),Transform=(Rotation=(X=0,Y=0,Z=0,W=1),Translation=(X=0,Y=2,Z=-3),Scale3D=(X=1,Y=1,Z=1))))
    WeightMinimum=0
    Weight=@n10.Value
    ExecutePin=@RigVMFunction_Sequence.E
   )
   n12 = &RigVMFunction_ControlFlowBranch(ExecuteContext=@n11.ExecutePin, Condition=RigVMVariableNode() {variable=CrouchingTargeting}.Value)
   &RigUnit_ModifyTransforms(
    ItemToModify=((Item=(Type=Bone,Name="base_sleeves_bk_r"),Transform=(Rotation=(X=0,Y=0,Z=0,W=1),Translation=(X=0,Y=0,Z=-10),Scale3D=(X=1,Y=1,Z=1))))
    WeightMinimum=0
    Weight=@n10.Value
    ExecutePin=@n12.True
   )
   &RigUnit_ModifyTransforms(
    ItemToModify=((Item=(Type=Bone,Name="base_sleeves_bk_l"),Transform=(Rotation=(X=0,Y=0,Z=0,W=1),Translation=(X=0,Y=-10,Z=0),Scale3D=(X=1,Y=1,Z=1))))
    WeightMinimum=0
    Weight=RigVMDispatchNode(BoneName=root, Default=1, Name=deformRig_l_upperarm_bk_alpha) {template="DISPATCH_RigDispatch_GetAnimAttribute(in Name,in BoneName,in Default,out Value,out Found)"}.Value
    ExecutePin=@RigVMFunction_Sequence.F
   )
   &RigUnit_ModifyTransforms(
    ItemToModify=((Item=(Type=Bone,Name="base_sleeves_bk_r"),Transform=(Rotation=(X=0,Y=0,Z=0,W=1),Translation=(X=0,Y=10,Z=0),Scale3D=(X=1,Y=1,Z=1))))
    WeightMinimum=0
    Weight=RigVMDispatchNode(BoneName=root, Default=1, Name=deformRig_r_upperarm_bk_alpha) {template="DISPATCH_RigDispatch_GetAnimAttribute(in Name,in BoneName,in Default,out Value,out Found)"}.Value
    ExecutePin=@RigVMFunction_Sequence.G
   )
   n13 = &RigUnit_SphericalPoseReader(
    ActiveRegionScaleFactors=(PositiveWidth=3,NegativeWidth=0.1,PositiveHeight=3,NegativeHeight=3)
    ActiveRegionSize=0.1
    Debug=(bDrawDebug=false,bDraw2D=false,DebugScale=25,DebugSegments=20,DebugThickness=0.25)
    DriverAxis=(X=1,Y=0,Z=0)
    DriverItem=(Type=Bone,Name="thigh_r")
    FalloffRegionScaleFactors=(PositiveWidth=1,NegativeWidth=1,PositiveHeight=1,NegativeHeight=1)
    FalloffSize=0.45
    OptionalParentItem=(Type=Bone)
    RotationOffset=(X=45,Y=0,Z=180)
    ExecutePin=@RigVMFunction_Sequence.H
   )
   > &RigUnit_ModifyTransforms(
    ItemToModify="((Item=(Type=Bone,Name=\"base_skirt_fr_r\"),Transform=(Rotation=(X=0.0467,Y=0.2522,Z=-0.1761,W=0.9504),Translation=(X=0.2802,Y=-0.1144,Z=0),Scale3D=(X=1,Y=1,Z=1))),(Item=(Type=Bone,Name=\"base_skirt_sd_r\"),Transform=(Rotation=(X=0,Y=0.1886,Z=0,W=0.9821),Trans ...(truncated)"
    WeightMinimum=0
    Weight=&RigVMFunction_MathDoubleMul(A=@n13.OutputParam, B=RigVMVariableNode() {variable=Crouching}.Value).Result
   )
   RigVMFunctionReturnNode(ExecuteContext=@RigVMFunction_Sequence.I)
  ]}
 New Function_ContainedGraph {nodes [
   RigVMFunctionEntryNode()
   > &RigUnit_ParentConstraint(
    AdvancedSettings=(InterpolationType=Average,RotationOrderForFilter=XZY)
    Child=(Type=Bone,Name="base_skirt_fr_c")
    Filter=(TranslationFilter=(bX=False,bY=False,bZ=False),RotationFilter=(bX=True,bY=True,bZ=True),ScaleFilter=(bX=True,bY=True,bZ=True))
    Parents=((Item=(Type=Bone,Name="thigh_l"),Weight=0.5),(Item=(Type=Bone,Name="thigh_r"),Weight=0.5))
    Weight=1
   )
   > &RigUnit_ParentConstraint(
    AdvancedSettings=(InterpolationType=Average,RotationOrderForFilter=XZY)
    Child=(Type=Bone,Name="base_skirt_bk_c")
    Filter=(TranslationFilter=(bX=False,bY=False,bZ=False),RotationFilter=(bX=True,bY=True,bZ=True),ScaleFilter=(bX=True,bY=True,bZ=True))
    Parents=((Item=(Type=Bone,Name="thigh_l"),Weight=0.5),(Item=(Type=Bone,Name="thigh_r"),Weight=0.5))
    Weight=1
   )
   > &RigUnit_ParentConstraint(
    AdvancedSettings=(InterpolationType=Average,RotationOrderForFilter=XZY)
    Child=(Type=Bone,Name="base_skirt_fr_l")
    Filter=(TranslationFilter=(bX=False,bY=False,bZ=False),RotationFilter=(bX=True,bY=True,bZ=True),ScaleFilter=(bX=True,bY=True,bZ=True))
    Parents=((Item=(Type=Bone,Name="thigh_l"),Weight=1))
    Weight=1
   )
   > &RigUnit_ParentConstraint(
    AdvancedSettings=(InterpolationType=Average,RotationOrderForFilter=XZY)
    Child=(Type=Bone,Name="base_skirt_sd_l")
    Filter=(TranslationFilter=(bX=False,bY=False,bZ=False),RotationFilter=(bX=True,bY=True,bZ=True),ScaleFilter=(bX=True,bY=True,bZ=True))
    Parents=((Item=(Type=Bone,Name="thigh_l"),Weight=1))
    Weight=1
   )
   > &RigUnit_ParentConstraint(
    AdvancedSettings=(InterpolationType=Average,RotationOrderForFilter=XZY)
    Child=(Type=Bone,Name="base_skirt_bk_l")
    Filter=(TranslationFilter=(bX=False,bY=False,bZ=False),RotationFilter=(bX=True,bY=True,bZ=True),ScaleFilter=(bX=True,bY=True,bZ=True))
    Parents=((Item=(Type=Bone,Name="thigh_l"),Weight=1))
    Weight=1
   )
   > &RigUnit_ParentConstraint(
    AdvancedSettings=(InterpolationType=Average,RotationOrderForFilter=XZY)
    Child=(Type=Bone,Name="base_skirt_fr_r")
    Filter=(TranslationFilter=(bX=False,bY=False,bZ=False),RotationFilter=(bX=True,bY=True,bZ=True),ScaleFilter=(bX=True,bY=True,bZ=True))
    Parents=((Item=(Type=Bone,Name="thigh_r"),Weight=1))
    Weight=1
   )
   > &RigUnit_ParentConstraint(
    AdvancedSettings=(InterpolationType=Average,RotationOrderForFilter=XZY)
    Child=(Type=Bone,Name="base_skirt_sd_r")
    Filter=(TranslationFilter=(bX=False,bY=False,bZ=False),RotationFilter=(bX=True,bY=True,bZ=True),ScaleFilter=(bX=True,bY=True,bZ=True))
    Parents=((Item=(Type=Bone,Name="thigh_r"),Weight=1))
    Weight=1
   )
   > &RigUnit_ParentConstraint(
    AdvancedSettings=(InterpolationType=Average,RotationOrderForFilter=XZY)
    Child=(Type=Bone,Name="base_skirt_bk_r")
    Filter=(TranslationFilter=(bX=False,bY=False,bZ=False),RotationFilter=(bX=True,bY=True,bZ=True),ScaleFilter=(bX=True,bY=True,bZ=True))
    Parents=((Item=(Type=Bone,Name="thigh_r"),Weight=1))
    Weight=1
   )
   > RigVMFunctionReturnNode()
  ]}
 New Function_ContainedGraph {nodes [
   RigVMFunctionEntryNode()
   > &RigUnit_ModifyTransforms(
    ItemToModify="((Item=(Type=Bone,Name=\"base_tie_01\"),Transform=(Rotation=(X=0,Y=0,Z=-0.0872,W=0.9962),Translation=(X=0,Y=0,Z=0),Scale3D=(X=1,Y=1,Z=1))),(Item=(Type=Bone,Name=\"base_tie_02\"),Transform=(Rotation=(X=0,Y=-0.3181,Z=0,W=0.948),Translation=(X=0,Y=0,Z=0),Scale3D= ...(truncated)"
    WeightMinimum=0
    Weight=RigVMVariableNode() {variable=Crouching}.Value
   )
   > RigVMFunctionReturnNode()
  ]}
 New Function_ContainedGraph {nodes [
   RigVMFunctionEntryNode()
   > RigVMFunction_Sequence = RigVMAggregateNode()
   &RigUnit_ModifyTransforms(
    ItemToModify=((Item=(Type=Bone,Name="base_belt_fr_l"),Transform=(Rotation=(X=0,Y=0.0837,Z=0,W=0.9965),Translation=(X=1.7542,Y=0,Z=0.278),Scale3D=(X=1,Y=1,Z=1))))
    WeightMinimum=0
    Weight=RigVMDispatchNode(BoneName=root, Default=1, Name=deformRig_l_thigh_up_alpha) {template="DISPATCH_RigDispatch_GetAnimAttribute(in Name,in BoneName,in Default,out Value,out Found)"}.Value
    ExecutePin=@RigVMFunction_Sequence.A
   )
   n1 = &RigUnit_SphericalPoseReader(
    ActiveRegionScaleFactors=(PositiveWidth=3,NegativeWidth=0.1,PositiveHeight=3,NegativeHeight=3)
    ActiveRegionSize=0.1
    Debug=(bDrawDebug=false,DebugScale=25,DebugSegments=20,DebugThickness=0.25)
    DriverAxis=(X=-1,Y=0,Z=0)
    DriverItem=(Type=Bone,Name="thigh_l")
    FalloffRegionScaleFactors=(PositiveWidth=1,NegativeWidth=1,PositiveHeight=1,NegativeHeight=1)
    FalloffSize=0.45
    OptionalParentItem=(Type=Bone)
    RotationOffset=(X=180,Y=0,Z=0)
    ExecutePin=@RigVMFunction_Sequence.C
   )
   > &RigUnit_ModifyTransforms(
    ItemToModify=((Item=(Type=Bone,Name="base_belt_sd_l"),Transform=(Rotation=(X=0,Y=0,Z=-0.3188,W=0.9478),Translation=(X=0.9677,Y=0,Z=0),Scale3D=(X=1,Y=1,Z=1))))
    WeightMinimum=0
    Weight=@n1.OutputParam
   )
   > n2 = &RigUnit_SphericalPoseReader(
    ActiveRegionScaleFactors=(PositiveWidth=3,NegativeWidth=0.1,PositiveHeight=3,NegativeHeight=3)
    ActiveRegionSize=0.1
    Debug=(bDrawDebug=false,DebugScale=25,DebugSegments=20,DebugThickness=0.25)
    DriverAxis=(X=1,Y=0,Z=0)
    DriverItem=(Type=Bone,Name="thigh_r")
    FalloffRegionScaleFactors=(PositiveWidth=1,NegativeWidth=1,PositiveHeight=1,NegativeHeight=1)
    FalloffSize=0.45
    OptionalParentItem=(Type=Bone)
    RotationOffset=(X=0,Y=0,Z=180)
   )
   > &RigUnit_ModifyTransforms(
    ItemToModify=((Item=(Type=Bone,Name="base_belt_sd_r"),Transform=(Rotation=(X=0,Y=0,Z=-0.1815,W=0.9834),Translation=(X=-4.7704,Y=-0.5006,Z=-0.0003),Scale3D=(X=1,Y=1,Z=1))))
    WeightMinimum=0
    Weight=@n2.OutputParam
   )
   > &RigUnit_ModifyTransforms(
    ItemToModify=((Item=(Type=Bone,Name="base_belt_fr_r"),Transform=(Rotation=(X=0,Y=-0.0994,Z=0,W=0.995),Translation=(X=-0.5858,Y=0,Z=-0.1446),Scale3D=(X=1,Y=1,Z=1))))
    WeightMinimum=0
    Weight=@n2.OutputParam
   )
   > &RigUnit_ModifyTransforms(
    ItemToModify=((Item=(Type=Bone,Name="dyn_belt_bk_r"),Transform=(Rotation=(X=0.0583,Y=-0.173,Z=0.0103,W=0.9831),Translation=(X=-1.1043,Y=0,Z=0),Scale3D=(X=1,Y=1,Z=1))))
    WeightMinimum=0
    Weight=@n2.OutputParam
   )
   > &RigUnit_ModifyTransforms(
    ItemToModify=((Item=(Type=Bone,Name="base_belt_sd_r"),Transform=(Rotation=(X=0,Y=0,Z=0,W=1),Translation=(X=-1,Y=-1,Z=1),Scale3D=(X=1,Y=1,Z=1))))
    Weight=1
    WeightMinimum=0
   )
   n3 = &RigUnit_SphericalPoseReader(
    ActiveRegionScaleFactors=(PositiveWidth=3,NegativeWidth=0.1,PositiveHeight=3,NegativeHeight=3)
    ActiveRegionSize=0.1
    Debug=(bDrawDebug=false,DebugScale=25,DebugSegments=20,DebugThickness=0.25)
    DriverAxis=(X=-1,Y=0,Z=0)
    DriverItem=(Type=Bone,Name="base_skirt_fr_r")
    FalloffRegionScaleFactors=(PositiveWidth=1,NegativeWidth=1,PositiveHeight=1,NegativeHeight=1)
    FalloffSize=0.45
    OptionalParentItem=(Type=Bone)
    RotationOffset=(X=90,Y=0,Z=0)
    ExecutePin=@RigVMFunction_Sequence.D
   )
   > &RigUnit_ModifyTransforms(
    ItemToModify=((Item=(Type=Bone,Name="base_belt_fr_r"),Transform=(Rotation=(X=0,Y=0.0361,Z=0,W=0.9993),Translation=(X=0.4,Y=0,Z=0.519),Scale3D=(X=1,Y=1,Z=1))))
    WeightMinimum=0
    Weight=@n3.OutputParam
   )
   RigVMFunctionReturnNode(ExecuteContext=@RigVMFunction_Sequence.E)
   n4 = &RigUnit_SphericalPoseReader(
    ActiveRegionScaleFactors=(PositiveWidth=3,NegativeWidth=0.1,PositiveHeight=3,NegativeHeight=3)
    ActiveRegionSize=0.1
    Debug=(bDrawDebug=false,DebugScale=25,DebugSegments=20,DebugThickness=0.25)
    DriverAxis=(X=-1,Y=0,Z=0)
    DriverItem=(Type=Bone,Name="base_skirt_bk_r")
    FalloffRegionScaleFactors=(PositiveWidth=1,NegativeWidth=1,PositiveHeight=1,NegativeHeight=1)
    FalloffSize=0.45
    OptionalParentItem=(Type=Bone)
    RotationOffset=(X=-90,Y=0,Z=0)
    ExecutePin=@RigVMFunction_Sequence.F
   )
   > &RigUnit_ModifyTransforms(
    ItemToModify=((Item=(Type=Bone,Name="base_belt_bk_r"),Transform=(Rotation=(X=0,Y=-0.1989,Z=0,W=0.98),Translation=(X=-5.1273,Y=0,Z=0.9663),Scale3D=(X=1,Y=1,Z=1))))
    WeightMinimum=0
    Weight=@n4.OutputParam
   )
   > &RigUnit_ModifyTransforms(
    ItemToModify=((Item=(Type=Bone,Name="base_belt_sd_r"),Transform=(Rotation=(X=0,Y=0,Z=0,W=1),Translation=(X=-3.1377,Y=0,Z=0),Scale3D=(X=1,Y=1,Z=1))))
    WeightMinimum=0
    Weight=@n4.OutputParam
   )
   > n5 = &RigUnit_SphericalPoseReader(
    ActiveRegionScaleFactors=(PositiveWidth=3,NegativeWidth=0.1,PositiveHeight=3,NegativeHeight=3)
    ActiveRegionSize=0.1
    Debug=(bDrawDebug=false,DebugScale=25,DebugSegments=20,DebugThickness=0.25)
    DriverAxis=(X=-1,Y=0,Z=0)
    DriverItem=(Type=Bone,Name="base_skirt_bk_l")
    FalloffRegionScaleFactors=(PositiveWidth=1,NegativeWidth=1,PositiveHeight=1,NegativeHeight=1)
    FalloffSize=0.45
    OptionalParentItem=(Type=Bone)
    RotationOffset=(X=-90,Y=0,Z=0)
   )
   > &RigUnit_ModifyTransforms(
    ItemToModify=((Item=(Type=Bone,Name="base_belt_bk_l"),Transform=(Rotation=(X=0,Y=0,Z=0,W=1),Translation=(X=0,Y=0,Z=0),Scale3D=(X=1,Y=1,Z=1))))
    WeightMinimum=0
    Weight=@n5.OutputParam
   )
   > &RigUnit_ModifyTransforms(
    ItemToModify=((Item=(Type=Bone,Name="base_belt_sd_l"),Transform=(Rotation=(X=0,Y=0,Z=0,W=1),Translation=(X=0,Y=0,Z=0),Scale3D=(X=1,Y=1,Z=1))))
    WeightMinimum=0
    Weight=@n5.OutputParam
   )
   RigVMFunction_MathBoolOr = &RigVMFunction_MathBoolOr(B=RigVMVariableNode() {variable=Sides}.Value, A=RigVMVariableNode() {variable=FrontOrBack}.Value)
   n6 = &RigVMFunction_ControlFlowBranch(Condition=@RigVMFunction_MathBoolOr.Result, ExecuteContext=@RigVMFunction_Sequence.G)
   n7 = RigVMDispatchNode(BoneName=root, Default=1, Name=deformRig_r_thigh_up_alpha) {template="DISPATCH_RigDispatch_GetAnimAttribute(in Name,in BoneName,in Default,out Value,out Found)"}
   Multiply = &RigVMFunction_MathDoubleMul(A=RigVMVariableNode() {variable=CrouchingMoving2D}.Value, B=@n7.Value)
   &RigUnit_ModifyTransforms(
    ItemToModify="((Item=(Type=Bone,Name=\"base_belt_fr_r\"),Transform=(Rotation=(X=0,Y=0,Z=0,W=1),Translation=(X=-3,Y=0,Z=0),Scale3D=(X=1,Y=1,Z=1))),(Item=(Type=Bone,Name=\"base_belt_sd_r\"),Transform=(Rotation=(X=0,Y=0,Z=0,W=1),Translation=(X=0,Y=0,Z=0),Scale3D=(X=1,Y=1,Z=1)) ...(truncated)"
    WeightMinimum=0
    Weight=@Multiply.Result
    ExecutePin=@n6.True
   )
   RigVMFunction_ControlFlowBranch = &RigVMFunction_ControlFlowBranch(Condition=RigVMVariableNode() {variable=Crouching_Bool}.Value, ExecuteContext=@RigVMFunction_Sequence.B)
   &RigUnit_ModifyTransforms(
    ItemToModify=((Item=(Type=Bone,Name="base_belt_fr_r"),Transform=(Rotation=(X=0,Y=-0.1825,Z=0,W=0.9832),Translation=(X=-6.3024,Y=0,Z=-0.0966),Scale3D=(X=1,Y=1,Z=1))))
    WeightMinimum=0
    Weight=@n7.Value
    ExecutePin=@RigVMFunction_ControlFlowBranch.False
   )
   > &RigUnit_ModifyTransforms(
    ItemToModify=((Item=(Type=Bone,Name="base_belt_sd_r"),Transform=(Rotation=(X=0,Y=0,Z=0,W=1),Translation=(X=0,Y=0,Z=0),Scale3D=(X=1,Y=1,Z=1))))
    WeightMinimum=0
    Weight=@n7.Value
   )
   &RigUnit_ModifyTransforms(
    ItemToModify=((Item=(Type=Bone,Name="base_belt_fr_r"),Transform=(Rotation=(X=0,Y=-0.1825,Z=0,W=0.9832),Translation=(X=-5.1,Y=0,Z=0.3561),Scale3D=(X=1,Y=1,Z=1))))
    WeightMinimum=0
    Weight=@n7.Value
    ExecutePin=@RigVMFunction_ControlFlowBranch.True
   )
   > &RigUnit_ModifyTransforms(
    ItemToModify=((Item=(Type=Bone,Name="base_belt_sd_r"),Transform=(Rotation=(X=0,Y=0,Z=0,W=1),Translation=(X=0,Y=0,Z=0),Scale3D=(X=1,Y=1,Z=1))))
    WeightMinimum=0
    Weight=@n7.Value
   )
  ]}
 RigVMFunctionLibrary {nodes [
   Belt Fix = RigVMCollapseNode(ExecuteContext=())
   Constrains = RigVMCollapseNode(ExecuteContext=())
   Controls = RigVMCollapseNode(DebugControls=False, ExecuteContext=())
   Delta Rotation = RigVMCollapseNode(ExecuteContext=(), debugRotation=False, onlyTwist=False, twistAxis=(X=0,Y=0,Z=0))
   Skirt Fix = RigVMCollapseNode(ExecuteContext=())
   Transforms Fix = RigVMCollapseNode(ExecuteContext=())
  ]}
 RigVMFunction_Sequence_1_ContainedGraph {nodes [
   n1 = RigVMFunctionEntryNode()
   RigVMFunction_Sequence = &RigVMFunction_Sequence(ExecuteContext=@n1.ExecuteContext)
   n2 = &RigVMFunction_Sequence(ExecuteContext=@RigVMFunction_Sequence.B)
   n3 = &RigVMFunction_Sequence(ExecuteContext=@n2.B)
   n4 = &RigVMFunction_Sequence(ExecuteContext=@n3.B)
   n5 = &RigVMFunction_Sequence(ExecuteContext=@n4.B)
   n6 = &RigVMFunction_Sequence(ExecuteContext=@n5.B)
   n7 = &RigVMFunction_Sequence(ExecuteContext=@n6.B)
   RigVMFunctionReturnNode(A=@RigVMFunction_Sequence.A, B=@n2.A, D=@n3.A, E=@n4.A, F=@n5.A, G=@n6.A, H=@n7.A, I=@n7.B)
  ]}
 RigVMFunction_Sequence_1_ContainedGraph {nodes [
   n1 = RigVMFunctionEntryNode()
   RigVMFunction_Sequence = &RigVMFunction_Sequence(ExecuteContext=@n1.ExecuteContext)
   n2 = &RigVMFunction_Sequence(ExecuteContext=@RigVMFunction_Sequence.B)
   n3 = &RigVMFunction_Sequence(ExecuteContext=@n2.B)
   n4 = &RigVMFunction_Sequence(ExecuteContext=@n3.B)
   n5 = &RigVMFunction_Sequence(ExecuteContext=@n4.B)
   n6 = &RigVMFunction_Sequence(ExecuteContext=@n5.B)
   RigVMFunctionReturnNode(A=@RigVMFunction_Sequence.A, B=@n2.A, C=@n3.A, D=@n4.A, E=@n5.A, F=@n6.A, G=@n6.B)
  ]}
 RigVMModel {nodes [
   RigUnit_BeginExecution = &RigUnit_BeginExecution()
   Controls = RigVMFunctionReferenceNode(DebugControls=false, ExecuteContext=@RigUnit_BeginExecution.ExecutePin)
   > FN_DeformRig_FNC = RigVMFunctionReferenceNode(
    DeformRig_ Curve="(deformRig_l_upperarm_up_alpha_72_#1=true,deformRig_r_upperarm_up_alpha_73_#2=true,deformRig_l_upperarm_fr_alpha_74_#3=true,deformRig_r_upperarm_fr_alpha_75_CB51B80C4 ...(truncated)"
    DeformRig_Bool="(upperarm_1_#4=True,elbow_3_#5=True,wrist_7_#6=True,groin_8_#7=true,glute_17_#8=false,knee_9_40C408324CA26 ...(truncated)"
    DeformRig_Debug=(upperarm_debug_18_#4=false,wrist_debug_20_#6=false,lat_debug_26_#9=false,groinGlute_debug_27_#7=false)
    DeformRig_Value="(upperarm_up_18_#10=(Rotation=(X=0,Y=0,Z=0,W=1),Translation=(X=0,Y=0,Z=8),Scale3D=(X=0.7,Y=1,Z=1)),upperarm_fwd_fr_19_#11=(Rotation=(X=0,Y=0,Z=0,W=1),Translation=(X=0,Y=10,Z=0),Scale3D=(X=0.5,Y=1,Z= ...(truncated)"
   )
   > Constrains = RigVMFunctionReferenceNode()
   > Skirt Fix = RigVMFunctionReferenceNode()
   > Belt Fix = RigVMFunctionReferenceNode()
  ]}
]
hierarchy [
 root {loc=<0 0 0>, rot=<0 0 0 1>}
 pelvis {parent=root, loc=<0 -0.5301 98.0571>, rot=<0.0127 0.707 -0.0127 -0.707>}
 spine_01 {parent=pelvis, loc=<6.1208 -0.9328 0>, rot=<0 0 0.034 -0.9994>}
 spine_02 {parent=spine_01, loc=<6.5997 -0.2095 0>, rot=<0 0 0.0238 -0.9997>}
 spine_03 {parent=spine_02, loc=<6.4914 -0.1237 0>, rot=<0 0 0.0149 -0.9999>}
 spine_04 {parent=spine_03, loc=<6.9598 0.0794 0>, rot=<0 0 -0.0132 -0.9999>}
 spine_05 {parent=spine_04, loc=<6.9987 0 0>, rot=<0 0 -0.0015 -1>}
 attach_cape {parent=spine_05, loc=<9.3216 -8.4501 0>, rot=<0.0616 -0.7044 -0.0616 -0.7044>}
 neck_01 {parent=spine_05, loc=<11.3138 0.1529 0>, rot=<0 0 -0.2003 -0.9797>}
 neck_02 {parent=neck_01, loc=<4.5627 0 0>, rot=<0 0 0.0504 -0.9987>}
 head {parent=neck_02, loc=<5.3914 0 0>, rot=<0 0 0.115 -0.9934>}
 attach_backpack {parent=spine_05, loc=<0 -8.35 0>, rot=<-0.0093 -0.707 0.0093 -0.707>}
 clavicle_l {parent=spine_05, loc=<5.4831 -0.1053 -3.2946>, rot=<0.0465 -0.7289 -0.009 -0.6829>}
 upperarm_l {parent=clavicle_l, loc=<9.8658 -1.2136 -0.4105>, rot=<0.0164 -0.3245 -0.0269 -0.9454>}
 lowerarm_l {parent=upperarm_l, loc=<26.465 -0.8345 0.259>, rot=<0.0036 0.0161 -0.2298 -0.9731>}
 hand_l {parent=lowerarm_l, loc=<22.7355 -0.3203 -0.6105>, rot=<-0.6784 -0.0223 0.0735 -0.7307>}
 thumb_01_l {parent=hand_l, loc=<1.3118 -0.3191 -2.459>, rot=<0.6546 -0.4042 0.0664 -0.6354>}
 thumb_02_l {parent=thumb_01_l, loc=<5.9336 -0.0448 0.0401>, rot=<-0.0037 0.0213 0.298 -0.9543>}
 thumb_03_l {parent=thumb_02_l, loc=<2.2223 0.3682 -0.0907>, rot=<0.0118 -0.013 0.0066 -0.9998>}
 index_metacarpal_l {parent=hand_l, loc=<1.6535 0.0183 -2.96>, rot=<0.0158 -0.0124 -0.0173 -0.9996>}
 index_01_l {parent=index_metacarpal_l, loc=<8.3008 0.0357 -0.0524>, rot=<-0.0044 0.0485 0.0924 -0.9945>}
 index_02_l {parent=index_01_l, loc=<3.3587 -0.3004 0.0452>, rot=<-0.002 0.0027 0.3225 -0.9466>}
 index_03_l {parent=index_02_l, loc=<2.5578 -0.123 -0.0252>, rot=<-0.002 0.0067 0.2558 -0.9667>}
 middle_metacarpal_l {parent=hand_l, loc=<2.1856 -0.1725 -0.5984>, rot=<0.006 0.0168 -0.0329 -0.9993>}
 middle_01_l {parent=middle_metacarpal_l, loc=<7.3275 -0.2447 0.0129>, rot=<-0.0208 0.0042 0.0858 -0.9961>}
 middle_02_l {parent=middle_01_l, loc=<4.0123 0.1401 0.1175>, rot=<-0.0218 0.0258 0.3512 -0.9357>}
 middle_03_l {parent=middle_02_l, loc=<2.6077 0.0534 0.1373>, rot=<0.0082 0.0188 0.1708 -0.9851>}
 ring_metacarpal_l {parent=hand_l, loc=<2.2074 -0.3625 1.3638>, rot=<-0.0284 0.0722 -0.0458 -0.9959>}
 ring_01_l {parent=ring_metacarpal_l, loc=<7.0295 -0.6197 -0.1446>, rot=<-0.0489 -0.0252 0.1892 -0.9804>}
 ring_02_l {parent=ring_01_l, loc=<3.6174 0.4647 -0.1027>, rot=<0 0 0.2806 -0.9598>}
 ring_03_l {parent=ring_02_l, loc=<2.4152 0.114 0.0199>, rot=<0 0 0.1189 -0.9929>}
 pinky_metacarpal_l {parent=hand_l, loc=<1.7616 -0.9753 3.3814>, rot=<-0.0397 0.0814 -0.069 -0.9935>}
 pinky_01_l {parent=pinky_metacarpal_l, loc=<7.1709 -0.4831 -0.1094>, rot=<-0.0654 -0.0035 0.2242 -0.9723>}
 pinky_02_l {parent=pinky_01_l, loc=<2.8274 0.0893 -0.1043>, rot=<0.0037 0.0011 0.2854 -0.9584>}
 pinky_03_l {parent=pinky_02_l, loc=<2.2046 0.0307 -0.0078>, rot=<-0.0002 0.0098 0.1445 -0.9895>}
 deform_wrist_dn_base_l {parent=hand_l, loc=<0 -1.5326 0>, rot=<0 0 0 -1>}
 deform_wrist_dn_l {parent=deform_wrist_dn_base_l, loc=<0 0 0>, rot=<0 0 0 1>}
 deform_wrist_up_base_l {parent=hand_l, loc=<0 2.3009 0>, rot=<0 0 0 -1>}
 deform_wrist_up_l {parent=deform_wrist_up_base_l, loc=<0 0 0>, rot=<0 0 0 1>}
 hand_attach_l {parent=hand_l, loc=<8.5 -3 0>, rot=<0.7071 0.7071 0 0>}
 weapon_l {parent=hand_l, loc=<-0.1974 -3.4111 0.3811>, rot=<0 0 0 1>}
 lowerarm_twist_02_l {parent=lowerarm_l, loc=<10.6707 0 0>, rot=<0 0 0 1>}
 lowerarm_twist_01_l {parent=lowerarm_l, loc=<18.316 0 0>, rot=<0 0 0 1>}
 deform_elbow_in_base_l {parent=lowerarm_l, loc=<0.1834 3.6062 -0.2572>, rot=<0 0 0 -1>}
 deform_elbow_in_l {parent=deform_elbow_in_base_l, loc=<0 0 0>, rot=<0 0 0 1>}
 deform_elbow_out_base_l {parent=lowerarm_l, loc=<0 -4 0>, rot=<0 0 0 1>}
 deform_elbow_out_l {parent=deform_elbow_out_base_l, loc=<0 0 0>, rot=<0 0 0 1>}
 upperarm_twist_01_l {parent=upperarm_l, loc=<7.2135 -0.1134 0>, rot=<0 0 0 1>}
 deform_upperarm_bk_base_l {parent=upperarm_twist_01_l, loc=<-6.4053 -4.3746 0.314>, rot=<0 0 0 -1>}
 deform_upperarm_bk_l {parent=deform_upperarm_bk_base_l, loc=<0 0 0>, rot=<0 0 0 1>}
 deform_upperarm_fr_base_l {parent=upperarm_twist_01_l, loc=<-5.6312 5.092 0.3354>, rot=<0 0 0 -1>}
 deform_upperarm_fr_l {parent=deform_upperarm_fr_base_l, loc=<0 0 0>, rot=<0 0 0 1>}
 deform_upperarm_up_base_l {parent=upperarm_twist_01_l, loc=<-8.891 0.4825 3.9517>, rot=<0 0 0 -1>}
 deform_upperarm_up_l {parent=deform_upperarm_up_base_l, loc=<0 0 0>, rot=<0 0 0 1>}
 upperarm_twist_02_l {parent=upperarm_l, loc=<19.7998 -0.1134 0>, rot=<0 0 0 1>}
 MSTR_sleeve_l {parent=upperarm_l, loc=<0 0 0>, rot=<0 0 0 1>}
 sleeves_low_l {parent=MSTR_sleeve_l, loc=<8.2609 0 0>, rot=<0 0 0 1>}
 base_sleeves_bk_l {parent=MSTR_sleeve_l, loc=<0 -2.5003 0>, rot=<0 0 0 1>}
 dyn_sleeves_bk_l {parent=base_sleeves_bk_l, loc=<0 0 0>, rot=<0 0 0 1>}
 base_sleeves_fr_l {parent=MSTR_sleeve_l, loc=<0 2.5 0>, rot=<0 0 0 1>}
 dyn_sleeves_fr_l {parent=base_sleeves_fr_l, loc=<0 0 0>, rot=<0 0 0 1>}
 clavicle_r {parent=spine_05, loc=<5.4831 -0.1053 3.2946>, rot=<0.7289 0.0465 -0.6829 0.009>}
 upperarm_r {parent=clavicle_r, loc=<-9.8658 1.2136 0.4105>, rot=<0.0164 -0.3245 -0.0269 -0.9454>}
 lowerarm_r {parent=upperarm_r, loc=<-26.465 0.8345 -0.259>, rot=<0.0036 0.0161 -0.2298 -0.9731>}
 hand_r {parent=lowerarm_r, loc=<-22.7355 0.3203 0.6106>, rot=<-0.6784 -0.0223 0.0735 -0.7307>}
 thumb_01_r {parent=hand_r, loc=<-1.3118 0.3191 2.459>, rot=<0.6546 -0.4042 0.0664 -0.6354>}
 thumb_02_r {parent=thumb_01_r, loc=<-5.9336 0.0448 -0.0401>, rot=<-0.0037 0.0213 0.298 -0.9543>}
 thumb_03_r {parent=thumb_02_r, loc=<-2.2223 -0.3682 0.0907>, rot=<0.0118 -0.013 0.0066 -0.9998>}
 index_metacarpal_r {parent=hand_r, loc=<-1.6535 -0.0183 2.96>, rot=<0.0158 -0.0124 -0.0173 -0.9996>}
 index_01_r {parent=index_metacarpal_r, loc=<-8.3008 -0.0357 0.0524>, rot=<-0.0044 0.0485 0.0924 -0.9945>}
 index_02_r {parent=index_01_r, loc=<-3.3587 0.3004 -0.0452>, rot=<-0.002 0.0027 0.3225 -0.9466>}
 index_03_r {parent=index_02_r, loc=<-2.5578 0.123 0.0252>, rot=<-0.002 0.0067 0.2558 -0.9667>}
 middle_metacarpal_r {parent=hand_r, loc=<-2.1856 0.1725 0.5984>, rot=<0.006 0.0168 -0.0329 -0.9993>}
 middle_01_r {parent=middle_metacarpal_r, loc=<-7.3275 0.2447 -0.0129>, rot=<-0.0208 0.0042 0.0858 -0.9961>}
 middle_02_r {parent=middle_01_r, loc=<-4.0123 -0.1401 -0.1175>, rot=<-0.0218 0.0258 0.3512 -0.9357>}
 middle_03_r {parent=middle_02_r, loc=<-2.6077 -0.0534 -0.1373>, rot=<0.0082 0.0188 0.1708 -0.9851>}
 ring_metacarpal_r {parent=hand_r, loc=<-2.2074 0.3625 -1.3638>, rot=<-0.0284 0.0722 -0.0458 -0.9959>}
 ring_01_r {parent=ring_metacarpal_r, loc=<-7.0295 0.6197 0.1446>, rot=<-0.0489 -0.0252 0.1892 -0.9804>}
 ring_02_r {parent=ring_01_r, loc=<-3.6174 -0.4647 0.1027>, rot=<0 0 0.2806 -0.9598>}
 ring_03_r {parent=ring_02_r, loc=<-2.4152 -0.114 -0.0199>, rot=<0 0 0.1189 -0.9929>}
 pinky_metacarpal_r {parent=hand_r, loc=<-1.7616 0.9753 -3.3814>, rot=<-0.0397 0.0814 -0.069 -0.9935>}
 pinky_01_r {parent=pinky_metacarpal_r, loc=<-7.1709 0.4831 0.1094>, rot=<-0.0654 -0.0035 0.2242 -0.9723>}
 pinky_02_r {parent=pinky_01_r, loc=<-2.8274 -0.0893 0.1043>, rot=<0.0037 0.0011 0.2854 -0.9584>}
 pinky_03_r {parent=pinky_02_r, loc=<-2.2046 -0.0307 0.0078>, rot=<-0.0002 0.0098 0.1445 -0.9895>}
 deform_wrist_dn_base_r {parent=hand_r, loc=<-0.0001 1.5328 0>, rot=<0 0 0 -1>}
 deform_wrist_dn_r {parent=deform_wrist_dn_base_r, loc=<0 0 0>, rot=<0 0 0 1>}
 deform_wrist_up_base_r {parent=hand_r, loc=<-0.0002 -2.3007 -0.0001>, rot=<0 0 0 -1>}
 deform_wrist_up_r {parent=deform_wrist_up_base_r, loc=<0 0 0>, rot=<0 0 0 1>}
 weapon_r {parent=hand_r, loc=<0.1974 3.4111 -0.3811>, rot=<0 0 0 1>}
 hand_attach_r {parent=hand_r, loc=<-8.5001 2.9999 0>, rot=<0 0 -0.7071 -0.7071>}
 lowerarm_twist_02_r {parent=lowerarm_r, loc=<-10.6707 0 0>, rot=<0 0 0 1>}
 lowerarm_twist_01_r {parent=lowerarm_r, loc=<-18.316 0 0>, rot=<0 0 0 1>}
 deform_elbow_in_base_r {parent=lowerarm_r, loc=<-0.1834 -3.6062 0.2572>, rot=<0 0 0 -1>}
 deform_elbow_in_r {parent=deform_elbow_in_base_r, loc=<0 0 0>, rot=<0 0 0 1>}
 deform_elbow_out_base_r {parent=lowerarm_r, loc=<0 4 -0.0001>, rot=<0 0 0 -1>}
 deform_elbow_out_r {parent=deform_elbow_out_base_r, loc=<0 0 0>, rot=<0 0 0 1>}
 upperarm_twist_01_r {parent=upperarm_r, loc=<-7.2135 0.1134 0>, rot=<0 0 0 1>}
 deform_upperarm_bk_base_r {parent=upperarm_twist_01_r, loc=<6.4053 4.3746 -0.314>, rot=<0 0 0 -1>}
 deform_upperarm_bk_r {parent=deform_upperarm_bk_base_r, loc=<0 0 0>, rot=<0 0 0 1>}
 deform_upperarm_fr_base_r {parent=upperarm_twist_01_r, loc=<5.6312 -5.092 -0.3354>, rot=<0 0 0 -1>}
 deform_upperarm_fr_r {parent=deform_upperarm_fr_base_r, loc=<0 0 0>, rot=<0 0 0 1>}
 deform_upperarm_up_base_r {parent=upperarm_twist_01_r, loc=<8.891 -0.4825 -3.9517>, rot=<0 0 0 -1>}
 deform_upperarm_up_r {parent=deform_upperarm_up_base_r, loc=<0 0 0>, rot=<0 0 0 1>}
 upperarm_twist_02_r {parent=upperarm_r, loc=<-19.7998 0.1134 0>, rot=<0 0 0 1>}
 MSTR_sleeve_r {parent=upperarm_r, loc=<0.0002 0 -0.0003>, rot=<0 0 0 1>}
 sleeves_low_r {parent=MSTR_sleeve_r, loc=<-8.2608 0 0>, rot=<0 0 0 1>}
 base_sleeves_bk_r {parent=MSTR_sleeve_r, loc=<-0.0004 2.5003 0.0005>, rot=<0 0 0 1>}
 dyn_sleeves_bk_r {parent=base_sleeves_bk_r, loc=<0 0 0>, rot=<0 0 0 1>}
 base_sleeves_fr_r {parent=MSTR_sleeve_r, loc=<-0.0002 -2.5 0.0003>, rot=<0 0 0 1>}
 dyn_sleeves_fr_r {parent=base_sleeves_fr_r, loc=<0 0 0>, rot=<0 0 0 1>}
 base_HoodieStrings_01_l {parent=spine_05, loc=<5.3097 7.1479 -2.2015>, rot=<-0.2682 -0.9627 0.032 -0.0156>}
 dyn_HoodieStrings_01_l {parent=base_HoodieStrings_01_l, loc=<0.0451 0.001 0.0637>, rot=<0 0 0 1>}
 dyn_HoodieStrings_02_l {parent=dyn_HoodieStrings_01_l, loc=<3.0686 0.0035 -0.0008>, rot=<-0.0057 -0.0072 0.0122 -0.9999>}
 dyn_HoodieStrings_03_l {parent=dyn_HoodieStrings_02_l, loc=<4.1809 0 0>, rot=<-0.0013 0.0012 0.0382 -0.9993>}
 dyn_HoodieStrings_04_l {parent=dyn_HoodieStrings_03_l, loc=<3.4009 0.0018 -0.0023>, rot=<-0.0043 0.003 0.171 -0.9853>}
 base_HoodieStrings_01_r {parent=spine_05, loc=<5.3519 7.115 2.0738>, rot=<-0.2642 -0.9644 -0.0045 -0.0125>}
 dyn_HoodieStrings_01_r {parent=base_HoodieStrings_01_r, loc=<0.0236 0.0155 0.0658>, rot=<0 0 0 1>}
 dyn_HoodieStrings_02_r {parent=dyn_HoodieStrings_01_r, loc=<3.105 0.0045 -0.0003>, rot=<0.0014 0.0021 0.0077 -1>}
 dyn_HoodieStrings_03_r {parent=dyn_HoodieStrings_02_r, loc=<4.2057 0.0018 -0.0007>, rot=<0 0 0 1>}
 deform_pec_aim_base_l {parent=spine_04, loc=<3.2456 9.3309 -1.2237>, rot=<0.0603 -0.4033 0.3318 -0.8507>}
 deform_pec_aim_l {parent=deform_pec_aim_base_l, loc=<0 0 0>, rot=<0 0 0 1>}
 deform_pec_base_l {parent=deform_pec_aim_l, loc=<9.3665 0 0>, rot=<0 0 0 1>}
 deform_pec_l {parent=deform_pec_base_l, loc=<0 0 0>, rot=<0 0 0 1>}
 deform_pec_aim_base_r {parent=spine_04, loc=<3.2452 9.3309 1.2237>, rot=<-0.4033 -0.0603 0.8507 0.3318>}
 deform_pec_aim_r {parent=deform_pec_aim_base_r, loc=<0 0 0>, rot=<0 0 0 1>}
 deform_pec_base_r {parent=deform_pec_aim_r, loc=<-9.3667 -0.0003 -0.0004>, rot=<0 0 0 1>}
 deform_pec_r {parent=deform_pec_base_r, loc=<0 0 0>, rot=<0 0 0 1>}
 base_pec_r {parent=spine_04, loc=<2.6141 11.6747 2.4313>, rot=<0.0294 -0.7065 -0.0294 -0.7065>}
 base_pec_l {parent=spine_04, loc=<2.6141 11.6747 -2.4313>, rot=<0.0294 -0.7065 -0.0294 -0.7065>}
 thigh_l {parent=pelvis, loc=<-1.0855 -0.2109 -8.7053>, rot=<0.0009 0.0355 0.0155 -0.9993>}
 calf_l {parent=thigh_l, loc=<-42.5284 -0.0019 -0.7297>, rot=<0.0006 -0.0161 -0.0391 -0.9991>}
 foot_l {parent=calf_l, loc=<-46.7401 -2.5014 -0.592>, rot=<0.0427 -0.0238 0.0353 -0.9982>}
 ball_l {parent=foot_l, loc=<-3.5658 10.8066 -0.1544>, rot=<0.0001 -0.0001 -0.7063 -0.7079>}
 calf_twist_02_l {parent=calf_l, loc=<-19.1546 -1.7943 -0.2638>, rot=<0.0237 0.0095 -0.0035 -0.9997>}
 calf_twist_01_l {parent=calf_l, loc=<-35.8483 -1.9197 -0.5794>, rot=<0.0237 0.0095 -0.0035 -0.9997>}
 deform_knee_in_base_l {parent=calf_l, loc=<0.8031 -5.0231 0.2456>, rot=<0 0 0 -1>}
 deform_knee_in_l {parent=deform_knee_in_base_l, loc=<0 0 0>, rot=<0 0 0 1>}
 deform_knee_out_base_l {parent=calf_l, loc=<0 4 0>, rot=<0 0 0 -1>}
 deform_knee_out_l {parent=deform_knee_out_base_l, loc=<0 0 0>, rot=<0 0 0 1>}
 thigh_twist_01_l {parent=thigh_l, loc=<-27.952 0 0>, rot=<0.0013 0 0 -1>}
 deform_groin_base_l {parent=thigh_l, loc=<0.4801 5.6776 -0.2486>, rot=<0 0 0 -1>}
 deform_groin_l {parent=deform_groin_base_l, loc=<0 0 0>, rot=<0 0 0 1>}
 base_Pouch {parent=thigh_l, loc=<-25.8302 4.4525 -5.4542>, rot=<-0.2407 -0.0466 -0.0296 -0.969>}
 dyn_Pouch {parent=base_Pouch, loc=<0 0 0>, rot=<0 0 0 1>}
 thigh_r {parent=pelvis, loc=<-1.0855 -0.2109 8.7053>, rot=<0.0355 -0.0009 0.9993 0.0155>}
 calf_r {parent=thigh_r, loc=<42.5284 0.0019 0.7297>, rot=<0.0006 -0.0161 -0.0391 -0.9991>}
 foot_r {parent=calf_r, loc=<46.7401 2.5014 0.592>, rot=<0.0427 -0.0238 0.0353 -0.9982>}
 ball_r {parent=foot_r, loc=<3.5658 -10.8066 0.1544>, rot=<0.0001 -0.0001 -0.7063 -0.7079>}
 calf_twist_02_r {parent=calf_r, loc=<19.1546 1.7943 0.2638>, rot=<0.0237 0.0095 -0.0035 -0.9997>}
 calf_twist_01_r {parent=calf_r, loc=<35.8483 1.9197 0.5794>, rot=<0.0237 0.0095 -0.0035 -0.9997>}
 deform_knee_in_base_r {parent=calf_r, loc=<-0.8031 5.0231 -0.2457>, rot=<0 0 0 -1>}
 deform_knee_in_r {parent=deform_knee_in_base_r, loc=<0 0 0>, rot=<0 0 0 1>}
 deform_knee_out_base_r {parent=calf_r, loc=<0 -4 0>, rot=<0 0 0 -1>}
 deform_knee_out_r {parent=deform_knee_out_base_r, loc=<0 0 0>, rot=<0 0 0 1>}
 thigh_twist_01_r {parent=thigh_r, loc=<27.952 0 0>, rot=<0.0013 0 0 -1>}
 deform_groin_base_r {parent=thigh_r, loc=<-0.4801 -5.6776 0.2486>, rot=<0 0 0 -1>}
 deform_groin_r {parent=deform_groin_base_r, loc=<0 0 0>, rot=<0 0 0 1>}
 deform_glute_base_l {parent=pelvis, loc=<-7.3466 -9.0213 -9.1245>, rot=<0.0172 0.031 -0.4477 -0.8935>}
 deform_glute_l {parent=deform_glute_base_l, loc=<0 0 0>, rot=<0 0 0 1>}
 deform_glute_base_r {parent=pelvis, loc=<-7.3464 -9.0213 9.1245>, rot=<0.031 -0.0172 0.8935 -0.4477>}
 deform_glute_r {parent=deform_glute_base_r, loc=<0 0 0>, rot=<0 0 0 1>}
 base_skirt_sd_r {parent=pelvis, loc=<2.3061 -1.4985 13.7669>, rot=<0 0 0 1>}
 dyn_skirt_sd_r_01 {parent=base_skirt_sd_r, loc=<0 0 0>, rot=<0.2107 0.0069 0.9761 0.0524>}
 dyn_skirt_sd_r_02 {parent=dyn_skirt_sd_r_01, loc=<8.2317 0 0>, rot=<0 0 0 1>}
 dyn_skirt_sd_r_03 {parent=dyn_skirt_sd_r_02, loc=<7.3712 0 0>, rot=<0 0 0 1>}
 base_skirt_bk_r {parent=pelvis, loc=<2.4714 -8.0615 9.6886>, rot=<0 -0.0087 0 -1>}
 dyn_skirt_bk_r_01 {parent=base_skirt_bk_r, loc=<0 0 0>, rot=<0.1758 -0.3613 0.9157 0.0112>}
 dyn_skirt_bk_r_02 {parent=dyn_skirt_bk_r_01, loc=<8.4595 0 0>, rot=<0 0 0 1>}
 dyn_skirt_bk_r_03 {parent=dyn_skirt_bk_r_02, loc=<8.1067 0 0>, rot=<0 0 0 1>}
 base_skirt_fr_r {parent=pelvis, loc=<1.9905 6.4656 9.598>, rot=<0 0 0 1>}
 dyn_skirt_fr_r_01 {parent=base_skirt_fr_r, loc=<0 0 0>, rot=<0.1811 0.4929 0.8505 0.0313>}
 dyn_skirt_fr_r_02 {parent=dyn_skirt_fr_r_01, loc=<7.7153 0 0>, rot=<0 0 0 1>}
 dyn_skirt_fr_r_03 {parent=dyn_skirt_fr_r_02, loc=<7.3703 0 0>, rot=<0 0 0 1>}
 base_skirt_fr_c {parent=pelvis, loc=<2.2424 8.3227 0>, rot=<0 0 0 1>}
 dyn_skirt_fr_c_01 {parent=base_skirt_fr_c, loc=<0 0 0>, rot=<0.0998 0.7 0.7 0.0998>}
 dyn_skirt_fr_c_02 {parent=dyn_skirt_fr_c_01, loc=<7.5455 0 0>, rot=<0 0 0 1>}
 dyn_skirt_fr_c_03 {parent=dyn_skirt_fr_c_02, loc=<8.6951 0 0>, rot=<0 0 0 1>}
 base_skirt_sd_l {parent=pelvis, loc=<2.2801 -2.564 -13.7685>, rot=<0 0 0 1>}
 dyn_skirt_sd_l_01 {parent=base_skirt_sd_l, loc=<0 0 0>, rot=<-0.002 -0.9775 0.0039 -0.2108>}
 dyn_skirt_sd_l_02 {parent=dyn_skirt_sd_l_01, loc=<8.2317 0 0>, rot=<0 0 0 1>}
 dyn_skirt_sd_l_03 {parent=dyn_skirt_sd_l_02, loc=<7.3712 0 0>, rot=<0 0 0 1>}
 base_skirt_bk_l {parent=pelvis, loc=<2.1985 -9.0127 -8.2733>, rot=<0 0 0 1>}
 dyn_skirt_bk_l_01 {parent=base_skirt_bk_l, loc=<0 0 0>, rot=<0.0213 -0.914 0.3681 -0.1695>}
 dyn_skirt_bk_l_02 {parent=dyn_skirt_bk_l_01, loc=<8.4595 0 0>, rot=<0 0 0 1>}
 dyn_skirt_bk_l_03 {parent=dyn_skirt_bk_l_02, loc=<8.1067 0 0>, rot=<0 0 0 1>}
 base_skirt_fr_l {parent=pelvis, loc=<2.1324 5.2528 -10.8557>, rot=<0 0 0 1>}
 dyn_skirt_fr_l_01 {parent=base_skirt_fr_l, loc=<0 0 0>, rot=<-0.0267 -0.9048 -0.3916 -0.1652>}
 dyn_skirt_fr_l_02 {parent=dyn_skirt_fr_l_01, loc=<7.7153 0 0>, rot=<0 0 0 1>}
 dyn_skirt_fr_l_03 {parent=dyn_skirt_fr_l_02, loc=<7.3703 0 0>, rot=<0 0 0 1>}
 base_skirt_bk_c {parent=pelvis, loc=<1.5541 -10.8793 0>, rot=<0.7071 0 0 -0.7071>}
 dyn_skirt_bk_c_01 {parent=base_skirt_bk_c, loc=<0 0 0>, rot=<0 -0.9945 0 -0.1052>}
 dyn_skirt_bk_c_02 {parent=dyn_skirt_bk_c_01, loc=<7.4289 0 0>, rot=<0 0 0 1>}
 dyn_skirt_bk_c_03 {parent=dyn_skirt_bk_c_02, loc=<8.2534 0 0>, rot=<0 0 0 1>}
 base_skirt_hanging_fr_l {parent=pelvis, loc=<4.0613 8.0889 -7.8982>, rot=<-0.0127 -0.707 0.0127 -0.707>}
 dyn_skirt_hanging_fr_l {parent=base_skirt_hanging_fr_l, loc=<0 0 0>, rot=<0.32 -0.4643 -0.4595 -0.6862>}
 base_skirt_hanging_bk_l {parent=pelvis, loc=<3.5831 -10.8189 -5.3527>, rot=<-0.0127 -0.707 0.0127 -0.707>}
 dyn_skirt_hanging_bk_l {parent=base_skirt_hanging_bk_l, loc=<0 0 0>, rot=<-0.3553 -0.4255 0.5415 -0.6321>}
 base_skirt_hanging_bk_r {parent=pelvis, loc=<3.5826 -10.8189 5.3527>, rot=<-0.0127 -0.707 0.0127 -0.707>}
 dyn_skirt_hanging_bk_r {parent=base_skirt_hanging_bk_r, loc=<0 0 0>, rot=<0.6321 0.5415 0.4255 -0.3553>}
 base_skirt_hanging_fr_r {parent=pelvis, loc=<4.0612 8.0889 7.8982>, rot=<-0.0127 -0.707 0.0127 -0.707>}
 dyn_skirt_hanging_fr_r {parent=base_skirt_hanging_fr_r, loc=<0 0 0>, rot=<0.6862 -0.4595 0.4643 0.32>}
 thigh_l_null {parent=pelvis, type=Null, loc=<-1.0855 -0.2109 -8.7053>, rot=<0.0009 0.0355 0.0155 -0.9993>}
 thigh_r_null {parent=pelvis, type=Null, loc=<-1.0855 -0.2109 8.7053>, rot=<0.0355 -0.0009 0.9993 0.0155>}
 base_megaphone {parent=pelvis, loc=<7.9797 5.084 -12.0657>, rot=<-0.0788 0.3953 -0.2193 -0.8885>}
 dyn_megaphone {parent=base_megaphone, loc=<0 0 0>, rot=<0 0 0 1>}
 base_nade {parent=pelvis, loc=<6.0831 3.2049 -12.9169>, rot=<-0.1626 0.2079 -0.0587 -0.9627>}
 dyn_nade {parent=base_nade, loc=<0 0 0>, rot=<0 0 0 1>}
 base_strap_fr_r {parent=pelvis, loc=<3.6677 9.8568 6.7648>, rot=<-0.1196 0.744 0.1485 -0.6404>}
 dyn_strap_fr_r_01 {parent=base_strap_fr_r, loc=<0 0 0>, rot=<0 0 0 1>}
 dyn_strap_fr_r_02 {parent=dyn_strap_fr_r_01, loc=<6.4226 0 0>, rot=<0.0702 -0.0531 0.338 -0.937>}
 dyn_strap_fr_r_03 {parent=dyn_strap_fr_r_02, loc=<6.4637 0 0>, rot=<0.0698 0 0 -0.9976>}
 base_strap_bk_r {parent=pelvis, loc=<4.3398 -11.8311 6.6361>, rot=<0.1153 0.7445 -0.2189 -0.6201>}
 dyn_strap_bk_r_01 {parent=base_strap_bk_r, loc=<0 0 0>, rot=<0 0 0 1>}
 dyn_strap_bk_r_02 {parent=dyn_strap_bk_r_01, loc=<7.5667 0 0>, rot=<-0.1062 -0.1134 -0.3406 -0.9273>}
 dyn_strap_bk_r_03 {parent=dyn_strap_bk_r_02, loc=<6.7848 0 0>, rot=<0 0 0 1>}
 base_belt {parent=pelvis, loc=<2.3151 -1.0609 4.8666>, rot=<0.0355 -0.0009 0.9993 0.0155>}
 dyn_belt {parent=base_belt, loc=<0 0 0>, rot=<0 0 0 1>}
 base_belt_front_c {parent=pelvis, loc=<4.7847 9.7714 0>, rot=<0 0 0.0528 -0.9986>}
 dyn_belt_front_c {parent=base_belt_front_c, loc=<0 0 0>, rot=<0 0 0 1>}
 base_belt_side_l {parent=pelvis, loc=<8.1661 -1.2298 -12.9985>, rot=<0.0171 0.1294 -0.0082 -0.9914>}
 dyn_belt_side_l {parent=base_belt_side_l, loc=<0 0 0>, rot=<0 0 0 1>}
 base_belt_back_c {parent=pelvis, loc=<5.8802 -11.6325 0>, rot=<0 0 -0.0867 -0.9962>}
 dyn_belt_back_c {parent=base_belt_back_c, loc=<0 0 0>, rot=<0 0 0 1>}
 base_belt_fr_r {parent=pelvis, loc=<3.3168 8.965 8.7261>, rot=<0.0409 -0.001 0.999 0.0155>}
 dyn_belt_fr_r {parent=base_belt_fr_r, loc=<0 0 0>, rot=<0.0012 -0.0005 -0.0176 -0.9998>}
 base_belt_fr_l {parent=pelvis, loc=<6.6644 7.7699 -8.6678>, rot=<0.0008 0.0277 0.0155 -0.9995>}
 dyn_belt_fr_l {parent=base_belt_fr_l, loc=<0 0 0>, rot=<0 0 0 -1>}
 base_belt_bk_r {parent=pelvis, loc=<3.6878 -11.1641 7.3589>, rot=<0.0355 -0.0009 0.9993 0.0155>}
 dyn_belt_bk_r {parent=base_belt_bk_r, loc=<-0.0815 0 0>, rot=<-0.003 0.0128 -0.0183 -0.9997>}
 base_belt_bk_l {parent=pelvis, loc=<7.7601 -9.7901 -7.359>, rot=<0.0009 0.0355 0.0155 -0.9993>}
 dyn_belt_bk_l {parent=base_belt_bk_l, loc=<0 0 0>, rot=<0 0 0 -1>}
 base_belt_sd_r {parent=pelvis, loc=<2.4956 0.09 15.5396>, rot=<0.0355 -0.0004 0.999 0.0289>}
 dyn_belt_sd_r {parent=base_belt_sd_r, loc=<0.0001 0 0>, rot=<0.0012 -0.0008 -0.0174 -0.9998>}
 base_belt_sd_l {parent=pelvis, loc=<8.1056 -0.1498 -13.0555>, rot=<0.0009 0.0355 0.0155 -0.9993>}
 dyn_belt_sd_l {parent=base_belt_sd_l, loc=<0 0 0>, rot=<0 0 0 -1>}
 attach {parent=root, loc=<0 0 0>, rot=<0 0 0 1>}
 ik_hand_root {parent=root, loc=<0 0 0>, rot=<0 0 0 1>}
 ik_hand_gun {parent=ik_hand_root, loc=<-48.1743 3.3545 104.4381>, rot=<0.6276 0.1309 0.4183 -0.6434>}
 ik_hand_r {parent=ik_hand_gun, loc=<0 0 0>, rot=<0 0 0 1>}
 ik_hand_l {parent=ik_hand_gun, loc=<59.3271 67.6997 34.3524>, rot=<-0.6158 -0.7027 -0.3565 0>}
 ik_foot_root {parent=root, loc=<0 0 0>, rot=<0 0 0 1>}
 ik_foot_l {parent=ik_foot_root, loc=<14.8262 -7.3545 8.296>, rot=<0.0357 0.7045 0.0284 -0.7082>}
 ik_foot_r {parent=ik_foot_root, loc=<-14.8262 -7.3545 8.296>, rot=<0.7082 0.0284 -0.7045 0.0357>}
 root_fp {parent=root, loc=<0 0 0>, rot=<0 0 0 1>}
 attach_fp {parent=root_fp, loc=<0 0 0>, rot=<0 0 0 1>}
 camera_root {parent=root_fp, loc=<0 0 0>, rot=<0 0 0 1>}
 camera {parent=camera_root, loc=<0 0 0>, rot=<0 0 0 1>}
 attach_emote_01 {parent=root, loc=<0 0 0>, rot=<0 0 0 1>}
 attach_emote_02 {parent=root, loc=<0 0 0>, rot=<0 0 0 1>}
 attach_emote_03 {parent=root, loc=<0 0 0>, rot=<0 0 0 1>}
 warp_point_01 {parent=root, loc=<0 0 0>, rot=<0 0 0 1>}
 warp_point_02 {parent=root, loc=<0 0 0>, rot=<0 0 0 1>}
 warp_point_03 {parent=root, loc=<0 0 0>, rot=<0 0 0 1>}
 thigh_l_ctrl {controlType=EULER_TRANSFORM, type=Control, loc=<0 0 0>, rot=<0 0 0 1>}
 calf_l_ctrl {controlType=EULER_TRANSFORM, parent=thigh_l_ctrl, type=Control, loc=<0 0 0>, rot=<0 0 0 1>}
 thigh_r_ctrl {controlType=EULER_TRANSFORM, type=Control, loc=<0 0 0>, rot=<0 0 0 1>}
 calf_r_ctrl {controlType=EULER_TRANSFORM, parent=thigh_r_ctrl, type=Control, loc=<0 0 0>, rot=<0 0 0 1>}
 base_tie_02_ctrl {controlType=EULER_TRANSFORM, type=Control, loc=<0 0 0>, rot=<0 0 0 1>}
 upperarm_l_ctrl {controlType=EULER_TRANSFORM, type=Control, loc=<0 0 0>, rot=<0 0 0 1>}
 upperarm_r_ctrl {controlType=EULER_TRANSFORM, type=Control, loc=<0 0 0>, rot=<0 0 0 1>}
 spine_01_ctrl {controlType=EULER_TRANSFORM, type=Control, loc=<0 0 0>, rot=<0 0 0 1>}
]
variables [
 Crouching {type=double, default=0}
 CrouchingMoving2D {type=double, default=0}
 CrouchingTargeting {type=bool, default=False}
 Crouching_Bool {type=bool, default=False}
 EmoteGoldenGlory {type=double, default=0}
 FrontOrBack {type=bool, default=False}
 Melee_Bool {type=bool, default=False}
 PlayingBalleteEmote {type=bool, default=False}
 PlayingEmote {type=double, default=0}
 Sides {type=bool, default=False}
]
