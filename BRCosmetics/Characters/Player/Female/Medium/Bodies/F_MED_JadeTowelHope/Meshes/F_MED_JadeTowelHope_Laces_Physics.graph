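# Physics asset for the hoodie drawstrings (laces) on the JadeTowelHope female medium body cosmetic.
# Small capsules form two string chains with tight swing and twist limits between links: four dynamic links on the left, three on the right.
# Both chain bases attach to spine_05 with free angular motion.
# Larger capsules on spine_04, spine_05 and both pec bones form the torso volume, also joined by free angular constraints.
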PhysicsAsset /BRCosmetics/Characters/Player/Female/Medium/Bodies/F_MED_JadeTowelHope/Meshes/F_MED_JadeTowelHope_Laces_Physics
bodies {
 base_HoodieStrings_01_l [{shapeType=Capsule, length=0.505, radius=0.505, shapeName=base_HoodieStrings_01_l_capsule, rotation=<rot 0 -90 0>}]
 base_HoodieStrings_01_r [{shapeType=Capsule, length=0.505, radius=0.505, shapeName=base_HoodieStrings_01_r_capsule, rotation=<rot 0 -90 0>}]
 base_pec_l [{shapeType=Capsule, length=3, radius=5, shapeName=base_pec_l_capsule, center=<0.257 -4.7943 -1.4348>, rotation=<rot 90 -1.4321 11.6846>}]
 base_pec_r [{shapeType=Capsule, length=3, radius=5, shapeName=base_pec_l_capsule, center=<0.257 -4.7943 -1.4348>, rotation=<rot 90 -1.4321 -11.6846>}]
 dyn_HoodieStrings_01_l [{shapeType=Capsule, length=2.5, radius=0.5, shapeName=dyn_HoodieStrings_01_l_capsule, center=<1.1193 0 0>, rotation=<rot 90 0 0>}]
 dyn_HoodieStrings_01_r [{shapeType=Capsule, length=2.5, radius=0.5, shapeName=dyn_HoodieStrings_01_r_capsule, center=<1.1193 0 0>, rotation=<rot 90 0 0>}]
 dyn_HoodieStrings_02_l [{shapeType=Capsule, length=2.5, radius=0.5, shapeName=dyn_HoodieStrings_02_l_capsule, center=<1.3913 0 0>, rotation=<rot 90 0 0>}]
 dyn_HoodieStrings_02_r [{shapeType=Capsule, length=2.5, radius=0.5, shapeName=dyn_HoodieStrings_02_r_capsule, center=<1.3913 0 0>, rotation=<rot 90 0 0>}]
 dyn_HoodieStrings_03_l [{shapeType=Capsule, length=2.5, radius=0.5, shapeName=dyn_HoodieStrings_03_l_capsule, center=<0.4887 0 0>, rotation=<rot 90 0 0>}]
 dyn_HoodieStrings_03_r [{shapeType=Capsule, length=2.5, radius=0.5, shapeName=dyn_HoodieStrings_03_r_capsule, center=<0.4887 0 0>, rotation=<rot 90 0 0>}]
 dyn_HoodieStrings_04_l [{
   shapeType=Capsule
   length=3.2362
   radius=0.5
   shapeName=dyn_HoodieStrings_04_l_capsule
   center=<0.8115 -0.0392 0>
   rotation=<rot 90 180 173.6401>
  }]
 spine_04 [{shapeType=Capsule, length=11.8455, radius=10.2395, shapeName=spine_04_capsule, center=<0.269 1.4421 0>, rotation=<rot 0 -12.5772 0>}]
 spine_05 [{shapeType=Capsule, length=12.2337, radius=9.1515, shapeName=spine_05_capsule, center=<-2.8034 0.8406 0>, rotation=<rot 0 -24.7705 0>}]
}
constraints [
 {bone1Name=dyn_HoodieStrings_01_l, bone2Name=base_HoodieStrings_01_l, swing1LimitDegrees=15, swing2LimitDegrees=10, twistLimitDegrees=1}
 {bone1Name=dyn_HoodieStrings_02_l, bone2Name=dyn_HoodieStrings_01_l, swing1LimitDegrees=15, swing2LimitDegrees=10, twistLimitDegrees=1}
 {bone1Name=dyn_HoodieStrings_03_l, bone2Name=dyn_HoodieStrings_02_l, swing1LimitDegrees=15, swing2LimitDegrees=10, twistLimitDegrees=1}
 {bone1Name=dyn_HoodieStrings_04_l, bone2Name=dyn_HoodieStrings_03_l, swing1LimitDegrees=15, swing2LimitDegrees=10, twistLimitDegrees=1}
 {bone1Name=dyn_HoodieStrings_01_r, bone2Name=base_HoodieStrings_01_r, swing1LimitDegrees=15, swing2LimitDegrees=10, twistLimitDegrees=1}
 {bone1Name=dyn_HoodieStrings_02_r, bone2Name=dyn_HoodieStrings_01_r, swing1LimitDegrees=15, swing2LimitDegrees=10, twistLimitDegrees=1}
 {bone1Name=dyn_HoodieStrings_03_r, bone2Name=dyn_HoodieStrings_02_r, swing1LimitDegrees=15, swing2LimitDegrees=10, twistLimitDegrees=1}
 {
  bone1Name=base_HoodieStrings_01_l
  bone2Name=spine_05
  swing1LimitDegrees=45
  swing1Motion=Free
  swing2LimitDegrees=45
  swing2Motion=Free
  twistLimitDegrees=45
  twistMotion=Free
 }
 {
  bone1Name=base_HoodieStrings_01_r
  bone2Name=spine_05
  swing1LimitDegrees=45
  swing1Motion=Free
  swing2LimitDegrees=45
  swing2Motion=Free
  twistLimitDegrees=45
  twistMotion=Free
 }
 {
  bone1Name=spine_05
  bone2Name=spine_04
  swing1LimitDegrees=45
  swing1Motion=Free
  swing2LimitDegrees=45
  swing2Motion=Free
  twistLimitDegrees=45
  twistMotion=Free
 }
 {
  bone1Name=base_pec_l
  bone2Name=spine_04
  swing1LimitDegrees=45
  swing1Motion=Free
  swing2LimitDegrees=45
  swing2Motion=Free
  twistLimitDegrees=45
  twistMotion=Free
 }
 {
  bone1Name=base_pec_r
  bone2Name=spine_04
  swing1LimitDegrees=45
  swing1Motion=Free
  swing2LimitDegrees=45
  swing2Motion=Free
  twistLimitDegrees=45
  twistMotion=Free
 }
]
