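# Physics asset for the female medium JadeTowelHope cosmetic body, covering its dynamic front skirt and pouch bones plus the torso and upper legs.
# Capsule bodies sit on five spine bones, both thighs, two chained skirt bones and the pouch.
# Spine and skirt links leave swing and twist free, while the pouch is angularly locked to the left thigh.
# No constraint joins the skirt chain or either thigh to the spine.
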PhysicsAsset /BRCosmetics/Characters/Player/Female/Medium/Bodies/F_MED_JadeTowelHope/Meshes/F_MED_JadeTowelHope_Physics
bodies {
 dyn_Pouch [{shapeType=Capsule, length=6.5437, radius=3.6643, shapeName=dyn_Pouch_capsule, center=<0.042 0 -1.1021>, rotation=<rot 87.7206 -0.0002 0>}]
 dyn_skirt_fr_c_01 [{shapeType=Capsule, length=29, radius=5, shapeName=dyn_skirt_fr_c_01_capsule, center=<6.33 0 -2.6631>, rotation=<rot 0 0 90>}]
 dyn_skirt_fr_c_02 [{shapeType=Capsule, length=28.5, radius=5.5, shapeName=dyn_skirt_fr_c_02_capsule, center=<9.3394 0 -3.4796>, rotation=<rot 0 0 90>}]
 spine_01 [{shapeType=Capsule, length=10, radius=10, shapeName=spine_01_capsule, center=<-5.5304 0 0>}]
 spine_02 [{shapeType=Capsule, length=10, radius=11.5, shapeName=spine_02_capsule}]
 spine_03 [{shapeType=Capsule, length=10, radius=11.5, shapeName=spine_03_capsule}]
 spine_04 [{shapeType=Capsule, length=10, radius=10, shapeName=spine_04_capsule, center=<0 1.9119 0>}]
 spine_05 [{shapeType=Capsule, length=10, radius=10, shapeName=spine_05_capsule, center=<0 1.2856 0>}]
 thigh_l [{shapeType=Capsule, length=24.5, radius=15, shapeName=thigh_l_capsule, center=<-26.4983 -0.3757 1.3313>, rotation=<rot 90 0 0>}]
 thigh_r [{shapeType=Capsule, length=24.5, radius=15, shapeName=thigh_l_capsule, center=<26.4983 0.3757 -1.3313>, rotation=<rot 90 0 180>}]
}
constraints [
 {
  bone1Name=spine_02
  bone2Name=spine_01
  swing1LimitDegrees=45
  swing1Motion=Free
  swing2LimitDegrees=45
  swing2Motion=Free
  twistLimitDegrees=45
  twistMotion=Free
 }
 {
  bone1Name=spine_03
  bone2Name=spine_02
  swing1LimitDegrees=45
  swing1Motion=Free
  swing2LimitDegrees=45
  swing2Motion=Free
  twistLimitDegrees=45
  twistMotion=Free
 }
 {
  bone1Name=spine_04
  bone2Name=spine_03
  swing1LimitDegrees=45
  swing1Motion=Free
  swing2LimitDegrees=45
  swing2Motion=Free
  twistLimitDegrees=45
  twistMotion=Free
 }
 {
  bone1Name=spine_05
  bone2Name=spine_04
  swing1LimitDegrees=45
  swing1Motion=Free
  swing2LimitDegrees=45
  swing2Motion=Free
  twistLimitDegrees=45
  twistMotion=Free
 }
 {
  bone1Name=dyn_skirt_fr_c_02
  bone2Name=dyn_skirt_fr_c_01
  swing1LimitDegrees=45
  swing1Motion=Free
  swing2LimitDegrees=45
  swing2Motion=Free
  twistLimitDegrees=45
  twistMotion=Free
 }
 {
  bone1Name=dyn_Pouch
  bone2Name=thigh_l
  swing1LimitDegrees=45
  swing1Motion=Locked
  swing2LimitDegrees=45
  swing2Motion=Locked
  twistLimitDegrees=45
  twistMotion=Locked
 }
]
