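# Physics asset for the crouching variant of the F_MED_JadeTowelHope female medium player body, giving a simplified collision proxy of the lower body and torso.
# Capsule bodies cover the pelvis, the five spine bones and both thighs, with two capsules on the pelvis and none on arms, head or lower legs.
# Constraints chain the spine down to the pelvis and attach each thigh to it, with all swing and twist motion left free.
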PhysicsAsset /BRCosmetics/Characters/Player/Female/Medium/Bodies/F_MED_JadeTowelHope/Meshes/F_MED_JadeTowelHope_Physics_Crouching
bodies {
 pelvis [
  {shapeType=Capsule, length=17.1695, radius=23.8286, shapeName=pelvis_capsule, center=<-7.2166 5.6598 0>}
  {shapeType=Capsule, length=22, radius=6, shapeName=pelvis_capsule, center=<-8.163 31.0209 0>}
 ]
 spine_01 [{shapeType=Capsule, length=10, radius=10, shapeName=spine_01_capsule, center=<-5.5304 0 0>}]
 spine_02 [{shapeType=Capsule, length=10, radius=11.5, shapeName=spine_02_capsule}]
 spine_03 [{shapeType=Capsule, length=10, radius=11.5, shapeName=spine_03_capsule}]
 spine_04 [{shapeType=Capsule, length=10, radius=10, shapeName=spine_04_capsule, center=<0 1.9119 0>}]
 spine_05 [{shapeType=Capsule, length=10, radius=10, shapeName=spine_05_capsule, center=<0 1.2856 0>}]
 thigh_l [{shapeType=Capsule, length=24.5, radius=11.5, shapeName=thigh_l_capsule, center=<-32.5183 -0.3757 1.3313>, rotation=<rot 90 0 0>}]
 thigh_r [{shapeType=Capsule, length=24.5, radius=11.5, shapeName=thigh_l_capsule, center=<32.5183 0.3757 -1.3313>, rotation=<rot 90 0 180>}]
}
constraints [
 {
  bone1Name=spine_02
  bone2Name=spine_01
  swing1LimitDegrees=45
  swing1Motion=Free
  swing2LimitDegrees=45
  swing2Motion=Free
  twistLimitDegrees=45
  twistMotion=Free
 }
 {
  bone1Name=spine_03
  bone2Name=spine_02
  swing1LimitDegrees=45
  swing1Motion=Free
  swing2LimitDegrees=45
  swing2Motion=Free
  twistLimitDegrees=45
  twistMotion=Free
 }
 {
  bone1Name=spine_04
  bone2Name=spine_03
  swing1LimitDegrees=45
  swing1Motion=Free
  swing2LimitDegrees=45
  swing2Motion=Free
  twistLimitDegrees=45
  twistMotion=Free
 }
 {
  bone1Name=spine_05
  bone2Name=spine_04
  swing1LimitDegrees=45
  swing1Motion=Free
  swing2LimitDegrees=45
  swing2Motion=Free
  twistLimitDegrees=45
  twistMotion=Free
 }
 {
  bone1Name=spine_01
  bone2Name=pelvis
  swing1LimitDegrees=45
  swing1Motion=Free
  swing2LimitDegrees=45
  swing2Motion=Free
  twistLimitDegrees=45
  twistMotion=Free
 }
 {
  bone1Name=thigh_l
  bone2Name=pelvis
  swing1LimitDegrees=45
  swing1Motion=Free
  swing2LimitDegrees=45
  swing2Motion=Free
  twistLimitDegrees=45
  twistMotion=Free
 }
 {
  bone1Name=thigh_r
  bone2Name=pelvis
  swing1LimitDegrees=45
  swing1Motion=Free
  swing2LimitDegrees=45
  swing2Motion=Free
  twistLimitDegrees=45
  twistMotion=Free
 }
]
