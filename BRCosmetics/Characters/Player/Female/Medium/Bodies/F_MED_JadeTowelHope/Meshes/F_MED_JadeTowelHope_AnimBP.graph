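# Post-copy Animation Blueprint for the JadeTowelHope female medium body cosmetic, adding secondary motion to the parent character's pose.
# The copied pose feeds a state-driven Control Rig and pec bone constraints, then is cached.
# Anim Dynamics chains simulate sleeves and skirt from a RigidBody cache, with sleeve variants switched by skydiving and crouch-targeting.
# Final additive bone offsets on the Dynamics cache adjust straps, downed-state skirt and backpack attachment.
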AnimBlueprint /BRCosmetics/Characters/Player/Female/Medium/Bodies/F_MED_JadeTowelHope/Meshes/F_MED_JadeTowelHope_AnimBP
target_skeleton=&/F_MED_JadeTowelHope_Skeleton
graphs [
 AnimGraph {nodes [
   n1 = "Transform (Modify) Bone - Bone: dyn_strap_bk_r_01"(
    ComponentPose=Local To Component(LocalPose=Use cached pose 'Dynamics Cache'())
    Alpha=1
    Translation="0,0,0"
    Rotation="23.857,-0.006,-0.000"
    Scale="1,1,1"
   ) {
    rotationMode=BMM_Additive
    rotationSpace=BCS_BoneSpace
    translationMode=BMM_Additive
    translationSpace=BCS_BoneSpace
    alphaBoolBlend {blendInTime=0.2, blendOption=Linear, blendOutTime=0.2}
    alphaScaleBiasClamp {bInterpResult=true, interpSpeedDecreasing=5, interpSpeedIncreasing=5}
    boneToModify {boneName=dyn_strap_bk_r_01}
    rotation=<rot 23.857 -0.006 0>
   }
   n2 = "Transform (Modify) Bone - Bone: dyn_strap_fr_r_02"(ComponentPose=@n1, Alpha=1, Translation="0,0,0", Rotation="-21.168,14.808,-3.925", Scale="1,1,1") {
    rotationMode=BMM_Additive
    rotationSpace=BCS_BoneSpace
    translationMode=BMM_Additive
    translationSpace=BCS_BoneSpace
    alphaBoolBlend {blendInTime=0.2, blendOption=Linear, blendOutTime=0.2}
    alphaScaleBiasClamp {bInterpResult=true, interpSpeedDecreasing=5, interpSpeedIncreasing=5}
    boneToModify {boneName=dyn_strap_fr_r_02}
    rotation=<rot -21.168 14.808 -3.925>
   }
   n3 = Get bIsDBNO()
   n4 = "Transform (Modify) Bone - Bone: dyn_skirt_bk_r_02"!(ComponentPose=@n2, bAlphaBoolEnabled:bool=@n3, Translation="0,0,0", Rotation="-13.849,-24.141,-0.580", Scale="1,1,1") -> (Pose) {
    alphaInputType=Bool
    rotationMode=BMM_Additive
    rotationSpace=BCS_BoneSpace
    alphaBoolBlend {blendOption=Linear}
    boneToModify {boneName=dyn_skirt_bk_r_02}
    rotation=<rot -13.849 -24.141 -0.58>
   }
   n5 = "Transform (Modify) Bone - Bone: dyn_skirt_bk_c_02"!(ComponentPose=@n4, bAlphaBoolEnabled:bool=@n3, Translation="0,0,0", Rotation="-18.594,-0.099,0.081", Scale="1,1,1") -> (Pose) {
    alphaInputType=Bool
    rotationMode=BMM_Additive
    rotationSpace=BCS_BoneSpace
    alphaBoolBlend {blendOption=Linear}
    boneToModify {boneName=dyn_skirt_bk_c_02}
    rotation=<rot -18.594 -0.099 0.081>
   }
   n6 = "Transform (Modify) Bone - Bone: dyn_skirt_bk_l_02"!(ComponentPose=@n5, bAlphaBoolEnabled:bool=@n3, Translation="0,0,0", Rotation="-25.057,-0.109,0.005", Scale="1,1,1") -> (Pose) {
    alphaInputType=Bool
    rotationMode=BMM_Additive
    rotationSpace=BCS_BoneSpace
    alphaBoolBlend {blendOption=Linear}
    boneToModify {boneName=dyn_skirt_bk_l_02}
    rotation=<rot -25.057 -0.109 0.005>
   }
   n7 = "Transform (Modify) Bone - Bone: attach_backpack"(
    ComponentPose=@n6
    Alpha=Get Alpha_CollisionHairAndArm L() -> (Alpha_CollisionHairAndArm L:double)
    Translation="0,0,0"
    Rotation="20,0,0"
    Scale="1,1,1"
   ) {
    rotationMode=BMM_Additive
    rotationSpace=BCS_BoneSpace
    alphaBoolBlend {blendOption=Linear}
    alphaScaleBiasClamp {bInterpResult=true, interpSpeedDecreasing=5, interpSpeedIncreasing=8}
    boneToModify {boneName=attach_backpack}
    rotation {pitch=20}
   }
   Output Pose(Result=Component To Local(ComponentPose=@n7)) {layerGroup=None}
   n8 = Control Rig!(
    Source=Copy Pose From Mesh() {bUseAttachedParent=true}
    Alpha
    PlayingEmote:double=Get Alpha_PlayingEmote() -> (Alpha_PlayingEmote:double)
    CrouchingMoving2D:double=Get Alpha_CrouchingMoving2D() -> (Alpha_CrouchingMoving2D:double)
    Crouching:double=Get Alpha_Crouching() -> (Alpha_Crouching:double)
    Melee_Bool:bool=Get bIsPlayingMeleeAnim()
    Crouching_Bool:bool=Get bIsCrouching()
    PlayingBalleteEmote:bool=Get isCustomEmotePlaying() -> (isCustomEmotePlaying:bool)
    Sides:bool=Get Sides() -> (Sides:bool)
    FrontOrBack:bool=Get FrontOrBack() -> (FrontOrBack:bool)
    CrouchingTargeting:bool=Get CrouchingTargeting() -> (CrouchingTargeting:bool)
    EmoteGoldenGlory:double=Get EmoteGoldenGlory() -> (EmoteGoldenGlory:double)
   ) -> (Pose) {
    alphaBoolBlend {blendOption=Linear}
    controlRigAssetReference {blueprintRigClass=&/F_MED_JadeTowelHope_CtrlRig.F_MED_JadeTowelHope_CtrlRig_C}
   }
   n9 = "Constraint - base_pec_l (PPPP)"!(
    ComponentPose=Local To Component(LocalPose=@n8)
    Alpha=1
    ConstraintWeights_0=0.6678
    ConstraintWeights_1:float=0.2074
    ConstraintWeights_2:float=0.1238
    ConstraintWeights_3:float=0.001
   ) -> (Pose) {
    alphaBoolBlend {blendOption=Linear}
    boneToModify {boneName=base_pec_l}
    constraintSetup [{targetBone {boneName=spine_04}}, {targetBone {boneName=spine_05}}, {targetBone {boneName=deform_pec_l}}, {targetBone {boneName=clavicle_l}}]
    constraintWeights [1,1,1,1]
   }
   n10 = "Constraint - base_pec_r (PPPP)"!(
    ComponentPose=@n9
    Alpha=1
    ConstraintWeights_0=0.6678
    ConstraintWeights_1:float=0.2074
    ConstraintWeights_2:float=0.1238
    ConstraintWeights_3:float=0.001
   ) -> (Pose) {
    alphaBoolBlend {blendOption=Linear}
    boneToModify {boneName=base_pec_r}
    constraintSetup [{targetBone {boneName=spine_04}}, {targetBone {boneName=spine_05}}, {targetBone {boneName=deform_pec_r}}, {targetBone {boneName=clavicle_r}}]
    constraintWeights [1,1,1,1]
   }
   Save cached pose 'CtrlRig Cache'(Pose=Component To Local(ComponentPose=@n10)) {cacheName=CtrlRig Cache}
   n11 = Anim Dynamics - Bone: MSTR_sleeve_l(ComponentPose=Local To Component(LocalPose=Use cached pose 'RigidBodyCache'()), Alpha=1, ExternalForce="0,0,0") {
    angularDampingOverride=1
    angularSpringConstant=20000
    bAngularSpring=true
    bOverrideAngularDamping=true
    bUseGravityOverride=true
    gravityScale=0.35
    simulationSpace=RootRelative
    alphaBoolBlend {blendOption=Linear}
    alphaScaleBiasClamp {bInterpResult=true, interpSpeedIncreasing=0.3}
    boundBone {boneName=MSTR_sleeve_l}
    componentAppliedLinearAccClamp=<10000 10000 10000>
    componentLinearAccScale=<1 1 1>
    componentLinearVelScale=<1 1 1>
    gravityOverride {z=-980}
    physicsBodyDefinitions [{
      boundBone {boneName=MSTR_sleeve_l}
      boxExtents=<20 30 30>
      constraintSetup {angularLimitsMax=<2 0 2>, angularLimitsMin=<-2 0 -2>, angularTarget=<1 0 0>}
      localJointOffset=<3.5 0 0>
     }]
   }
   n12 = Anim Dynamics - Bone: MSTR_sleeve_r(ComponentPose=@n11, Alpha=1, ExternalForce="0,0,0") {
    angularDampingOverride=1
    angularSpringConstant=20000
    bAngularSpring=true
    bOverrideAngularDamping=true
    bUseGravityOverride=true
    gravityScale=0.35
    simulationSpace=RootRelative
    alphaBoolBlend {blendOption=Linear}
    alphaScaleBiasClamp {bInterpResult=true, interpSpeedIncreasing=0.3}
    boundBone {boneName=MSTR_sleeve_r}
    componentAppliedLinearAccClamp=<10000 10000 10000>
    componentLinearAccScale=<1 1 1>
    componentLinearVelScale=<1 1 1>
    gravityOverride {z=-980}
    physicsBodyDefinitions [{
      boundBone {boneName=MSTR_sleeve_r}
      boxExtents=<20 30 30>
      constraintSetup {angularLimitsMax=<2 0 2>, angularLimitsMin=<-2 0 -2>, angularTarget=<1 0 0>}
      localJointOffset=<3.5 0 0>
     }]
   }
   n13 = Anim Dynamics - Bone: dyn_sleeves_fr_l(ComponentPose=@n12, Alpha=1, ExternalForce="0,0,0") {
    angularDampingOverride=1
    angularSpringConstant=20000
    bAngularSpring=true
    bOverrideAngularDamping=true
    bUseGravityOverride=true
    gravityScale=0.35
    simulationSpace=RootRelative
    alphaBoolBlend {blendOption=Linear}
    boundBone {boneName=dyn_sleeves_fr_l}
    componentAppliedLinearAccClamp=<10000 10000 10000>
    componentLinearAccScale=<1 1 1>
    componentLinearVelScale=<1 1 1>
    gravityOverride {z=-980}
    physicsBodyDefinitions [{
      boundBone {boneName=dyn_sleeves_fr_l}
      boxExtents=<20 30 30>
      constraintSetup {angularLimitsMax=<5 5 5>, angularLimitsMin=<-5 -5 0>, angularTarget=<1 0 0>}
      localJointOffset=<2 0 0>
     }]
   }
   n14 = Anim Dynamics - Bone: dyn_sleeves_bk_l(ComponentPose=@n13, Alpha=1, ExternalForce="0,0,0") {
    angularDampingOverride=1
    angularSpringConstant=20000
    bAngularSpring=true
    bOverrideAngularDamping=true
    bUseGravityOverride=true
    gravityScale=0.35
    simulationSpace=RootRelative
    alphaBoolBlend {blendOption=Linear}
    boundBone {boneName=dyn_sleeves_bk_l}
    componentAppliedLinearAccClamp=<10000 10000 10000>
    componentLinearAccScale=<1 1 1>
    componentLinearVelScale=<1 1 1>
    gravityOverride {z=-980}
    physicsBodyDefinitions [{
      boundBone {boneName=dyn_sleeves_bk_l}
      boxExtents=<20 30 30>
      constraintSetup {angularLimitsMax=<5 5 0>, angularLimitsMin=<-5 -5 -5>, angularTarget=<1 0 0>}
      localJointOffset=<2 0 0>
     }]
   }
   n15 = Anim Dynamics - Bone: dyn_sleeves_bk_r(ComponentPose=@n14, Alpha=1, ExternalForce="0,0,0") {
    angularDampingOverride=1
    angularSpringConstant=20000
    bAngularSpring=true
    bOverrideAngularDamping=true
    bUseGravityOverride=true
    gravityScale=0.35
    simulationSpace=RootRelative
    alphaBoolBlend {blendOption=Linear}
    boundBone {boneName=dyn_sleeves_bk_r}
    componentAppliedLinearAccClamp=<10000 10000 10000>
    componentLinearAccScale=<1 1 1>
    componentLinearVelScale=<1 1 1>
    gravityOverride {z=-980}
    physicsBodyDefinitions [{
      boundBone {boneName=dyn_sleeves_bk_r}
      boxExtents=<20 30 30>
      constraintSetup {angularLimitsMax=<5 5 0>, angularLimitsMin=<-5 -5 -5>, angularTarget=<1 0 0>}
      localJointOffset=<2 0 0>
     }]
   }
   n16 = Anim Dynamics - Bone: dyn_sleeves_fr_r(ComponentPose=@n15, Alpha=1, ExternalForce="0,0,0") {
    angularDampingOverride=1
    angularSpringConstant=20000
    bAngularSpring=true
    bOverrideAngularDamping=true
    bUseGravityOverride=true
    gravityScale=0.35
    simulationSpace=RootRelative
    alphaBoolBlend {blendOption=Linear}
    boundBone {boneName=dyn_sleeves_fr_r}
    componentAppliedLinearAccClamp=<10000 10000 10000>
    componentLinearAccScale=<1 1 1>
    componentLinearVelScale=<1 1 1>
    gravityOverride {z=-980}
    physicsBodyDefinitions [{
      boundBone {boneName=dyn_sleeves_fr_r}
      boxExtents=<20 30 30>
      constraintSetup {angularLimitsMax=<5 5 5>, angularLimitsMin=<-5 -5 0>, angularTarget=<1 0 0>}
      localJointOffset=<2 0 0>
     }]
   }
   n17 = Anim Dynamics - Bone: MSTR_sleeve_l(ComponentPose=Local To Component(LocalPose=Use cached pose 'RigidBodyCache'()), Alpha=1, ExternalForce="0,0,0") {
    angularDampingOverride=1
    angularSpringConstant=20000
    bAngularSpring=true
    bOverrideAngularDamping=true
    bUseGravityOverride=true
    gravityScale=0.35
    simulationSpace=RootRelative
    alphaBoolBlend {blendOption=Linear}
    alphaScaleBiasClamp {bInterpResult=true, interpSpeedIncreasing=0.3}
    boundBone {boneName=MSTR_sleeve_l}
    componentAppliedLinearAccClamp=<10000 10000 10000>
    componentLinearAccScale=<1 1 1>
    componentLinearVelScale=<1 1 1>
    gravityOverride {z=-980}
    physicsBodyDefinitions [{
      boundBone {boneName=MSTR_sleeve_l}
      boxExtents=<20 30 30>
      constraintSetup {angularLimitsMax=<5 0 5>, angularLimitsMin=<-5 0 -5>, angularTarget=<1 0 0>}
      localJointOffset=<3.5 0 0>
     }]
   }
   n18 = Anim Dynamics - Bone: MSTR_sleeve_r(ComponentPose=@n17, Alpha=1, ExternalForce="0,0,0") {
    angularDampingOverride=1
    angularSpringConstant=20000
    bAngularSpring=true
    bOverrideAngularDamping=true
    bUseGravityOverride=true
    gravityScale=0.35
    simulationSpace=RootRelative
    alphaBoolBlend {blendOption=Linear}
    alphaScaleBiasClamp {bInterpResult=true, interpSpeedIncreasing=0.3}
    boundBone {boneName=MSTR_sleeve_r}
    componentAppliedLinearAccClamp=<10000 10000 10000>
    componentLinearAccScale=<1 1 1>
    componentLinearVelScale=<1 1 1>
    gravityOverride {z=-980}
    physicsBodyDefinitions [{
      boundBone {boneName=MSTR_sleeve_r}
      boxExtents=<20 30 30>
      constraintSetup {angularLimitsMax=<5 0 5>, angularLimitsMin=<-5 0 -5>, angularTarget=<1 0 0>}
      localJointOffset=<3.5 0 0>
     }]
   }
   n19 = Anim Dynamics - Bone: dyn_sleeves_fr_l(ComponentPose=@n18, Alpha=1, ExternalForce="0,0,0") {
    angularDampingOverride=1
    angularSpringConstant=20000
    bAngularSpring=true
    bOverrideAngularDamping=true
    bUseGravityOverride=true
    gravityScale=0.35
    simulationSpace=RootRelative
    alphaBoolBlend {blendOption=Linear}
    boundBone {boneName=dyn_sleeves_fr_l}
    componentAppliedLinearAccClamp=<10000 10000 10000>
    componentLinearAccScale=<1 1 1>
    componentLinearVelScale=<1 1 1>
    gravityOverride {z=-980}
    physicsBodyDefinitions [{
      boundBone {boneName=dyn_sleeves_fr_l}
      boxExtents=<20 30 30>
      constraintSetup {angularLimitsMax=<5 5 5>, angularLimitsMin=<-5 -5 0>, angularTarget=<1 0 0>}
      localJointOffset=<2 0 0>
     }]
   }
   n20 = Anim Dynamics - Bone: dyn_sleeves_bk_l(ComponentPose=@n19, Alpha=1, ExternalForce="0,0,0") {
    angularDampingOverride=1
    angularSpringConstant=20000
    bAngularSpring=true
    bOverrideAngularDamping=true
    bUseGravityOverride=true
    gravityScale=0.35
    simulationSpace=RootRelative
    alphaBoolBlend {blendOption=Linear}
    boundBone {boneName=dyn_sleeves_bk_l}
    componentAppliedLinearAccClamp=<10000 10000 10000>
    componentLinearAccScale=<1 1 1>
    componentLinearVelScale=<1 1 1>
    gravityOverride {z=-980}
    physicsBodyDefinitions [{
      boundBone {boneName=dyn_sleeves_bk_l}
      boxExtents=<20 30 30>
      constraintSetup {angularLimitsMax=<7 7 0>, angularLimitsMin=<-7 -7 -7>, angularTarget=<1 0 0>}
      localJointOffset=<2 0 0>
     }]
   }
   n21 = Anim Dynamics - Bone: dyn_sleeves_bk_r(ComponentPose=@n20, Alpha=1, ExternalForce="0,0,0") {
    angularDampingOverride=1
    angularSpringConstant=20000
    bAngularSpring=true
    bOverrideAngularDamping=true
    bUseGravityOverride=true
    gravityScale=0.35
    simulationSpace=RootRelative
    alphaBoolBlend {blendOption=Linear}
    boundBone {boneName=dyn_sleeves_bk_r}
    componentAppliedLinearAccClamp=<10000 10000 10000>
    componentLinearAccScale=<1 1 1>
    componentLinearVelScale=<1 1 1>
    gravityOverride {z=-980}
    physicsBodyDefinitions [{
      boundBone {boneName=dyn_sleeves_bk_r}
      boxExtents=<20 30 30>
      constraintSetup {angularLimitsMax=<7 7 0>, angularLimitsMin=<-7 -7 -7>, angularTarget=<1 0 0>}
      localJointOffset=<2 0 0>
     }]
   }
   n22 = Anim Dynamics - Bone: dyn_sleeves_fr_r(ComponentPose=@n21, Alpha=1, ExternalForce="0,0,0") {
    angularDampingOverride=1
    angularSpringConstant=20000
    bAngularSpring=true
    bOverrideAngularDamping=true
    bUseGravityOverride=true
    gravityScale=0.35
    simulationSpace=RootRelative
    alphaBoolBlend {blendOption=Linear}
    boundBone {boneName=dyn_sleeves_fr_r}
    componentAppliedLinearAccClamp=<10000 10000 10000>
    componentLinearAccScale=<1 1 1>
    componentLinearVelScale=<1 1 1>
    gravityOverride {z=-980}
    physicsBodyDefinitions [{
      boundBone {boneName=dyn_sleeves_fr_r}
      boxExtents=<20 30 30>
      constraintSetup {angularLimitsMax=<5 5 5>, angularLimitsMin=<-5 -5 0>, angularTarget=<1 0 0>}
      localJointOffset=<2 0 0>
     }]
   }
   n23 = Blend Poses by bool(
    BlendPose_0=Component To Local(ComponentPose=@n22)
    BlendPose_1=Component To Local(ComponentPose=@n16)
    BlendTime_0=0.2
    BlendTime_1=0.4
    bActiveValue=Get bIsSkyDiving()
   ) {blendType=Linear}
   n24 = Anim Dynamics - Bone: MSTR_sleeve_l(ComponentPose=Local To Component(LocalPose=Use cached pose 'RigidBodyCache'()), Alpha=1, ExternalForce="0,0,0") {
    angularDampingOverride=1
    angularSpringConstant=20000
    bAngularSpring=true
    bOverrideAngularDamping=true
    bUseGravityOverride=true
    gravityScale=0.35
    simulationSpace=RootRelative
    alphaBoolBlend {blendOption=Linear}
    alphaScaleBiasClamp {bInterpResult=true, interpSpeedIncreasing=0.3}
    boundBone {boneName=MSTR_sleeve_l}
    componentAppliedLinearAccClamp=<10000 10000 10000>
    componentLinearAccScale=<1 1 1>
    componentLinearVelScale=<1 1 1>
    gravityOverride {z=-980}
    physicsBodyDefinitions [{
      boundBone {boneName=MSTR_sleeve_l}
      boxExtents=<20 30 30>
      constraintSetup {angularLimitsMax=<2 0 2>, angularLimitsMin=<-2 0 -2>, angularTarget=<1 0 0>}
      localJointOffset=<3.5 0 0>
     }]
   }
   n25 = Anim Dynamics - Bone: MSTR_sleeve_r(ComponentPose=@n24, Alpha=1, ExternalForce="0,0,0") {
    angularDampingOverride=1
    angularSpringConstant=20000
    bAngularSpring=true
    bOverrideAngularDamping=true
    bUseGravityOverride=true
    gravityScale=0.35
    simulationSpace=RootRelative
    alphaBoolBlend {blendOption=Linear}
    alphaScaleBiasClamp {bInterpResult=true, interpSpeedIncreasing=0.3}
    boundBone {boneName=MSTR_sleeve_r}
    componentAppliedLinearAccClamp=<10000 10000 10000>
    componentLinearAccScale=<1 1 1>
    componentLinearVelScale=<1 1 1>
    gravityOverride {z=-980}
    physicsBodyDefinitions [{
      boundBone {boneName=MSTR_sleeve_r}
      boxExtents=<20 30 30>
      constraintSetup {angularLimitsMax=<2 0 2>, angularLimitsMin=<-2 0 -2>, angularTarget=<1 0 0>}
      localJointOffset=<3.5 0 0>
     }]
   }
   n26 = Anim Dynamics - Bone: dyn_sleeves_fr_l(ComponentPose=@n25, Alpha=1, ExternalForce="0,0,0") {
    angularDampingOverride=1
    angularSpringConstant=20000
    bAngularSpring=true
    bOverrideAngularDamping=true
    bUseGravityOverride=true
    gravityScale=0.35
    simulationSpace=RootRelative
    alphaBoolBlend {blendOption=Linear}
    boundBone {boneName=dyn_sleeves_fr_l}
    componentAppliedLinearAccClamp=<10000 10000 10000>
    componentLinearAccScale=<1 1 1>
    componentLinearVelScale=<1 1 1>
    gravityOverride {z=-980}
    physicsBodyDefinitions [{
      boundBone {boneName=dyn_sleeves_fr_l}
      boxExtents=<20 30 30>
      constraintSetup {angularLimitsMax=<5 5 5>, angularLimitsMin=<-5 -5 0>, angularTarget=<1 0 0>}
      localJointOffset=<2 0 0>
     }]
   }
   n27 = Anim Dynamics - Bone: dyn_sleeves_bk_l(ComponentPose=@n26, Alpha=1, ExternalForce="0,0,0") {
    angularDampingOverride=1
    angularSpringConstant=20000
    bAngularSpring=true
    bOverrideAngularDamping=true
    bUseGravityOverride=true
    gravityScale=0.35
    simulationSpace=RootRelative
    alphaBoolBlend {blendOption=Linear}
    boundBone {boneName=dyn_sleeves_bk_l}
    componentAppliedLinearAccClamp=<10000 10000 10000>
    componentLinearAccScale=<1 1 1>
    componentLinearVelScale=<1 1 1>
    gravityOverride {z=-980}
    physicsBodyDefinitions [{
      boundBone {boneName=dyn_sleeves_bk_l}
      boxExtents=<20 30 30>
      constraintSetup {angularLimitsMax=<7 7 0>, angularLimitsMin=<-7 -7 -7>, angularTarget=<1 0 0>}
      localJointOffset=<2 0 0>
     }]
   }
   n28 = Anim Dynamics - Bone: dyn_sleeves_bk_r(ComponentPose=@n27, Alpha=1, ExternalForce="0,0,0") {
    angularDampingOverride=1
    angularSpringConstant=20000
    bAngularSpring=true
    bOverrideAngularDamping=true
    bUseGravityOverride=true
    gravityScale=0.35
    simulationSpace=RootRelative
    alphaBoolBlend {blendOption=Linear}
    boundBone {boneName=dyn_sleeves_bk_r}
    componentAppliedLinearAccClamp=<10000 10000 10000>
    componentLinearAccScale=<1 1 1>
    componentLinearVelScale=<1 1 1>
    gravityOverride {z=-980}
    physicsBodyDefinitions [{
      boundBone {boneName=dyn_sleeves_bk_r}
      boxExtents=<20 30 30>
      constraintSetup {angularLimitsMax=<7 7 0>, angularLimitsMin=<-7 -7 -7>, angularTarget=<1 0 0>}
      localJointOffset=<2 0 0>
     }]
   }
   n29 = Anim Dynamics - Bone: dyn_sleeves_fr_r(ComponentPose=@n28, Alpha=1, ExternalForce="0,0,0") {
    angularDampingOverride=1
    angularSpringConstant=20000
    bAngularSpring=true
    bOverrideAngularDamping=true
    bUseGravityOverride=true
    gravityScale=0.35
    simulationSpace=RootRelative
    alphaBoolBlend {blendOption=Linear}
    boundBone {boneName=dyn_sleeves_fr_r}
    componentAppliedLinearAccClamp=<10000 10000 10000>
    componentLinearAccScale=<1 1 1>
    componentLinearVelScale=<1 1 1>
    gravityOverride {z=-980}
    physicsBodyDefinitions [{
      boundBone {boneName=dyn_sleeves_fr_r}
      boxExtents=<20 30 30>
      constraintSetup {angularLimitsMax=<5 5 5>, angularLimitsMin=<-5 -5 0>, angularTarget=<1 0 0>}
      localJointOffset=<2 0 0>
     }]
   }
   n30 = Component To Local(ComponentPose=@n29)
   n31 = Get CrouchingTargeting()
   n32 = Anim Dynamics - Bone: dyn_skirt_fr_c_01(
    ComponentPose=Local To Component(LocalPose=Blend Poses by bool(BlendPose_0=@n30, BlendPose_1=@n23, BlendTime_0=0.2, BlendTime_1=0.4, bActiveValue=@n31) {blendType=Linear})
    Alpha=1
    ExternalForce="0,0,0"
   ) {
    angularDampingOverride=1
    angularSpringConstant=20000
    bAngularSpring=true
    bOverrideAngularDamping=true
    bUseGravityOverride=true
    gravityScale=0.35
    simulationSpace=RootRelative
    alphaBoolBlend {blendOption=Linear}
    boundBone {boneName=dyn_skirt_fr_c_01}
    componentAppliedLinearAccClamp=<2500 2500 1500>
    componentLinearAccScale=<1 1 1>
    componentLinearVelScale=<1 1 1>
    gravityOverride {z=-980}
    physicsBodyDefinitions [{
      boundBone {boneName=dyn_skirt_fr_c_01}
      boxExtents=<20 30 30>
      constraintSetup {angularLimitsMax=<5 0 5>, angularLimitsMin=<-5 -15 -5>, angularTarget=<1 0 0>}
      localJointOffset=<2 0 0>
     }]
    simSpaceSettings {simSpaceAngularAlpha=0.1}
   }
   n33 = Anim Dynamics - Bone: dyn_skirt_fr_c_02(ComponentPose=@n32, Alpha=1, ExternalForce="0,0,0") {
    angularDampingOverride=1
    angularSpringConstant=20000
    bAngularSpring=true
    bOverrideAngularDamping=true
    bUseGravityOverride=true
    gravityScale=0.35
    simulationSpace=RootRelative
    alphaBoolBlend {blendOption=Linear}
    boundBone {boneName=dyn_skirt_fr_c_02}
    componentAppliedLinearAccClamp=<2500 2500 1500>
    componentLinearAccScale=<1 1 1>
    componentLinearVelScale=<1 1 1>
    gravityOverride {z=-980}
    physicsBodyDefinitions [{
      boundBone {boneName=dyn_skirt_fr_c_02}
      boxExtents=<20 30 30>
      constraintSetup {angularLimitsMax=<5 0 5>, angularLimitsMin=<-5 -15 -5>, angularTarget=<1 0 0>}
      localJointOffset=<2 0 0>
     }]
    simSpaceSettings {simSpaceAngularAlpha=0.1}
   }
   n34 = Anim Dynamics - Bone: dyn_skirt_fr_c_03(ComponentPose=@n33, Alpha=1, ExternalForce="0,0,0") {
    angularDampingOverride=1
    angularSpringConstant=20000
    bAngularSpring=true
    bOverrideAngularDamping=true
    bUseGravityOverride=true
    gravityScale=0.35
    simulationSpace=RootRelative
    alphaBoolBlend {blendOption=Linear}
    boundBone {boneName=dyn_skirt_fr_c_03}
    componentAppliedLinearAccClamp=<2500 2500 1500>
    componentLinearAccScale=<1 1 1>
    componentLinearVelScale=<1 1 1>
    gravityOverride {z=-980}
    physicsBodyDefinitions [{
      boundBone {boneName=dyn_skirt_fr_c_03}
      boxExtents=<20 30 30>
      constraintSetup {angularLimitsMax=<5 0 5>, angularLimitsMin=<-5 -15 -5>, angularTarget=<1 0 0>}
      localJointOffset=<2 0 0>
     }]
    simSpaceSettings {simSpaceAngularAlpha=0.1}
   }
   n35 = Anim Dynamics - Bone: dyn_skirt_fr_l_01(ComponentPose=@n34, Alpha=1, ExternalForce="0,0,0") {
    angularDampingOverride=1
    angularSpringConstant=20000
    bAngularSpring=true
    bOverrideAngularDamping=true
    bUseGravityOverride=true
    gravityScale=0.35
    simulationSpace=RootRelative
    alphaBoolBlend {blendOption=Linear}
    boundBone {boneName=dyn_skirt_fr_l_01}
    componentAppliedLinearAccClamp=<2500 2500 1500>
    componentLinearAccScale=<1 1 1>
    componentLinearVelScale=<1 1 1>
    gravityOverride {z=-980}
    physicsBodyDefinitions [{
      boundBone {boneName=dyn_skirt_fr_l_01}
      boxExtents=<20 30 30>
      constraintSetup {angularLimitsMax=<5 0 5>, angularLimitsMin=<-5 -5 -5>, angularTarget=<1 0 0>}
      localJointOffset=<2 0 0>
     }]
    simSpaceSettings {simSpaceAngularAlpha=0.1}
   }
   n36 = Anim Dynamics - Bone: dyn_skirt_fr_l_02(ComponentPose=@n35, Alpha=1, ExternalForce="0,0,0") {
    angularDampingOverride=1
    angularSpringConstant=20000
    bAngularSpring=true
    bOverrideAngularDamping=true
    bUseGravityOverride=true
    gravityScale=0.35
    simulationSpace=RootRelative
    alphaBoolBlend {blendOption=Linear}
    boundBone {boneName=dyn_skirt_fr_l_02}
    componentAppliedLinearAccClamp=<2500 2500 1500>
    componentLinearAccScale=<1 1 1>
    componentLinearVelScale=<1 1 1>
    gravityOverride {z=-980}
    physicsBodyDefinitions [{
      boundBone {boneName=dyn_skirt_fr_l_02}
      boxExtents=<20 30 30>
      constraintSetup {angularLimitsMax=<5 0 5>, angularLimitsMin=<-5 -5 -5>, angularTarget=<1 0 0>}
      localJointOffset=<2 0 0>
     }]
    simSpaceSettings {simSpaceAngularAlpha=0.1}
   }
   n37 = Anim Dynamics - Bone: dyn_skirt_fr_l_03(ComponentPose=@n36, Alpha=1, ExternalForce="0,0,0") {
    angularDampingOverride=1
    angularSpringConstant=20000
    bAngularSpring=true
    bOverrideAngularDamping=true
    bUseGravityOverride=true
    gravityScale=0.35
    simulationSpace=RootRelative
    alphaBoolBlend {blendOption=Linear}
    boundBone {boneName=dyn_skirt_fr_l_03}
    componentAppliedLinearAccClamp=<2500 2500 1500>
    componentLinearAccScale=<1 1 1>
    componentLinearVelScale=<1 1 1>
    gravityOverride {z=-980}
    physicsBodyDefinitions [{
      boundBone {boneName=dyn_skirt_fr_l_03}
      boxExtents=<20 30 30>
      constraintSetup {angularLimitsMax=<5 0 5>, angularLimitsMin=<-5 -5 -15>, angularTarget=<1 0 0>}
      localJointOffset=<2 0 0>
     }]
    simSpaceSettings {simSpaceAngularAlpha=0.1}
   }
   n38 = Anim Dynamics - Bone: dyn_skirt_sd_l_01(ComponentPose=@n37, Alpha=1, ExternalForce="0,0,0") {
    angularDampingOverride=1
    angularSpringConstant=20000
    bAngularSpring=true
    bOverrideAngularDamping=true
    bUseGravityOverride=true
    gravityScale=0.35
    simulationSpace=RootRelative
    alphaBoolBlend {blendOption=Linear}
    boundBone {boneName=dyn_skirt_sd_l_01}
    componentAppliedLinearAccClamp=<2500 2500 1500>
    componentLinearAccScale=<1 1 1>
    componentLinearVelScale=<1 1 1>
    gravityOverride {z=-980}
    physicsBodyDefinitions [{
      boundBone {boneName=dyn_skirt_sd_l_01}
      boxExtents=<20 30 30>
      constraintSetup {angularLimitsMax=<5 0 5>, angularLimitsMin=<-5 -15 -5>, angularTarget=<1 0 0>}
      localJointOffset=<2 0 0>
     }]
    simSpaceSettings {simSpaceAngularAlpha=0.1}
   }
   n39 = Anim Dynamics - Bone: dyn_skirt_sd_l_02(ComponentPose=@n38, Alpha=1, ExternalForce="0,0,0") {
    angularDampingOverride=1
    angularSpringConstant=20000
    bAngularSpring=true
    bOverrideAngularDamping=true
    bUseGravityOverride=true
    gravityScale=0.35
    simulationSpace=RootRelative
    alphaBoolBlend {blendOption=Linear}
    boundBone {boneName=dyn_skirt_sd_l_02}
    componentAppliedLinearAccClamp=<2500 2500 1500>
    componentLinearAccScale=<1 1 1>
    componentLinearVelScale=<1 1 1>
    gravityOverride {z=-980}
    physicsBodyDefinitions [{
      boundBone {boneName=dyn_skirt_sd_l_02}
      boxExtents=<20 30 30>
      constraintSetup {angularLimitsMax=<5 0 5>, angularLimitsMin=<-5 -15 -5>, angularTarget=<1 0 0>}
      localJointOffset=<2 0 0>
     }]
    simSpaceSettings {simSpaceAngularAlpha=0.1}
   }
   n40 = Anim Dynamics - Bone: dyn_skirt_sd_l_03(ComponentPose=@n39, Alpha=1, ExternalForce="0,0,0") {
    angularDampingOverride=1
    angularSpringConstant=20000
    bAngularSpring=true
    bOverrideAngularDamping=true
    bUseGravityOverride=true
    gravityScale=0.35
    simulationSpace=RootRelative
    alphaBoolBlend {blendOption=Linear}
    boundBone {boneName=dyn_skirt_sd_l_03}
    componentAppliedLinearAccClamp=<2500 2500 1500>
    componentLinearAccScale=<1 1 1>
    componentLinearVelScale=<1 1 1>
    gravityOverride {z=-980}
    physicsBodyDefinitions [{
      boundBone {boneName=dyn_skirt_sd_l_03}
      boxExtents=<20 30 30>
      constraintSetup {angularLimitsMax=<5 0 5>, angularLimitsMin=<-5 -15 -5>, angularTarget=<1 0 0>}
      localJointOffset=<2 0 0>
     }]
    simSpaceSettings {simSpaceAngularAlpha=0.1}
   }
   n41 = Anim Dynamics - Bone: dyn_skirt_bk_l_01(ComponentPose=@n40, Alpha=1, ExternalForce="0,0,0") {
    angularDampingOverride=1
    angularSpringConstant=20000
    bAngularSpring=true
    bOverrideAngularDamping=true
    bUseGravityOverride=true
    gravityScale=0.35
    simulationSpace=RootRelative
    alphaBoolBlend {blendOption=Linear}
    boundBone {boneName=dyn_skirt_bk_l_01}
    componentAppliedLinearAccClamp=<2500 2500 1500>
    componentLinearAccScale=<1 1 1>
    componentLinearVelScale=<1 1 1>
    gravityOverride {z=-980}
    physicsBodyDefinitions [{
      boundBone {boneName=dyn_skirt_bk_l_01}
      boxExtents=<20 30 30>
      constraintSetup {angularLimitsMax=<5 0 5>, angularLimitsMin=<-5 -15 -5>, angularTarget=<1 0 0>}
      localJointOffset=<2 0 0>
     }]
    simSpaceSettings {simSpaceAngularAlpha=0.1}
   }
   n42 = Anim Dynamics - Bone: dyn_skirt_bk_l_02(ComponentPose=@n41, Alpha=1, ExternalForce="0,0,0") {
    angularDampingOverride=1
    angularSpringConstant=20000
    bAngularSpring=true
    bOverrideAngularDamping=true
    bUseGravityOverride=true
    gravityScale=0.35
    simulationSpace=RootRelative
    alphaBoolBlend {blendOption=Linear}
    boundBone {boneName=dyn_skirt_bk_l_02}
    componentAppliedLinearAccClamp=<2500 2500 1500>
    componentLinearAccScale=<1 1 1>
    componentLinearVelScale=<1 1 1>
    gravityOverride {z=-980}
    physicsBodyDefinitions [{
      boundBone {boneName=dyn_skirt_bk_l_02}
      boxExtents=<20 30 30>
      constraintSetup {angularLimitsMax=<5 0 5>, angularLimitsMin=<-5 -15 -5>, angularTarget=<1 0 0>}
      localJointOffset=<2 0 0>
     }]
    simSpaceSettings {simSpaceAngularAlpha=0.1}
   }
   n43 = Anim Dynamics - Bone: dyn_skirt_bk_l_03(ComponentPose=@n42, Alpha=1, ExternalForce="0,0,0") {
    angularDampingOverride=1
    angularSpringConstant=20000
    bAngularSpring=true
    bOverrideAngularDamping=true
    bUseGravityOverride=true
    gravityScale=0.35
    simulationSpace=RootRelative
    alphaBoolBlend {blendOption=Linear}
    boundBone {boneName=dyn_skirt_bk_l_03}
    componentAppliedLinearAccClamp=<2500 2500 1500>
    componentLinearAccScale=<1 1 1>
    componentLinearVelScale=<1 1 1>
    gravityOverride {z=-980}
    physicsBodyDefinitions [{
      boundBone {boneName=dyn_skirt_bk_l_03}
      boxExtents=<20 30 30>
      constraintSetup {angularLimitsMax=<5 0 5>, angularLimitsMin=<-5 -15 -5>, angularTarget=<1 0 0>}
      localJointOffset=<2 0 0>
     }]
    simSpaceSettings {simSpaceAngularAlpha=0.1}
   }
   n44 = Anim Dynamics - Bone: dyn_skirt_bk_c_01(ComponentPose=@n43, Alpha=1, ExternalForce="0,0,0") {
    angularDampingOverride=1
    angularSpringConstant=20000
    bAngularSpring=true
    bOverrideAngularDamping=true
    bUseGravityOverride=true
    gravityScale=0.35
    simulationSpace=RootRelative
    alphaBoolBlend {blendOption=Linear}
    boundBone {boneName=dyn_skirt_bk_c_01}
    componentAppliedLinearAccClamp=<2500 2500 1500>
    componentLinearAccScale=<1 1 1>
    componentLinearVelScale=<1 1 1>
    gravityOverride {z=-980}
    physicsBodyDefinitions [{
      boundBone {boneName=dyn_skirt_bk_c_01}
      boxExtents=<20 30 30>
      constraintSetup {angularLimitsMax=<5 0 5>, angularLimitsMin=<-5 -15 -5>, angularTarget=<1 0 0>}
      localJointOffset=<2 0 0>
     }]
    simSpaceSettings {simSpaceAngularAlpha=0.1}
   }
   n45 = Anim Dynamics - Bone: dyn_skirt_bk_c_02(ComponentPose=@n44, Alpha=1, ExternalForce="0,0,0") {
    angularDampingOverride=1
    angularSpringConstant=20000
    bAngularSpring=true
    bOverrideAngularDamping=true
    bUseGravityOverride=true
    gravityScale=0.35
    simulationSpace=RootRelative
    alphaBoolBlend {blendOption=Linear}
    boundBone {boneName=dyn_skirt_bk_c_02}
    componentAppliedLinearAccClamp=<2500 2500 1500>
    componentLinearAccScale=<1 1 1>
    componentLinearVelScale=<1 1 1>
    gravityOverride {z=-980}
    physicsBodyDefinitions [{
      boundBone {boneName=dyn_skirt_bk_c_02}
      boxExtents=<20 30 30>
      constraintSetup {angularLimitsMax=<5 0 5>, angularLimitsMin=<-5 -15 -5>, angularTarget=<1 0 0>}
      localJointOffset=<2 0 0>
     }]
    simSpaceSettings {simSpaceAngularAlpha=0.1}
   }
   n46 = Anim Dynamics - Bone: dyn_skirt_bk_c_03(ComponentPose=@n45, Alpha=1, ExternalForce="0,0,0") {
    angularDampingOverride=1
    angularSpringConstant=20000
    bAngularSpring=true
    bOverrideAngularDamping=true
    bUseGravityOverride=true
    gravityScale=0.35
    simulationSpace=RootRelative
    alphaBoolBlend {blendOption=Linear}
    boundBone {boneName=dyn_skirt_bk_c_03}
    componentAppliedLinearAccClamp=<2500 2500 1500>
    componentLinearAccScale=<1 1 1>
    componentLinearVelScale=<1 1 1>
    gravityOverride {z=-980}
    physicsBodyDefinitions [{
      boundBone {boneName=dyn_skirt_bk_c_03}
      boxExtents=<20 30 30>
      constraintSetup {angularLimitsMax=<5 0 5>, angularLimitsMin=<-5 -15 -5>, angularTarget=<1 0 0>}
      localJointOffset=<2 0 0>
     }]
    simSpaceSettings {simSpaceAngularAlpha=0.1}
   }
   n47 = Anim Dynamics - Bone: dyn_skirt_bk_r_01(ComponentPose=@n46, Alpha=1, ExternalForce="0,0,0") {
    angularDampingOverride=1
    angularSpringConstant=20000
    bAngularSpring=true
    bOverrideAngularDamping=true
    bUseGravityOverride=true
    gravityScale=0.35
    simulationSpace=RootRelative
    alphaBoolBlend {blendOption=Linear}
    boundBone {boneName=dyn_skirt_bk_r_01}
    componentAppliedLinearAccClamp=<2500 2500 1500>
    componentLinearAccScale=<1 1 1>
    componentLinearVelScale=<1 1 1>
    gravityOverride {z=-980}
    physicsBodyDefinitions [{
      boundBone {boneName=dyn_skirt_bk_r_01}
      boxExtents=<20 30 30>
      constraintSetup {angularLimitsMax=<5 0 5>, angularLimitsMin=<-5 -15 -5>, angularTarget=<1 0 0>}
      localJointOffset=<2 0 0>
     }]
    simSpaceSettings {simSpaceAngularAlpha=0.1}
   }
   n48 = Anim Dynamics - Bone: dyn_skirt_bk_r_02(ComponentPose=@n47, Alpha=1, ExternalForce="0,0,0") {
    angularDampingOverride=1
    angularSpringConstant=20000
    bAngularSpring=true
    bOverrideAngularDamping=true
    bUseGravityOverride=true
    gravityScale=0.35
    simulationSpace=RootRelative
    alphaBoolBlend {blendOption=Linear}
    boundBone {boneName=dyn_skirt_bk_r_02}
    componentAppliedLinearAccClamp=<2500 2500 1500>
    componentLinearAccScale=<1 1 1>
    componentLinearVelScale=<1 1 1>
    gravityOverride {z=-980}
    physicsBodyDefinitions [{
      boundBone {boneName=dyn_skirt_bk_r_02}
      boxExtents=<20 30 30>
      constraintSetup {angularLimitsMax=<5 0 5>, angularLimitsMin=<-5 -15 -5>, angularTarget=<1 0 0>}
      localJointOffset=<2 0 0>
     }]
    simSpaceSettings {simSpaceAngularAlpha=0.1}
   }
   n49 = Anim Dynamics - Bone: dyn_skirt_bk_r_03(ComponentPose=@n48, Alpha=1, ExternalForce="0,0,0") {
    angularDampingOverride=1
    angularSpringConstant=20000
    bAngularSpring=true
    bOverrideAngularDamping=true
    bUseGravityOverride=true
    gravityScale=0.35
    simulationSpace=RootRelative
    alphaBoolBlend {blendOption=Linear}
    boundBone {boneName=dyn_skirt_bk_r_03}
    componentAppliedLinearAccClamp=<2500 2500 1500>
    componentLinearAccScale=<1 1 1>
    componentLinearVelScale=<1 1 1>
    gravityOverride {z=-980}
    physicsBodyDefinitions [{
      boundBone {boneName=dyn_skirt_bk_r_03}
      boxExtents=<20 30 30>
      constraintSetup {angularLimitsMax=<5 0 5>, angularLimitsMin=<-5 -15 -5>, angularTarget=<1 0 0>}
      localJointOffset=<2 0 0>
     }]
    simSpaceSettings {simSpaceAngularAlpha=0.1}
   }
   n50 = Anim Dynamics - Bone: dyn_skirt_sd_r_01(ComponentPose=@n49, Alpha=1, ExternalForce="0,0,0") {
    angularDampingOverride=1
    angularSpringConstant=20000
    bAngularSpring=true
    bOverrideAngularDamping=true
    bUseGravityOverride=true
    gravityScale=0.35
    simulationSpace=RootRelative
    alphaBoolBlend {blendOption=Linear}
    boundBone {boneName=dyn_skirt_sd_r_01}
    componentAppliedLinearAccClamp=<2500 2500 1500>
    componentLinearAccScale=<1 1 1>
    componentLinearVelScale=<1 1 1>
    gravityOverride {z=-980}
    physicsBodyDefinitions [{
      boundBone {boneName=dyn_skirt_sd_r_01}
      boxExtents=<20 30 30>
      constraintSetup {angularLimitsMax=<5 0 5>, angularLimitsMin=<-5 -15 -5>, angularTarget=<1 0 0>}
      localJointOffset=<2 0 0>
     }]
    simSpaceSettings {simSpaceAngularAlpha=0.1}
   }
   n51 = Anim Dynamics - Bone: dyn_skirt_sd_r_02(ComponentPose=@n50, Alpha=1, ExternalForce="0,0,0") {
    angularDampingOverride=1
    angularSpringConstant=20000
    bAngularSpring=true
    bOverrideAngularDamping=true
    bUseGravityOverride=true
    gravityScale=0.35
    simulationSpace=RootRelative
    alphaBoolBlend {blendOption=Linear}
    boundBone {boneName=dyn_skirt_sd_r_02}
    componentAppliedLinearAccClamp=<2500 2500 1500>
    componentLinearAccScale=<1 1 1>
    componentLinearVelScale=<1 1 1>
    gravityOverride {z=-980}
    physicsBodyDefinitions [{
      boundBone {boneName=dyn_skirt_sd_r_02}
      boxExtents=<20 30 30>
      constraintSetup {angularLimitsMax=<5 0 5>, angularLimitsMin=<-5 -15 -5>, angularTarget=<1 0 0>}
      localJointOffset=<2 0 0>
     }]
    simSpaceSettings {simSpaceAngularAlpha=0.1}
   }
   n52 = Anim Dynamics - Bone: dyn_skirt_sd_r_03(ComponentPose=@n51, Alpha=1, ExternalForce="0,0,0") {
    angularDampingOverride=1
    angularSpringConstant=20000
    bAngularSpring=true
    bOverrideAngularDamping=true
    bUseGravityOverride=true
    gravityScale=0.35
    simulationSpace=RootRelative
    alphaBoolBlend {blendOption=Linear}
    boundBone {boneName=dyn_skirt_sd_r_03}
    componentAppliedLinearAccClamp=<2500 2500 1500>
    componentLinearAccScale=<1 1 1>
    componentLinearVelScale=<1 1 1>
    gravityOverride {z=-980}
    physicsBodyDefinitions [{
      boundBone {boneName=dyn_skirt_sd_r_03}
      boxExtents=<20 30 30>
      constraintSetup {angularLimitsMax=<5 0 5>, angularLimitsMin=<-5 -15 -5>, angularTarget=<1 0 0>}
      localJointOffset=<2 0 0>
     }]
    simSpaceSettings {simSpaceAngularAlpha=0.1}
   }
   n53 = Anim Dynamics - Bone: dyn_skirt_fr_r_01(ComponentPose=@n52, Alpha=1, ExternalForce="0,0,0") {
    angularDampingOverride=1
    angularSpringConstant=20000
    bAngularSpring=true
    bOverrideAngularDamping=true
    bUseGravityOverride=true
    gravityScale=0.35
    simulationSpace=RootRelative
    alphaBoolBlend {blendOption=Linear}
    boundBone {boneName=dyn_skirt_fr_r_01}
    componentAppliedLinearAccClamp=<2500 2500 1500>
    componentLinearAccScale=<1 1 1>
    componentLinearVelScale=<1 1 1>
    gravityOverride {z=-980}
    physicsBodyDefinitions [{
      boundBone {boneName=dyn_skirt_fr_r_01}
      boxExtents=<20 30 30>
      constraintSetup {angularLimitsMax=<5 0 5>, angularLimitsMin=<-5 -5 -5>, angularTarget=<1 0 0>}
      localJointOffset=<2 0 0>
     }]
    simSpaceSettings {simSpaceAngularAlpha=0.1}
   }
   n54 = Anim Dynamics - Bone: dyn_skirt_fr_r_02(ComponentPose=@n53, Alpha=1, ExternalForce="0,0,0") {
    angularDampingOverride=1
    angularSpringConstant=20000
    bAngularSpring=true
    bOverrideAngularDamping=true
    bUseGravityOverride=true
    gravityScale=0.35
    simulationSpace=RootRelative
    alphaBoolBlend {blendOption=Linear}
    boundBone {boneName=dyn_skirt_fr_r_02}
    componentAppliedLinearAccClamp=<2500 2500 1500>
    componentLinearAccScale=<1 1 1>
    componentLinearVelScale=<1 1 1>
    gravityOverride {z=-980}
    physicsBodyDefinitions [{
      boundBone {boneName=dyn_skirt_fr_r_02}
      boxExtents=<20 30 30>
      constraintSetup {angularLimitsMax=<5 0 5>, angularLimitsMin=<-5 -5 -5>, angularTarget=<1 0 0>}
      localJointOffset=<2 0 0>
     }]
    simSpaceSettings {simSpaceAngularAlpha=0.1}
   }
   n55 = Anim Dynamics - Bone: dyn_skirt_fr_r_03(ComponentPose=@n54, Alpha=1, ExternalForce="0,0,0") {
    angularDampingOverride=1
    angularSpringConstant=20000
    bAngularSpring=true
    bOverrideAngularDamping=true
    bUseGravityOverride=true
    gravityScale=0.35
    simulationSpace=RootRelative
    alphaBoolBlend {blendOption=Linear}
    boundBone {boneName=dyn_skirt_fr_r_03}
    componentAppliedLinearAccClamp=<2500 2500 1500>
    componentLinearAccScale=<1 1 1>
    componentLinearVelScale=<1 1 1>
    gravityOverride {z=-980}
    physicsBodyDefinitions [{
      boundBone {boneName=dyn_skirt_fr_r_03}
      boxExtents=<20 30 30>
      constraintSetup {angularLimitsMax=<5 -5 5>, angularLimitsMin=<-5 -15 -5>, angularTarget=<1 0 0>}
      localJointOffset=<2 0 0>
     }]
    simSpaceSettings {simSpaceAngularAlpha=0.1}
   }
   n56 = Anim Dynamics - Bone: dyn_belt_sd_r(ComponentPose=@n55, Alpha=1, ExternalForce="0,0,0") {
    angularDampingOverride=1
    angularSpringConstant=20000
    bAngularSpring=true
    bOverrideAngularDamping=true
    bUseGravityOverride=true
    gravityScale=0.35
    simulationSpace=RootRelative
    alphaBoolBlend {blendOption=Linear}
    boundBone {boneName=dyn_belt_sd_r}
    componentAppliedLinearAccClamp=<2500 2500 1500>
    componentLinearAccScale=<1 1 1>
    componentLinearVelScale=<1 1 1>
    gravityOverride {z=-980}
    physicsBodyDefinitions [{
      boundBone {boneName=dyn_belt_sd_r}
      boxExtents=<20 30 30>
      constraintSetup {angularLimitsMax=<2 2 2>, angularLimitsMin=<-2 -2 -2>, angularTarget=<1 0 0>, linearAxesMin=<-2 0 0>}
      localJointOffset=<4 0 0>
     }]
   }
   n57 = Anim Dynamics - Bone: dyn_belt_fr_r(ComponentPose=@n56, Alpha=1, ExternalForce="0,0,0") {
    angularDampingOverride=1
    angularSpringConstant=20000
    bAngularSpring=true
    bOverrideAngularDamping=true
    bUseGravityOverride=true
    gravityScale=0.35
    simulationSpace=RootRelative
    alphaBoolBlend {blendOption=Linear}
    boundBone {boneName=dyn_belt_fr_r}
    componentAppliedLinearAccClamp=<2500 2500 1500>
    componentLinearAccScale=<1 1 1>
    componentLinearVelScale=<1 1 1>
    gravityOverride {z=-980}
    physicsBodyDefinitions [{
      boundBone {boneName=dyn_belt_fr_r}
      boxExtents=<20 30 30>
      constraintSetup {angularLimitsMax=<2 2 2>, angularLimitsMin=<-2 -2 -2>, angularTarget=<1 0 0>, linearAxesMin=<-2 0 0>}
      localJointOffset=<4 0 0>
     }]
   }
   n58 = Anim Dynamics - Bone: dyn_belt_bk_r(ComponentPose=@n57, Alpha=1, ExternalForce="0,0,0") {
    angularDampingOverride=1
    angularSpringConstant=20000
    bAngularSpring=true
    bOverrideAngularDamping=true
    bUseGravityOverride=true
    gravityScale=0.35
    simulationSpace=RootRelative
    alphaBoolBlend {blendOption=Linear}
    boundBone {boneName=dyn_belt_bk_r}
    componentAppliedLinearAccClamp=<2500 2500 1500>
    componentLinearAccScale=<1 1 1>
    componentLinearVelScale=<1 1 1>
    gravityOverride {z=-980}
    physicsBodyDefinitions [{
      boundBone {boneName=dyn_belt_bk_r}
      boxExtents=<20 30 30>
      constraintSetup {angularLimitsMax=<2 2 2>, angularLimitsMin=<-2 -2 -2>, angularTarget=<1 0 0>, linearAxesMin=<-2 0 0>}
      localJointOffset=<4 0 0>
     }]
   }
   n59 = Anim Dynamics - Bone: dyn_strap_fr_r_01(ComponentPose=@n58, Alpha=1, ExternalForce="0,0,0") {
    angularDampingOverride=1
    angularSpringConstant=20000
    bAngularSpring=true
    bOverrideAngularDamping=true
    bUseGravityOverride=true
    gravityScale=0.35
    simulationSpace=RootRelative
    alphaBoolBlend {blendOption=Linear}
    boundBone {boneName=dyn_strap_fr_r_01}
    componentAppliedLinearAccClamp=<10000 10000 10000>
    componentLinearAccScale=<0.7 0.7 0.7>
    componentLinearVelScale=<1 1 1>
    gravityOverride {z=-980}
    physicsBodyDefinitions [{
      boundBone {boneName=dyn_strap_fr_r_01}
      boxExtents=<20 30 30>
      constraintSetup {angularLimitsMax=<5 0 0>, angularLimitsMin=<-5 -5 -5>, angularTarget=<1 0 0>}
      localJointOffset=<4 0 0>
     }]
   }
   n60 = Anim Dynamics - Bone: dyn_strap_fr_r_02(ComponentPose=@n59, Alpha=1, ExternalForce="0,0,0") {
    angularDampingOverride=1
    angularSpringConstant=20000
    bAngularSpring=true
    bOverrideAngularDamping=true
    bUseGravityOverride=true
    gravityScale=0.35
    simulationSpace=RootRelative
    alphaBoolBlend {blendOption=Linear}
    boundBone {boneName=dyn_strap_fr_r_02}
    componentAppliedLinearAccClamp=<10000 10000 10000>
    componentLinearAccScale=<0.7 0.7 0.7>
    componentLinearVelScale=<1 1 1>
    gravityOverride {z=-980}
    physicsBodyDefinitions [{
      boundBone {boneName=dyn_strap_fr_r_02}
      boxExtents=<20 30 30>
      constraintSetup {angularLimitsMax=<5 0 0>, angularLimitsMin=<-25 -5 -25>, angularTarget=<1 0 0>}
      localJointOffset=<4 0 0>
     }]
   }
   n61 = Anim Dynamics - Bone: dyn_strap_fr_r_03(ComponentPose=@n60, Alpha=1, ExternalForce="0,0,0") {
    angularDampingOverride=1
    angularSpringConstant=20000
    bAngularSpring=true
    bOverrideAngularDamping=true
    bUseGravityOverride=true
    gravityScale=0.35
    simulationSpace=RootRelative
    alphaBoolBlend {blendOption=Linear}
    boundBone {boneName=dyn_strap_fr_r_03}
    componentAppliedLinearAccClamp=<10000 10000 10000>
    componentLinearAccScale=<0.7 0.7 0.7>
    componentLinearVelScale=<1 1 1>
    gravityOverride {z=-980}
    physicsBodyDefinitions [{
      boundBone {boneName=dyn_strap_fr_r_03}
      boxExtents=<20 30 30>
      constraintSetup {angularLimitsMax=<0 0 0>, angularLimitsMin=<0 -5 -25>, angularTarget=<1 0 0>}
      localJointOffset=<4 0 0>
     }]
   }
   n62 = Anim Dynamics - Bone: dyn_strap_bk_r_01(ComponentPose=@n61, Alpha=1, ExternalForce="0,0,0") {
    angularDampingOverride=1
    angularSpringConstant=20000
    bAngularSpring=true
    bOverrideAngularDamping=true
    bUseGravityOverride=true
    gravityScale=0.35
    simulationSpace=RootRelative
    alphaBoolBlend {blendOption=Linear}
    boundBone {boneName=dyn_strap_bk_r_01}
    componentAppliedLinearAccClamp=<10000 10000 10000>
    componentLinearAccScale=<0.7 0.7 0.7>
    componentLinearVelScale=<1 1 1>
    gravityOverride {z=-980}
    physicsBodyDefinitions [{
      boundBone {boneName=dyn_strap_bk_r_01}
      boxExtents=<20 30 30>
      constraintSetup {angularLimitsMax=<5 0 5>, angularLimitsMin=<-5 0 0>, angularTarget=<1 0 0>}
      localJointOffset=<4 0 0>
     }]
   }
   n63 = Anim Dynamics - Bone: dyn_strap_bk_r_02(ComponentPose=@n62, Alpha=1, ExternalForce="0,0,0") {
    angularDampingOverride=1
    angularSpringConstant=20000
    bAngularSpring=true
    bOverrideAngularDamping=true
    bUseGravityOverride=true
    gravityScale=0.35
    simulationSpace=RootRelative
    alphaBoolBlend {blendOption=Linear}
    boundBone {boneName=dyn_strap_bk_r_02}
    componentAppliedLinearAccClamp=<10000 10000 10000>
    componentLinearAccScale=<0.7 0.7 0.7>
    componentLinearVelScale=<1 1 1>
    gravityOverride {z=-980}
    physicsBodyDefinitions [{
      boundBone {boneName=dyn_strap_bk_r_02}
      boxExtents=<20 30 30>
      constraintSetup {angularLimitsMax=<5 0 30>, angularLimitsMin=<-25 0 0>, angularTarget=<1 0 0>}
      localJointOffset=<4 0 0>
     }]
   }
   n64 = Anim Dynamics - Bone: dyn_strap_bk_r_03(ComponentPose=@n63, Alpha=1, ExternalForce="0,0,0") {
    angularDampingOverride=1
    angularSpringConstant=20000
    bAngularSpring=true
    bOverrideAngularDamping=true
    bUseGravityOverride=true
    gravityScale=0.35
    simulationSpace=RootRelative
    alphaBoolBlend {blendOption=Linear}
    boundBone {boneName=dyn_strap_bk_r_03}
    componentAppliedLinearAccClamp=<10000 10000 10000>
    componentLinearAccScale=<0.7 0.7 0.7>
    componentLinearVelScale=<1 1 1>
    gravityOverride {z=-980}
    physicsBodyDefinitions [{
      boundBone {boneName=dyn_strap_bk_r_03}
      boxExtents=<20 30 30>
      constraintSetup {angularLimitsMax=<5 0 15>, angularLimitsMin=<-15 0 0>, angularTarget=<1 0 0>}
      localJointOffset=<4 0 0>
     }]
   }
   Save cached pose 'Dynamics Cache'(Pose=Component To Local(ComponentPose=@n64)) {cacheName=Dynamics Cache}
   n65 = RigidBody(ComponentPose=Local To Component(LocalPose=Use cached pose 'CtrlRig Cache'())) {
    overridePhysicsAsset=&/F_MED_JadeTowelHope_Physics_Crouching
    alphaBoolBlend {blendOption=Linear}
    baseBoneRef {boneName=spine_04}
    componentAppliedLinearAccClamp=<3000 3000 3000>
    componentLinearAccScale=<0.3 0.3 0.3>
    componentLinearVelScale=<0.3 0.3 0.3>
    simSpaceSettings {worldAlpha=0.5}
   }
   n66 = Component To Local(ComponentPose=@n65)
   n67 = RigidBody(ComponentPose=Local To Component(LocalPose=Use cached pose 'CtrlRig Cache'())) {
    bOverrideWorldGravity=true
    overridePhysicsAsset=&/F_MED_JadeTowelHope_Physics
    simulationSpace=BaseBoneSpace
    alphaBoolBlend {blendOption=Linear}
    baseBoneRef {boneName=spine_04}
    componentAppliedLinearAccClamp=<3000 3000 3000>
    componentLinearAccScale=<0.3 0.3 0.3>
    componentLinearVelScale=<0.3 0.3 0.3>
    overrideWorldGravity {z=-980}
    simSpaceSettings {worldAlpha=0.6}
   }
   n68 = Component To Local(ComponentPose=@n67)
   n69 = Get Crouching Bool() -> (Crouching Bool:bool)
   n70 = RigidBody!(
    ComponentPose=Local To Component(LocalPose=Blend Poses by bool(BlendPose_0=@n66, BlendPose_1=@n68, BlendTime_0=0.2, BlendTime_1=0.4, bActiveValue=@n69) {blendType=Linear})
    Alpha
    OverrideWorldGravity:Vector="0,0,-980"
    ExternalForce
   ) -> (Pose) {
    bOverrideWorldGravity=true
    overridePhysicsAsset=&/F_MED_JadeTowelHope_Laces_Physics
    simulationSpace=BaseBoneSpace
    alphaBoolBlend {blendOption=Linear}
    alphaScaleBiasClamp {bInterpResult=true}
    baseBoneRef {boneName=spine_03}
    componentAppliedLinearAccClamp=<3000 3000 3000>
    componentLinearAccScale=<0.3 0.3 0.3>
    componentLinearVelScale=<0.3 0.3 0.3>
    simSpaceSettings {
     maxAngularAcceleration=1000
     maxAngularVelocity=3
     maxLinearAcceleration=700
     maxLinearVelocity=500
     velocityScaleZ=0
     worldAlpha=0.7
     externalLinearDragV=<0.2 0.2>
    }
   }
   Save cached pose 'RigidBodyCache'(Pose=Component To Local(ComponentPose=@n70)) {cacheName=RigidBodyCache}
   Use cached pose 'CtrlRig Cache'()
   Use cached pose 'RigidBodyCache'()
   Get backpackSocket_rotate() -> (backpackSocket_rotate:Rotator)
  ]}
 EventGraph {nodes [
   n1 = Event BlueprintUpdateAnimation()
   > Set Alpha_DeltaTimeX(Alpha_DeltaTimeX:double=@n1.DeltaTimeX) -> (Output_Get:double)
   > n2 = Sequence!(execute) -> (
    then_0
    then_1
    then_2:Exec
    then_3:Exec
    then_4:Exec
    then_5:Exec
    then_6:Exec
    then_7:Exec
    then_8:Exec
    then_9:Exec
    then_10:Exec
    then_11:Exec
    then_12:Exec
   )
   n3 = GetOwnerSkeletalMeshComponent()
   n4 = BreakRotator(InRot=BreakTransform(InTransform=GetDeltaTransformFromRefPose(self=@n3, BoneName=clavicle_l, BaseName=spine_03)).Rotation)
   Set backpackSocket_rotate(execute=@n2.then_10, backpackSocket_rotate:Rotator=MakeRotator(Pitch=MapRangeClamped(Value=@n4, InRangeA=-10, InRangeB=-30, OutRangeB=25))) -> (Output_Get:Rotator)
   n5 = Cast To FortPlayerPawn(execute=@n2.then_11, Object=TryGetPawnOwner())
   n6 = GetSkeletalMeshForPartType(self=@n5, PartType=Backpack)
   > Set JadeTowelHopeBP(JadeTowelHopeBP:bool=MatchesWildcard(SourceString=GetDisplayName(Object=@n6), Wildcard=*JadeTowelHope*, SearchCase=CaseSensitive)) -> (Output_Get:bool)
   n7 = GetAnimInstance(self=GetOwnerSkeletalMeshComponent())
   n8 = OR Boolean(
    A=Montage_IsPlaying(self=@n7, Montage=/BRCosmetics/Animation/Game/MainPlayer/Emotes/GoldenGlory/CMF/Emote_GoldenGlory_CMF_M)
    B=Montage_IsPlaying(self=@n7, Montage=/BRCosmetics/Animation/Game/MainPlayer/Emotes/GoldenGlory/CMF/Emote_GoldenGlory_CMF_FrontEnd_M)
   )
   Set EmoteGoldenGlory(execute=@n2.then_12, EmoteGoldenGlory:double="To Float (Boolean)"(InBool=@n8)) -> (Output_Get:double)
   > n9 = Branch(Condition=@n8)
   > Set Dyn_EmoteGoldenGlory(Dyn_EmoteGoldenGlory:double=0.4) -> (Output_Get:double)
   Set Dyn_EmoteGoldenGlory(execute=@n9.else, Dyn_EmoteGoldenGlory=1)
   n10 = Branch(execute=@n2.then_3, Condition=Get bIsPlayingMeleeAnim())
   > Set Alpha_Melee(Alpha_Melee:double=1) -> (Output_Get:double)
   Set Alpha_Melee(execute=@n10.else, Alpha_Melee=0)
   n11 = GetAnimInstance(self=GetOwnerSkeletalMeshComponent())
   Is Valid(exec=@n2.then_8, InputObject=@n11)
   > Set isCustomEmotePlaying(isCustomEmotePlaying:bool=Montage_IsPlaying(self=@n11, Montage=/BRCosmetics/Animation/Game/MainPlayer/Emotes/BalletSpin/Emote_BalletSpin_CMF_M)) -> (Output_Get:bool)
   n12 = Get bIsMoving2D()
   n13 = Branch(execute=@n2.then_0, Condition=AND Boolean(A=NOT Boolean(A=@n12), B=Get bIsCrouching()))
   > Set Alpha_CrouchingIdle(Alpha_CrouchingIdle:double=1) -> (Output_Get:double)
   Set Alpha_CrouchingIdle(execute=@n13.else, Alpha_CrouchingIdle=0)
   n14 = Get bIsInFrontEnd()
   n15 = GetOwnerSkeletalMeshComponent()
   n16 = BreakRotator(InRot=BreakTransform(InTransform=GetDeltaTransformFromRefPose(self=@n15, BoneName=spine_05, BaseName=pelvis)).Rotation)
   n17 = Branch(
    execute=@n2.then_9
    Condition=AND Boolean!(A="float > float"(A=@n16, B=16), B=Get JadeTowelHopeBP() -> (JadeTowelHopeBP:bool), C:bool=NOT Boolean(A=@n14)) -> (ReturnValue)
   )
   > Set Alpha_CollisionHairAndArm L(Alpha_CollisionHairAndArm L:double=1) -> (Output_Get:double)
   Set Alpha_CollisionHairAndArm L(execute=@n17.else, Alpha_CollisionHairAndArm L=0)
   n18 = PlayerMovement(execute=@n2.then_4, AnimInstance=Self-Reference(), XDotProductCoefficient=100, YDotProductCoefficient=100, ZDotProductCoefficient=100)
   n19 = BreakVector(InVec=MakeVector(X=@n18, Y=@n18.YMovement))
   n20 = "float < float"(A=@n19, B=-50)
   n21 = "float > float"(A=@n19, B=50)
   > Set Sides(Sides:bool=OR Boolean(A=@n20, B=@n21)) -> (Output_Get:bool)
   n22 = "float > float"(A=@n19.Y, B=40)
   n23 = "float < float"(A=@n19.Y, B=-40)
   > Set FrontOrBack(FrontOrBack:bool=OR Boolean(A=@n22, B=@n23)) -> (Output_Get:bool)
   > Set CrouchingTargeting(CrouchingTargeting:bool=AND Boolean!(A=Get Crouching Bool(), B=Get Targeting() -> (Targeting:bool)) -> (ReturnValue)) -> (Output_Get:bool)
   > Set Left(Left:double="To Float (Boolean)"(InBool=@n20)) -> (Output_Get:double)
   n24 = NOT Boolean(A=OR Boolean(A=Get bIsJumping(), B=Get bIsFalling()))
   > Set Right(Right:double="To Float (Boolean)"(InBool=AND Boolean(A=@n21, B=@n24))) -> (Output_Get:double)
   > Set Forward(Forward:double="To Float (Boolean)"(InBool=@n22)) -> (Output_Get:double)
   > Set Backward(Backward:double="To Float (Boolean)"(InBool=@n23)) -> (Output_Get:double)
   n25 = Get bIsKneeSliding()
   n26 = Get bIsSliding()
   n27 = Branch(execute=@n2.then_5, Condition=AND Boolean!(A=NOT Boolean(A=@n25), B=NOT Boolean(A=@n26), C:bool=Get bIsCrouching()) -> (ReturnValue))
   > Set Crouching Bool(Crouching Bool:bool=true) -> (Output_Get:bool)
   Delay(execute=@n27.else, Duration=0.2)
   > Set Crouching Bool(Crouching Bool=false)
   n28 = GetOwnerSkeletalMeshComponent()
   n29 = BreakRotator(InRot=BreakTransform(InTransform=GetDeltaTransformFromRefPose(self=@n28, BoneName=hand_r, BaseName=lowerarm_r)).Rotation)
   n30 = Branch(execute=@n2.then_7, Condition="float > float"(A=@n29.Pitch, B=31))
   > Set Detected_RotateHandWithWeaponsLow(Detected_RotateHandWithWeaponsLow:double=1) -> (Output_Get:double)
   Set Detected_RotateHandWithWeaponsLow(execute=@n30.else, Detected_RotateHandWithWeaponsLow=0)
   n31 = AND Boolean!(A=Get bIsMoving2D(), B=Get bIsCrouching()) -> (ReturnValue)
   Set Alpha_CrouchingMoving2D(
    execute=@n2.then_1
    Alpha_CrouchingMoving2D:double=FInterpTo(
     Current=Get Alpha_CrouchingMoving2D()
     Target="To Float (Boolean)"(InBool=@n31)
     DeltaTime=Get Alpha_DeltaTimeX() -> (Alpha_DeltaTimeX:double)
     InterpSpeed=10
    )
   ) -> (Output_Get:double)
   n32 = BreakTransform(InTransform=GetDeltaTransformFromRefPose(self=GetOwnerSkeletalMeshComponent(), BoneName=hand_r, BaseName=lowerarm_r))
   n33 = Branch(execute=@n2.then_6, Condition="float < float"(A=BreakRotator(InRot=@n32.Rotation).Pitch, B=-12))
   > Set Detected_RotateHandWithWeapons(Detected_RotateHandWithWeapons:double=1) -> (Output_Get:double)
   Set Detected_RotateHandWithWeapons(execute=@n33.else, Detected_RotateHandWithWeapons=0)
   n34 = Branch(execute=@n2.then_2, Condition=Get bIsCrouching())
   > Set Alpha_Crouching(Alpha_Crouching:double=1) -> (Output_Get:double)
   Set Alpha_Crouching(execute=@n34.else, Alpha_Crouching=0)
   Event OnBeginPlayingEmote() {is=title K2Node_Event}
   > Set Alpha_PlayingEmote(Alpha_PlayingEmote:double=1) -> (Output_Get:double)
   Event OnEndPlayingEmote() {is=title K2Node_Event}
   > Set Alpha_PlayingEmote(Alpha_PlayingEmote=0)
   Event OnBeginSkydiving()
   > Set SkyDiving(SkyDiving:double=0) -> (Output_Get:double)
   > Set Skydive Amount(Skydive Amount:double=0.5) -> (Output_Get:double)
   Event OnEndSkydiving()
   > Set SkyDiving(SkyDiving=1)
   > Set Skydive Amount(Skydive Amount=1)
   Event OnBeginTargeting() {is=title K2Node_Event}
   > Set Targeting(Targeting:bool=true) -> (Output_Get:bool)
   Event OnEndTargeting() {is=title K2Node_Event}
   > Set Targeting(Targeting=false)
   PrintString!(execute, InString="To String (Float)"(InDouble=@n29.Pitch), bPrintToScreen, bPrintToLog, TextColor, Duration=0.2, Key) -> (then)
  ]}
]
variables [
 Alpha_CollisionHairAndArm L {type=number}
 Alpha_Crouching {type=number}
 Alpha_CrouchingIdle {type=number}
 Alpha_CrouchingMoving2D {type=number}
 Alpha_DeltaTimeX {type=number}
 Alpha_Melee {type=number}
 Alpha_PlayingEmote {type=number}
 backpackSocket_rotate {type=Rotator}
 Backward {type=number}
 Crouching Bool {type=boolean}
 CrouchingTargeting {type=boolean}
 Detected_RotateHandWithWeapons {type=number}
 Detected_RotateHandWithWeaponsLow {type=number}
 Dyn_EmoteGoldenGlory {type=number}
 EmoteGoldenGlory {type=number}
 Forward {type=number}
 FrontOrBack {type=boolean}
 isCustomEmotePlaying {type=boolean}
 JadeTowelHopeBP {type=boolean}
 Left {type=number}
 NewVar {type=Rotator}
 Right {type=number}
 Sides {type=boolean}
 Skydive Amount {type=number}
 SkyDiving {type=number}
 Targeting {type=boolean}
]
