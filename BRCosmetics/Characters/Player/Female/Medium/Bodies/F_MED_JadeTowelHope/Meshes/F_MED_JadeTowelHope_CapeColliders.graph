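# Physics asset of capsule colliders for the F_MED_JadeTowelHope female body, named as the collision set for its cape.
# Single capsules cover the pelvis, two lower spine bones, arms, hands, legs and feet, with no head or neck bodies.
# Five back and side dyn_skirt_*_03 bones also carry capsules, so the skirt contributes collision volume.
# Constraints link limbs and spine in a pelvis-rooted tree, with upper arms joined to spine_02 and skirt bodies directly to the pelvis.
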PhysicsAsset /BRCosmetics/Characters/Player/Female/Medium/Bodies/F_MED_JadeTowelHope/Meshes/F_MED_JadeTowelHope_CapeColliders
bodies {
 calf_l [{shapeType=Capsule, length=28.6339, radius=11.064, center=<-27.6471 -0.0714 -0.9885>, rotation=<rot 0 79.183 -88.6776>}]
 calf_r [{shapeType=Capsule, length=28.6339, radius=11.064, center=<27.6471 0.0714 0.9885>, rotation=<rot 0 79.183 -88.6776>}]
 dyn_skirt_bk_c_03 [{shapeType=Capsule, length=9.8476, radius=5.5605, shapeName=dyn_skirt_bk_c_03_capsule, center=<2.3777 -0.3266 0.5849>, rotation=<rot 0 0 90>}]
 dyn_skirt_bk_l_03 [{
   shapeType=Capsule
   length=11.5371
   radius=5
   shapeName=dyn_skirt_bk_l_03_capsule
   center=<2.2624 0.8633 -1.5168>
   rotation=<rot 5.38 174.1938 -81.0107>
  }]
 dyn_skirt_bk_r_03 [{
   shapeType=Capsule
   length=11.5371
   radius=5
   shapeName=dyn_skirt_bk_l_03_capsule
   center=<2.2521 1.0646 -1.5168>
   rotation=<rot 5.38 -168.3666 81.0107>
  }]
 dyn_skirt_sd_l_03 [{shapeType=Capsule, length=10.8167, radius=5.3212, shapeName=dyn_skirt_sd_l_03_capsule, center=<2.474 0 -2.1624>, rotation=<rot 0 0 90>}]
 dyn_skirt_sd_r_03 [{shapeType=Capsule, length=10.7792, radius=4.8214, shapeName=dyn_skirt_sd_r_03_capsule, center=<2.474 0 -2.1624>, rotation=<rot 0 0 90>}]
 foot_l [{shapeType=Capsule, length=13.0383, radius=8.0176, shapeName=foot_l_capsule, center=<-3.92 5.601 0>, rotation=<rot 0 6.7594 90>}]
 foot_r [{shapeType=Capsule, length=13.0383, radius=8.0176, shapeName=foot_l_capsule, center=<3.92 -5.601 0>, rotation=<rot 0 6.7594 -90>}]
 hand_l [{shapeType=Capsule, length=8.7464, radius=5.6478, center=<8.2519 -2.6605 -0.069>, rotation=<rot 0 81.071 -71.8704>}]
 hand_r [{shapeType=Capsule, length=8.7464, radius=5.6478, center=<-8.2519 2.6605 0.069>, rotation=<rot 0 81.071 -71.8704>}]
 lowerarm_l [{shapeType=Capsule, length=14.2095, radius=7.8957, center=<12.2213 0.1704 -1.4491>, rotation=<rot 0 -91.2928 -87.4337>}]
 lowerarm_r [{shapeType=Capsule, length=14.2095, radius=7.8957, center=<-12.2213 -0.1704 1.4491>, rotation=<rot 0 -91.2928 -87.4337>}]
 pelvis [{shapeType=Capsule, length=17.2775, radius=18.4014, center=<-13.3639 0.1721 0.2841>, rotation=<rot 10.9044 -90 -1.326>}]
 spine_01 [{shapeType=Capsule, length=14, radius=9, shapeName=spine_01_capsule, center=<-2.2945 -4.7527 0>}]
 spine_02 [{shapeType=Capsule, length=16.2478, radius=11.2157, shapeName=spine_02_capsule, center=<13.119 0 0>, rotation=<rot 90 0 0>}]
 thigh_l [{shapeType=Capsule, length=20.3977, radius=14.7105, center=<-27.582 -5.5485 1.5122>, rotation=<rot 0 -90.1171 93.1475>}]
 thigh_r [{shapeType=Capsule, length=20.3977, radius=14.7105, center=<27.582 5.5485 -1.5122>, rotation=<rot 0 -90.1171 -86.8525>}]
 upperarm_l [{shapeType=Capsule, length=13.6971, radius=5.8414, center=<13.0072 0.2824 -0.9398>, rotation=<rot 0 88.3156 -84.273>}]
 upperarm_r [{shapeType=Capsule, length=13.6967, radius=5.8411, center=<-13.0035 -0.2709 0.94>, rotation=<rot 0 88.5178 -84.275>}]
}
constraints [
 {bone1Name=upperarm_l, bone2Name=spine_02, swing1LimitDegrees=45, swing2LimitDegrees=45, twistLimitDegrees=45}
 {bone1Name=lowerarm_l, bone2Name=upperarm_l, swing1LimitDegrees=45, swing2LimitDegrees=45, twistLimitDegrees=45}
 {bone1Name=hand_l, bone2Name=lowerarm_l, swing1LimitDegrees=45, swing2LimitDegrees=45, twistLimitDegrees=45}
 {bone1Name=upperarm_r, bone2Name=spine_02, swing1LimitDegrees=45, swing2LimitDegrees=45, twistLimitDegrees=45}
 {bone1Name=lowerarm_r, bone2Name=upperarm_r, swing1LimitDegrees=45, swing2LimitDegrees=45, twistLimitDegrees=45}
 {bone1Name=hand_r, bone2Name=lowerarm_r, swing1LimitDegrees=45, swing2LimitDegrees=45, twistLimitDegrees=45}
 {bone1Name=calf_l, bone2Name=thigh_l, swing1LimitDegrees=45, swing2LimitDegrees=45, twistLimitDegrees=45}
 {bone1Name=calf_r, bone2Name=thigh_r, swing1LimitDegrees=45, swing2LimitDegrees=45, twistLimitDegrees=45}
 {
  bone1Name=foot_l
  bone2Name=calf_l
  swing1LimitDegrees=45
  swing1Motion=Free
  swing2LimitDegrees=45
  swing2Motion=Free
  twistLimitDegrees=45
  twistMotion=Free
 }
 {
  bone1Name=foot_r
  bone2Name=calf_r
  swing1LimitDegrees=45
  swing1Motion=Free
  swing2LimitDegrees=45
  swing2Motion=Free
  twistLimitDegrees=45
  twistMotion=Free
 }
 {
  bone1Name=spine_02
  bone2Name=spine_01
  swing1LimitDegrees=45
  swing1Motion=Free
  swing2LimitDegrees=45
  swing2Motion=Free
  twistLimitDegrees=45
  twistMotion=Free
 }
 {
  bone1Name=spine_01
  bone2Name=pelvis
  swing1LimitDegrees=45
  swing1Motion=Free
  swing2LimitDegrees=45
  swing2Motion=Free
  twistLimitDegrees=45
  twistMotion=Free
 }
 {
  bone1Name=dyn_skirt_bk_c_03
  bone2Name=pelvis
  swing1LimitDegrees=45
  swing1Motion=Free
  swing2LimitDegrees=45
  swing2Motion=Free
  twistLimitDegrees=45
  twistMotion=Free
 }
 {
  bone1Name=dyn_skirt_bk_l_03
  bone2Name=pelvis
  swing1LimitDegrees=45
  swing1Motion=Free
  swing2LimitDegrees=45
  swing2Motion=Free
  twistLimitDegrees=45
  twistMotion=Free
 }
 {
  bone1Name=dyn_skirt_bk_r_03
  bone2Name=pelvis
  swing1LimitDegrees=45
  swing1Motion=Free
  swing2LimitDegrees=45
  swing2Motion=Free
  twistLimitDegrees=45
  twistMotion=Free
 }
 {
  bone1Name=thigh_l
  bone2Name=pelvis
  swing1LimitDegrees=45
  swing1Motion=Free
  swing2LimitDegrees=45
  swing2Motion=Free
  twistLimitDegrees=45
  twistMotion=Free
 }
 {
  bone1Name=thigh_r
  bone2Name=pelvis
  swing1LimitDegrees=45
  swing1Motion=Free
  swing2LimitDegrees=45
  swing2Motion=Free
  twistLimitDegrees=45
  twistMotion=Free
 }
 {
  bone1Name=dyn_skirt_sd_r_03
  bone2Name=pelvis
  swing1LimitDegrees=45
  swing1Motion=Free
  swing2LimitDegrees=45
  swing2Motion=Free
  twistLimitDegrees=45
  twistMotion=Free
 }
 {
  bone1Name=dyn_skirt_sd_l_03
  bone2Name=pelvis
  swing1LimitDegrees=45
  swing1Motion=Free
  swing2LimitDegrees=45
  swing2Motion=Free
  twistLimitDegrees=45
  twistMotion=Free
 }
]
