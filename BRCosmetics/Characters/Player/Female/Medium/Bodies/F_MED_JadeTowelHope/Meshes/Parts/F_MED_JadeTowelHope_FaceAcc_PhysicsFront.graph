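# Physics asset for the front hair bones of the F_MED_JadeTowelHope female character's face accessory part.
# Capsule bodies cover the head, three center front hair bones and two left side hair bones.
# Front bones 02 and 03 both constrain directly to front 01 rather than in series, while the side bones form a two-link chain; both groups root on the head.
# Every constraint locks twist and sets narrow swing limits.
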PhysicsAsset /BRCosmetics/Characters/Player/Female/Medium/Bodies/F_MED_JadeTowelHope/Meshes/Parts/F_MED_JadeTowelHope_FaceAcc_PhysicsFront
bodies {
 dyn_hair_fr_a_01_c [{shapeType=Capsule, length=3, radius=1, shapeName=dyn_hair_fr_a_01_c_capsule, rotation=<rot 90 0 0>}]
 dyn_hair_fr_a_02_c [{shapeType=Capsule, length=3, radius=1, shapeName=dyn_hair_fr_a_02_c_capsule, rotation=<rot 90 0 0>}]
 dyn_hair_fr_a_03_c [{shapeType=Capsule, length=3, radius=1, shapeName=dyn_hair_fr_a_03_c_capsule, rotation=<rot 90 0 0>}]
 dyn_hair_sd_a_01_l [{
   shapeType=Capsule
   length=1.0345
   radius=0.592
   shapeName=dyn_hair_sd_a_01_l_capsule
   center=<1.8141 0.2931 -0.2581>
   rotation=<rot 0 85.2086 -65.802>
  }]
 dyn_hair_sd_a_02_l [{
   shapeType=Capsule
   length=1.5073
   radius=0.6092
   shapeName=dyn_hair_sd_a_02_l_capsule
   center=<1.0972 -0.8026 0.021>
   rotation=<rot 0 51.3844 -88.5535>
  }]
 head [{shapeType=Capsule, length=6, radius=10, shapeName=head_capsule, rotation=<rot 90 0 0>}]
}
constraints [
 {
  bone1Name=dyn_hair_fr_a_03_c
  bone2Name=dyn_hair_fr_a_01_c
  swing1LimitDegrees=12
  swing2LimitDegrees=10
  twistLimitDegrees=45
  twistMotion=Locked
 }
 {
  bone1Name=dyn_hair_fr_a_02_c
  bone2Name=dyn_hair_fr_a_01_c
  swing1LimitDegrees=12
  swing2LimitDegrees=10
  twistLimitDegrees=45
  twistMotion=Locked
 }
 {bone1Name=dyn_hair_fr_a_01_c, bone2Name=head, swing1LimitDegrees=12, swing2LimitDegrees=10, twistLimitDegrees=45, twistMotion=Locked}
 {bone1Name=dyn_hair_sd_a_01_l, bone2Name=head, swing1LimitDegrees=10, swing2LimitDegrees=10, twistLimitDegrees=45, twistMotion=Locked}
 {
  bone1Name=dyn_hair_sd_a_02_l
  bone2Name=dyn_hair_sd_a_01_l
  swing1LimitDegrees=12
  swing2LimitDegrees=10
  twistLimitDegrees=45
  twistMotion=Locked
 }
]
