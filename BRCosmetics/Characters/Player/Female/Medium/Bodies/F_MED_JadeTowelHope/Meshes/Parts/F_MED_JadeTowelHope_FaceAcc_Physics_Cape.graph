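# Physics asset for the JadeTowelHope female character's face accessory, Cape variant, giving capsule physics to left and right ponytails.
# Each side chains six dynamic bones from a head-attached base, with three strands branching off the last, using tight swing limits and locked twist.
# Head, spine, pelvis, arm and calf capsules form the body volume, joined by free-rotation constraints, with large torso capsules on the upper spine.
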PhysicsAsset /BRCosmetics/Characters/Player/Female/Medium/Bodies/F_MED_JadeTowelHope/Meshes/Parts/F_MED_JadeTowelHope_FaceAcc_Physics_Cape
bodies {
 base_ponytail_l [{shapeType=Capsule, length=1, radius=1, shapeName=base_ponytail_l_capsule, rotation=<rot 0 -90 0>}]
 base_ponytail_r [{shapeType=Capsule, length=1, radius=1, shapeName=base_ponytail_r_capsule, rotation=<rot 0 -90 0>}]
 calf_l [{shapeType=Capsule, length=26, radius=12, shapeName=calf_l_capsule, center=<-14.7061 -0.5088 -0.1658>, rotation=<rot 90 80.7734 84.0177>}]
 calf_r [{shapeType=Capsule, length=26, radius=12, shapeName=calf_l_capsule, center=<14.7061 0.5088 -0.1658>, rotation=<rot 90 80.7734 -95.9823>}]
 dyn_ponytail_01_l [{
   shapeType=Capsule
   length=8
   radius=4
   shapeName=dyn_ponytail_01_l_capsule
   center=<2.5267 -2.4196 -0.0358>
   rotation=<rot 77.7444 136.567 180>
  }]
 dyn_ponytail_01_r [{shapeType=Capsule, length=8, radius=4, shapeName=dyn_ponytail_01_l_capsule, center=<-2.67 2.4434 0.0215>, rotation=<rot 77.7444 136.567 180>}]
 dyn_ponytail_02_l [{
   shapeType=Capsule
   length=8.7127
   radius=7.365
   shapeName=dyn_ponytail_02_l_capsule
   center=<4.744 -5.5916 -0.0702>
   rotation=<rot 90 172.875 -169.361>
  }]
 dyn_ponytail_02_r [{
   shapeType=Capsule
   length=8.7127
   radius=4.865
   shapeName=dyn_ponytail_02_l_capsule
   center=<-4.3141 5.3334 0.0537>
   rotation=<rot 90 5.7106 16.3394>
  }]
 dyn_ponytail_03_l [{shapeType=Capsule, length=7, radius=7.6573, shapeName=dyn_ponytail_03_l_capsule, center=<3 0 -1.2104>, rotation=<rot 90 0 0>}]
 dyn_ponytail_03_r [{shapeType=Capsule, length=9.5, radius=7.6573, shapeName=dyn_ponytail_03_l_capsule, center=<-3 0 2.2935>, rotation=<rot 90 0 0>}]
 dyn_ponytail_04_l [{shapeType=Capsule, length=15, radius=10, shapeName=dyn_ponytail_04_l_capsule, center=<3 3.5676 0>, rotation=<rot 60.5732 -179.9999 -180>}]
 dyn_ponytail_04_r [{shapeType=Capsule, length=15, radius=10, shapeName=dyn_ponytail_04_l_capsule, center=<-3 -3.5676 0>, rotation=<rot 60.5732 -179.9999 -180>}]
 dyn_ponytail_05_l [{
   shapeType=Capsule
   length=7.5378
   radius=11.0825
   shapeName=dyn_ponytail_05_l_capsule
   center=<2.4492 2.1712 6.4257>
   rotation=<rot 90 180 -151.1024>
  }]
 dyn_ponytail_05_r [{
   shapeType=Capsule
   length=5.2639
   radius=11.8564
   shapeName=dyn_ponytail_05_l_capsule
   center=<-3 -3.4582 -6.0487>
   rotation=<rot 90 180 -136.5841>
  }]
 dyn_ponytail_06_l [{shapeType=Capsule, length=14.5, radius=19.5, shapeName=dyn_ponytail_06_l_capsule, center=<10.6296 -2.6248 0>, rotation=<rot 90 0 0>}]
 dyn_ponytail_06_r [{shapeType=Capsule, length=12, radius=14, shapeName=dyn_ponytail_06_l_capsule, center=<-3.9559 1.0305 3.4703>, rotation=<rot 90 0 0>}]
 dyn_ponytail_strand_a_l [{shapeType=Capsule, length=4.7255, radius=2.5717, shapeName=dyn_ponytail_strand_a_l_capsule, center=<2.8488 0 0>}]
 dyn_ponytail_strand_a_r [{
   shapeType=Capsule
   length=8.2255
   radius=2.5717
   shapeName=dyn_ponytail_strand_a_l_capsule
   center=<-4.2199 -0.2 6.8>
   rotation=<rot 45.3639 0.0107 174.3447>
  }]
 dyn_ponytail_strand_b_l [{shapeType=Capsule, length=8.0737, radius=5.4978, shapeName=dyn_ponytail_strand_b_l_capsule, center=<4.3739 0 -1.9>}]
 dyn_ponytail_strand_b_r [{
   shapeType=Capsule
   length=16.0737
   radius=5.4978
   shapeName=dyn_ponytail_strand_b_l_capsule
   center=<-7.6426 0 11.6>
   rotation=<rot 67.6583 0.0339 -179.9634>
  }]
 dyn_ponytail_strand_c_l [{
   shapeType=Capsule
   length=11.0205
   radius=4.3901
   shapeName=dyn_ponytail_strand_c_r_capsule
   center=<3.9785 -2.3504 -6.5219>
   rotation=<rot 33.472 -90 155.9223>
  }]
 dyn_ponytail_strand_c_r [{
   shapeType=Capsule
   length=11.0205
   radius=4.3901
   shapeName=dyn_ponytail_strand_c_r_capsule
   center=<-3.9785 2.3504 6.5219>
   rotation=<rot 33.472 -90 -24.0777>
  }]
 head [{shapeType=Capsule, length=2, radius=2, shapeName=head_capsule}]
 lowerarm_l [{shapeType=Capsule, length=35.51, radius=6, shapeName=lowerarm_l_capsule, center=<18.2404 0 0>, rotation=<rot 90 0 0>}]
 lowerarm_r [{shapeType=Capsule, length=35.51, radius=6, shapeName=lowerarm_l_capsule, center=<-18.2404 0 0>, rotation=<rot 90 0 180>}]
 pelvis [{shapeType=Capsule, length=37.5, radius=19, shapeName=pelvis_capsule, center=<0 5.0491 0>}]
 spine_04 [
  {shapeType=Capsule, length=22.5, radius=18.5, shapeName=spine_04_capsule, center=<0 8.7048 0>}
  {shapeType=Capsule, length=24, radius=15, shapeName=spine_04_capsule, center=<16.2163 16.0656 -5.9285>}
 ]
 spine_05 [{shapeType=Capsule, length=76, radius=16.5, shapeName=spine_05_capsule, center=<-22.0121 -0.3249 0>, rotation=<rot 90 180 176.9954>}]
 upperarm_l [{shapeType=Capsule, length=19.005, radius=7, shapeName=upperarm_l_capsule, center=<12.4116 0 0>, rotation=<rot 90 0 0>}]
 upperarm_r [{shapeType=Capsule, length=19.005, radius=7, shapeName=upperarm_l_capsule, center=<-12.4116 0 0>, rotation=<rot 90 0 180>}]
}
constraints [
 {
  bone1Name=dyn_ponytail_06_l
  bone2Name=dyn_ponytail_05_l
  swing1LimitDegrees=20
  swing2LimitDegrees=25
  twistLimitDegrees=45
  twistMotion=Locked
 }
 {
  bone1Name=dyn_ponytail_05_l
  bone2Name=dyn_ponytail_04_l
  swing1LimitDegrees=20
  swing2LimitDegrees=25
  twistLimitDegrees=45
  twistMotion=Locked
 }
 {
  bone1Name=dyn_ponytail_04_l
  bone2Name=dyn_ponytail_03_l
  swing1LimitDegrees=15
  swing2LimitDegrees=25
  twistLimitDegrees=45
  twistMotion=Locked
 }
 {
  bone1Name=dyn_ponytail_03_l
  bone2Name=dyn_ponytail_02_l
  swing1LimitDegrees=15
  swing2LimitDegrees=15
  twistLimitDegrees=45
  twistMotion=Locked
 }
 {
  bone1Name=dyn_ponytail_02_l
  bone2Name=dyn_ponytail_01_l
  swing1LimitDegrees=15
  swing2LimitDegrees=20
  twistLimitDegrees=45
  twistMotion=Locked
 }
 {
  bone1Name=dyn_ponytail_strand_a_l
  bone2Name=dyn_ponytail_06_l
  swing1LimitDegrees=20
  swing2LimitDegrees=20
  twistLimitDegrees=45
  twistMotion=Locked
 }
 {
  bone1Name=dyn_ponytail_strand_b_l
  bone2Name=dyn_ponytail_06_l
  swing1LimitDegrees=20
  swing2LimitDegrees=20
  twistLimitDegrees=45
  twistMotion=Locked
 }
 {
  bone1Name=dyn_ponytail_01_l
  bone2Name=base_ponytail_l
  swing1LimitDegrees=15
  swing2LimitDegrees=20
  twistLimitDegrees=45
  twistMotion=Locked
 }
 {
  bone1Name=dyn_ponytail_02_r
  bone2Name=dyn_ponytail_01_r
  swing1LimitDegrees=15
  swing2LimitDegrees=15
  twistLimitDegrees=45
  twistMotion=Locked
 }
 {
  bone1Name=dyn_ponytail_03_r
  bone2Name=dyn_ponytail_02_r
  swing1LimitDegrees=15
  swing2LimitDegrees=15
  twistLimitDegrees=45
  twistMotion=Locked
 }
 {
  bone1Name=dyn_ponytail_05_r
  bone2Name=dyn_ponytail_04_r
  swing1LimitDegrees=20
  swing2LimitDegrees=15
  twistLimitDegrees=45
  twistMotion=Locked
 }
 {
  bone1Name=dyn_ponytail_strand_b_r
  bone2Name=dyn_ponytail_06_r
  swing1LimitDegrees=10
  swing2LimitDegrees=10
  twistLimitDegrees=45
  twistMotion=Locked
 }
 {
  bone1Name=dyn_ponytail_strand_a_r
  bone2Name=dyn_ponytail_06_r
  swing1LimitDegrees=10
  swing2LimitDegrees=10
  twistLimitDegrees=45
  twistMotion=Locked
 }
 {
  bone1Name=dyn_ponytail_06_r
  bone2Name=dyn_ponytail_05_r
  swing1LimitDegrees=20
  swing2LimitDegrees=15
  twistLimitDegrees=45
  twistMotion=Locked
 }
 {
  bone1Name=head
  bone2Name=spine_05
  swing1LimitDegrees=45
  swing1Motion=Free
  swing2LimitDegrees=45
  swing2Motion=Free
  twistLimitDegrees=45
  twistMotion=Free
 }
 {
  bone1Name=spine_04
  bone2Name=pelvis
  swing1LimitDegrees=45
  swing1Motion=Free
  swing2LimitDegrees=45
  swing2Motion=Free
  twistLimitDegrees=45
  twistMotion=Free
 }
 {
  bone1Name=upperarm_l
  bone2Name=spine_05
  swing1LimitDegrees=45
  swing1Motion=Free
  swing2LimitDegrees=45
  swing2Motion=Free
  twistLimitDegrees=45
  twistMotion=Free
 }
 {
  bone1Name=dyn_ponytail_04_r
  bone2Name=dyn_ponytail_03_r
  swing1LimitDegrees=20
  swing2LimitDegrees=25
  twistLimitDegrees=45
  twistMotion=Locked
 }
 {
  bone1Name=upperarm_r
  bone2Name=spine_05
  swing1LimitDegrees=45
  swing1Motion=Free
  swing2LimitDegrees=45
  swing2Motion=Free
  twistLimitDegrees=45
  twistMotion=Free
 }
 {
  bone1Name=lowerarm_r
  bone2Name=upperarm_r
  swing1LimitDegrees=45
  swing1Motion=Free
  swing2LimitDegrees=45
  swing2Motion=Free
  twistLimitDegrees=45
  twistMotion=Free
 }
 {
  bone1Name=base_ponytail_l
  bone2Name=head
  swing1LimitDegrees=45
  swing1Motion=Free
  swing2LimitDegrees=45
  swing2Motion=Free
  twistLimitDegrees=45
  twistMotion=Free
 }
 {
  bone1Name=base_ponytail_r
  bone2Name=head
  swing1LimitDegrees=45
  swing1Motion=Free
  swing2LimitDegrees=45
  swing2Motion=Free
  twistLimitDegrees=45
  twistMotion=Free
 }
 {
  bone1Name=dyn_ponytail_01_r
  bone2Name=base_ponytail_r
  swing1LimitDegrees=25
  swing2LimitDegrees=10
  twistLimitDegrees=45
  twistMotion=Locked
 }
 {
  bone1Name=lowerarm_l
  bone2Name=upperarm_l
  swing1LimitDegrees=45
  swing1Motion=Free
  swing2LimitDegrees=45
  swing2Motion=Free
  twistLimitDegrees=45
  twistMotion=Free
 }
 {
  bone1Name=calf_l
  bone2Name=pelvis
  swing1LimitDegrees=45
  swing1Motion=Free
  swing2LimitDegrees=45
  swing2Motion=Free
  twistLimitDegrees=45
  twistMotion=Free
 }
 {
  bone1Name=calf_r
  bone2Name=pelvis
  swing1LimitDegrees=45
  swing1Motion=Free
  swing2LimitDegrees=45
  swing2Motion=Free
  twistLimitDegrees=45
  twistMotion=Free
 }
 {
  bone1Name=dyn_ponytail_strand_c_r
  bone2Name=dyn_ponytail_06_r
  swing1LimitDegrees=20
  swing2LimitDegrees=20
  twistLimitDegrees=45
  twistMotion=Locked
 }
 {
  bone1Name=dyn_ponytail_strand_c_l
  bone2Name=dyn_ponytail_06_l
  swing1LimitDegrees=20
  swing2LimitDegrees=20
  twistLimitDegrees=45
  twistMotion=Locked
 }
 {
  bone1Name=spine_05
  bone2Name=spine_04
  swing1LimitDegrees=45
  swing1Motion=Free
  swing2LimitDegrees=45
  swing2Motion=Free
  twistLimitDegrees=45
  twistMotion=Free
 }
]
