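# Post-process AnimBP for the F_MED_JadeTowelHope face accessory mesh, adding secondary motion to hair, ponytails, ears, glasses and mic.
# bRigidBodyEnabled selects between an AnimDynamics chain over a cached rigid-body pose and a constraint-driven ponytail pose.
# Ponytail bones are constrained to neck, arm and spine, then corrected with additive offsets.
# RigidBody nodes use a backpack-specific physics asset, with alphas for skydiving, including a gravity override.
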AnimBlueprint /BRCosmetics/Characters/Player/Female/Medium/Bodies/F_MED_JadeTowelHope/Meshes/Parts/F_MED_JadeTowelHope_FaceAcc_AnimBP
target_skeleton=&/F_MED_JadeTowelHope_FaceAcc_Skeleton
graphs [
 AnimGraph {nodes [
   Component To Local()
   n1 = Anim Dynamics - Bone: dyn_hair_sd_b_01_l(ComponentPose=Local To Component(LocalPose=Use cached pose 'RigidBodyCache'()), Alpha=1, ExternalForce="0,0,0") {
    angularDampingOverride=1
    angularSpringConstant=20000
    bAngularSpring=true
    bOverrideAngularDamping=true
    bUseGravityOverride=true
    gravityScale=0.35
    simulationSpace=RootRelative
    alphaBoolBlend {blendOption=Linear}
    boundBone {boneName=dyn_hair_sd_b_01_l}
    componentAppliedLinearAccClamp=<3000 3000 3000>
    componentLinearAccScale=<1 1 1>
    componentLinearVelScale=<1 1 1>
    gravityOverride {z=-980}
    physicsBodyDefinitions [{
      boundBone {boneName=dyn_hair_sd_b_01_l}
      boxExtents=<20 30 30>
      constraintSetup {angularLimitsMax=<2 0 25>, angularLimitsMin=<-2 -15 0>, angularTarget=<1 0 0>}
      localJointOffset=<4 0 0>
     }]
   }
   n2 = Anim Dynamics - Bone: dyn_hair_sd_b_02_l(ComponentPose=@n1, Alpha=1, ExternalForce="0,0,0") {
    angularDampingOverride=1
    angularSpringConstant=20000
    bAngularSpring=true
    bOverrideAngularDamping=true
    bUseGravityOverride=true
    gravityScale=0.35
    simulationSpace=RootRelative
    alphaBoolBlend {blendOption=Linear}
    boundBone {boneName=dyn_hair_sd_b_02_l}
    componentAppliedLinearAccClamp=<3000 3000 3000>
    componentLinearAccScale=<1 1 1>
    componentLinearVelScale=<1 1 1>
    gravityOverride {z=-980}
    physicsBodyDefinitions [{
      boundBone {boneName=dyn_hair_sd_b_02_l}
      boxExtents=<20 30 30>
      constraintSetup {angularLimitsMax=<2 0 20>, angularLimitsMin=<-2 -20 0>, angularTarget=<1 0 0>}
      localJointOffset=<4 0 0>
     }]
   }
   n3 = Anim Dynamics - Bone: dyn_hair_sd_b_03_l(ComponentPose=@n2, Alpha=1, ExternalForce="0,0,0") {
    angularDampingOverride=1
    angularSpringConstant=20000
    bAngularSpring=true
    bOverrideAngularDamping=true
    bUseGravityOverride=true
    gravityScale=0.35
    simulationSpace=RootRelative
    alphaBoolBlend {blendOption=Linear}
    boundBone {boneName=dyn_hair_sd_b_03_l}
    componentAppliedLinearAccClamp=<3000 3000 3000>
    componentLinearAccScale=<1 1 1>
    componentLinearVelScale=<1 1 1>
    gravityOverride {z=-980}
    physicsBodyDefinitions [{
      boundBone {boneName=dyn_hair_sd_b_03_l}
      boxExtents=<20 30 30>
      constraintSetup {angularLimitsMax=<2 0 20>, angularLimitsMin=<-2 -20 0>, angularTarget=<1 0 0>}
      localJointOffset=<4 0 0>
     }]
   }
   n4 = Anim Dynamics - Bone: dyn_hair_sd_b_04_l(ComponentPose=@n3, Alpha=1, ExternalForce="0,0,0") {
    angularDampingOverride=1
    angularSpringConstant=20000
    bAngularSpring=true
    bOverrideAngularDamping=true
    bUseGravityOverride=true
    gravityScale=0.35
    simulationSpace=RootRelative
    alphaBoolBlend {blendOption=Linear}
    boundBone {boneName=dyn_hair_sd_b_04_l}
    componentAppliedLinearAccClamp=<3000 3000 3000>
    componentLinearAccScale=<1 1 1>
    componentLinearVelScale=<1 1 1>
    gravityOverride {z=-980}
    physicsBodyDefinitions [{
      boundBone {boneName=dyn_hair_sd_b_04_l}
      boxExtents=<20 30 30>
      constraintSetup {angularLimitsMax=<2 0 20>, angularLimitsMin=<-2 -20 0>, angularTarget=<1 0 0>}
      localJointOffset=<4 0 0>
     }]
   }
   n5 = Anim Dynamics - Bone: dyn_hair_sd_a_r(ComponentPose=@n4, Alpha=1, ExternalForce="0,0,0") {
    angularDampingOverride=1
    angularSpringConstant=20000
    bAngularSpring=true
    bOverrideAngularDamping=true
    bUseGravityOverride=true
    gravityScale=0.35
    simulationSpace=RootRelative
    alphaBoolBlend {blendOption=Linear}
    boundBone {boneName=dyn_hair_sd_a_r}
    componentAppliedLinearAccClamp=<3000 3000 3000>
    componentLinearAccScale=<1 1 1>
    componentLinearVelScale=<1 1 1>
    gravityOverride {z=-980}
    physicsBodyDefinitions [{
      boundBone {boneName=dyn_hair_sd_a_r}
      boxExtents=<20 30 30>
      constraintSetup {angularLimitsMax=<10 25 30>, angularLimitsMin=<0 0 0>, angularTarget=<1 0 0>}
      localJointOffset=<4 0 0>
     }]
   }
   n6 = Anim Dynamics - Bone: dyn_hair_sd_a_01_r(ComponentPose=@n5, Alpha=1, ExternalForce="0,0,0") {
    angularDampingOverride=1
    angularSpringConstant=20000
    bAngularSpring=true
    bOverrideAngularDamping=true
    bUseGravityOverride=true
    gravityScale=0.35
    simulationSpace=RootRelative
    alphaBoolBlend {blendOption=Linear}
    boundBone {boneName=dyn_hair_sd_a_01_r}
    componentAppliedLinearAccClamp=<3000 3000 3000>
    componentLinearAccScale=<1 1 1>
    componentLinearVelScale=<1 1 1>
    gravityOverride {z=-980}
    physicsBodyDefinitions [{
      boundBone {boneName=dyn_hair_sd_a_01_r}
      boxExtents=<20 30 30>
      constraintSetup {angularLimitsMax=<10 25 30>, angularLimitsMin=<0 0 0>, angularTarget=<1 0 0>}
      localJointOffset=<4 0 0>
     }]
   }
   n7 = Anim Dynamics - Bone: dyn_hair_sd_a_02_r(ComponentPose=@n6, Alpha=1, ExternalForce="0,0,0") {
    angularDampingOverride=1
    angularSpringConstant=20000
    bAngularSpring=true
    bOverrideAngularDamping=true
    bUseGravityOverride=true
    gravityScale=0.35
    simulationSpace=RootRelative
    alphaBoolBlend {blendOption=Linear}
    boundBone {boneName=dyn_hair_sd_a_02_r}
    componentAppliedLinearAccClamp=<3000 3000 3000>
    componentLinearAccScale=<1 1 1>
    componentLinearVelScale=<1 1 1>
    gravityOverride {z=-980}
    physicsBodyDefinitions [{
      boundBone {boneName=dyn_hair_sd_a_02_r}
      boxExtents=<20 30 30>
      constraintSetup {angularLimitsMax=<10 25 30>, angularLimitsMin=<0 0 0>, angularTarget=<1 0 0>}
      localJointOffset=<4 0 0>
     }]
   }
   n8 = Anim Dynamics - Bone: dyn_hair_sd_b_01_r(ComponentPose=@n7, Alpha=1, ExternalForce="0,0,0") {
    angularDampingOverride=1
    angularSpringConstant=20000
    bAngularSpring=true
    bOverrideAngularDamping=true
    bUseGravityOverride=true
    gravityScale=0.35
    simulationSpace=RootRelative
    alphaBoolBlend {blendOption=Linear}
    boundBone {boneName=dyn_hair_sd_b_01_r}
    componentAppliedLinearAccClamp=<3000 3000 3000>
    componentLinearAccScale=<1 1 1>
    componentLinearVelScale=<1 1 1>
    gravityOverride {z=-980}
    physicsBodyDefinitions [{
      boundBone {boneName=dyn_hair_sd_b_01_r}
      boxExtents=<20 30 30>
      constraintSetup {angularLimitsMax=<15 0 35>, angularLimitsMin=<0 -10 0>, angularTarget=<1 0 0>}
      localJointOffset=<4 0 0>
     }]
   }
   n9 = Anim Dynamics - Bone: dyn_hair_sd_b_02_r(ComponentPose=@n8, Alpha=1, ExternalForce="0,0,0") {
    angularDampingOverride=1
    angularSpringConstant=20000
    bAngularSpring=true
    bOverrideAngularDamping=true
    bUseGravityOverride=true
    gravityScale=0.35
    simulationSpace=RootRelative
    alphaBoolBlend {blendOption=Linear}
    boundBone {boneName=dyn_hair_sd_b_02_r}
    componentAppliedLinearAccClamp=<3000 3000 3000>
    componentLinearAccScale=<1 1 1>
    componentLinearVelScale=<1 1 1>
    gravityOverride {z=-980}
    physicsBodyDefinitions [{
      boundBone {boneName=dyn_hair_sd_b_02_r}
      boxExtents=<20 30 30>
      constraintSetup {angularLimitsMax=<15 0 35>, angularLimitsMin=<0 -10 0>, angularTarget=<1 0 0>}
      localJointOffset=<4 0 0>
     }]
   }
   n10 = Anim Dynamics - Bone: dyn_hair_sd_b_03_r(ComponentPose=@n9, Alpha=1, ExternalForce="0,0,0") {
    angularDampingOverride=1
    angularSpringConstant=20000
    bAngularSpring=true
    bOverrideAngularDamping=true
    bUseGravityOverride=true
    gravityScale=0.35
    simulationSpace=RootRelative
    alphaBoolBlend {blendOption=Linear}
    boundBone {boneName=dyn_hair_sd_b_03_r}
    componentAppliedLinearAccClamp=<3000 3000 3000>
    componentLinearAccScale=<1 1 1>
    componentLinearVelScale=<1 1 1>
    gravityOverride {z=-980}
    physicsBodyDefinitions [{
      boundBone {boneName=dyn_hair_sd_b_03_r}
      boxExtents=<20 30 30>
      constraintSetup {angularLimitsMax=<15 0 35>, angularLimitsMin=<0 -10 0>, angularTarget=<1 0 0>}
      localJointOffset=<4 0 0>
     }]
   }
   n11 = Anim Dynamics - Bone: dyn_ear_l(ComponentPose=@n10, Alpha=1, ExternalForce="0,0,0") {
    angularDampingOverride=1
    angularSpringConstant=20000
    bAngularSpring=true
    bOverrideAngularDamping=true
    bUseGravityOverride=true
    gravityScale=0.35
    simulationSpace=RootRelative
    alphaBoolBlend {blendOption=Linear}
    boundBone {boneName=dyn_ear_l}
    componentAppliedLinearAccClamp=<3000 3000 3000>
    componentLinearAccScale=<1 1 1>
    componentLinearVelScale=<1 1 1>
    gravityOverride {z=-980}
    physicsBodyDefinitions [{
      boundBone {boneName=dyn_ear_l}
      boxExtents=<20 30 30>
      constraintSetup {angularLimitsMax=<0 3 10>, angularLimitsMin=<0 -3 -10>, angularTarget=<1 0 0>}
      localJointOffset=<4 0 0>
     }]
   }
   n12 = Anim Dynamics - Bone: dyn_ear_r(ComponentPose=@n11, Alpha=1, ExternalForce="0,0,0") {
    angularDampingOverride=1
    angularSpringConstant=20000
    bAngularSpring=true
    bOverrideAngularDamping=true
    bUseGravityOverride=true
    gravityScale=0.35
    simulationSpace=RootRelative
    alphaBoolBlend {blendOption=Linear}
    boundBone {boneName=dyn_ear_r}
    componentAppliedLinearAccClamp=<3000 3000 3000>
    componentLinearAccScale=<1 1 1>
    componentLinearVelScale=<1 1 1>
    gravityOverride {z=-980}
    physicsBodyDefinitions [{
      boundBone {boneName=dyn_ear_r}
      boxExtents=<20 30 30>
      constraintSetup {angularLimitsMax=<0 3 10>, angularLimitsMin=<0 -3 -10>, angularTarget=<1 0 0>}
      localJointOffset=<4 0 0>
     }]
   }
   n13 = Anim Dynamics - Bone: dyn_glasses(ComponentPose=@n12, Alpha=1, ExternalForce="0,0,0") {
    angularDampingOverride=1
    angularSpringConstant=20000
    bAngularSpring=true
    bOverrideAngularDamping=true
    bUseGravityOverride=true
    gravityScale=0.35
    simulationSpace=RootRelative
    alphaBoolBlend {blendOption=Linear}
    boundBone {boneName=dyn_glasses}
    componentAppliedLinearAccClamp=<3000 3000 3000>
    componentLinearAccScale=<0.4 0.4 0.4>
    componentLinearVelScale=<0.4 0.4 0.4>
    gravityOverride {z=-980}
    physicsBodyDefinitions [{
      boundBone {boneName=dyn_glasses}
      boxExtents=<20 20 20>
      constraintSetup {angularLimitsMax=<0 0 0>, angularLimitsMin=<0 0 -10>, angularTarget=<1 0 0>}
      localJointOffset=<4 0 0>
     }]
   }
   n14 = Anim Dynamics - Bone: dyn_mic(ComponentPose=@n13, Alpha=1, ExternalForce="0,0,0") {
    angularDampingOverride=1
    angularSpringConstant=20000
    bAngularSpring=true
    bOverrideAngularDamping=true
    bUseGravityOverride=true
    gravityScale=0.35
    simulationSpace=RootRelative
    alphaBoolBlend {blendOption=Linear}
    boundBone {boneName=dyn_mic}
    componentAppliedLinearAccClamp=<3000 3000 3000>
    componentLinearAccScale=<1 1 1>
    componentLinearVelScale=<1 1 1>
    gravityOverride {z=-980}
    physicsBodyDefinitions [{
      boundBone {boneName=dyn_mic}
      boxExtents=<20 30 30>
      constraintSetup {angularLimitsMax=<2 2 2>, angularLimitsMin=<-2 -2 -2>, angularTarget=<1 0 0>}
      localJointOffset=<4 0 0>
     }]
   }
   n15 = "Constraint - dyn_ponytail_01_l (P)"(ComponentPose=Local To Component(LocalPose=Use cached pose 'StartCache'()), Alpha=1, ConstraintWeights_0=1) {
    alphaBoolBlend {blendOption=Linear}
    boneToModify {boneName=dyn_ponytail_01_l}
    constraintSetup [{targetBone {boneName=neck_01}}]
    constraintWeights [1]
   }
   n16 = "Constraint - dyn_ponytail_01_r (P)"(ComponentPose=@n15, Alpha=1, ConstraintWeights_0=1) {
    alphaBoolBlend {blendOption=Linear}
    boneToModify {boneName=dyn_ponytail_01_r}
    constraintSetup [{targetBone {boneName=neck_01}}]
    constraintWeights [1]
   }
   n17 = "Constraint - dyn_ponytail_02_l (P)"(ComponentPose=@n16, Alpha=1, ConstraintWeights_0=1) {
    alphaBoolBlend {blendOption=Linear}
    alphaScaleBiasClamp {bInterpResult=true, interpSpeedDecreasing=5, interpSpeedIncreasing=5}
    boneToModify {boneName=dyn_ponytail_02_l}
    constraintSetup [{targetBone {boneName=neck_01}}]
    constraintWeights [1]
   }
   n18 = "Constraint - dyn_ponytail_02_r (P)"(ComponentPose=@n17, Alpha=1, ConstraintWeights_0=1) {
    alphaBoolBlend {blendOption=Linear}
    alphaScaleBiasClamp {bInterpResult=true, interpSpeedDecreasing=5, interpSpeedIncreasing=5}
    boneToModify {boneName=dyn_ponytail_02_r}
    constraintSetup [{targetBone {boneName=neck_01}}]
    constraintWeights [1]
   }
   n19 = "Constraint - dyn_ponytail_03_l (P)"(ComponentPose=@n18, Alpha=1, ConstraintWeights_0=1) {
    alphaBoolBlend {blendOption=Linear}
    boneToModify {boneName=dyn_ponytail_03_l}
    constraintSetup [{targetBone {boneName=spine_05}}]
    constraintWeights [1]
   }
   n20 = "Constraint - dyn_ponytail_03_r (P)"(ComponentPose=@n19, Alpha=1, ConstraintWeights_0=1) {
    alphaBoolBlend {blendOption=Linear}
    boneToModify {boneName=dyn_ponytail_03_r}
    constraintSetup [{targetBone {boneName=spine_05}}]
    constraintWeights [1]
   }
   n21 = "Constraint - dyn_ponytail_04_l (P)"(ComponentPose=@n20, Alpha=1, ConstraintWeights_0=1) {
    alphaBoolBlend {blendOption=Linear}
    boneToModify {boneName=dyn_ponytail_04_l}
    constraintSetup [{targetBone {boneName=spine_04}}]
    constraintWeights [1]
   }
   n22 = "Constraint - dyn_ponytail_04_r (P)"(ComponentPose=@n21, Alpha=1, ConstraintWeights_0=1) {
    alphaBoolBlend {blendOption=Linear}
    boneToModify {boneName=dyn_ponytail_04_r}
    constraintSetup [{targetBone {boneName=spine_04}}]
    constraintWeights [1]
   }
   n23 = "Constraint - dyn_ponytail_05_l (P)"(ComponentPose=@n22, Alpha=1, ConstraintWeights_0=1) {
    alphaBoolBlend {blendOption=Linear}
    boneToModify {boneName=dyn_ponytail_05_l}
    constraintSetup [{targetBone {boneName=spine_03}}]
    constraintWeights [1]
   }
   n24 = "Constraint - dyn_ponytail_05_l (P)"(ComponentPose=@n23, Alpha=1, ConstraintWeights_0=1) {
    alphaBoolBlend {blendOption=Linear}
    boneToModify {boneName=dyn_ponytail_05_l}
    constraintSetup [{targetBone {boneName=spine_03}}]
    constraintWeights [1]
   }
   n25 = "Constraint - dyn_ponytail_06_l (P)"(ComponentPose=@n24, Alpha=1, ConstraintWeights_0=1) {
    alphaBoolBlend {blendOption=Linear}
    boneToModify {boneName=dyn_ponytail_06_l}
    constraintSetup [{targetBone {boneName=spine_01}}]
    constraintWeights [1]
   }
   n26 = "Constraint - dyn_ponytail_06_r (P)"(ComponentPose=@n25, Alpha=1, ConstraintWeights_0=1) {
    alphaBoolBlend {blendOption=Linear}
    boneToModify {boneName=dyn_ponytail_06_r}
    constraintSetup [{targetBone {boneName=spine_01}}]
    constraintWeights [1]
   }
   n27 = "Transform (Modify) Bone - Bone: base_ponytail_l"(ComponentPose=@n26, Alpha=1, Translation="-3,-12,-7", Rotation="0,0,0", Scale="1,1,1") {
    rotationMode=BMM_Additive
    rotationSpace=BCS_BoneSpace
    scaleMode=BMM_Additive
    scaleSpace=BCS_BoneSpace
    translationMode=BMM_Additive
    translationSpace=BCS_BoneSpace
    alphaBoolBlend {blendInTime=0.2, blendOption=Linear, blendOutTime=0.2}
    alphaScaleBiasClamp {bInterpResult=true}
    boneToModify {boneName=base_ponytail_l}
    translation=<-3 -12 -7>
   }
   n28 = "Transform (Modify) Bone - Bone: base_ponytail_l"(ComponentPose=@n27, Alpha=1, Translation="0,0,0", Rotation="0,-30,0", Scale="1,1,1") {
    rotationMode=BMM_Additive
    rotationSpace=BCS_BoneSpace
    scaleMode=BMM_Additive
    scaleSpace=BCS_BoneSpace
    translationMode=BMM_Additive
    translationSpace=BCS_BoneSpace
    alphaBoolBlend {blendInTime=0.2, blendOption=Linear, blendOutTime=0.2}
    alphaScaleBiasClamp {bInterpResult=true}
    boneToModify {boneName=base_ponytail_l}
    rotation {yaw=-30}
   }
   n29 = "Transform (Modify) Bone - Bone: base_ponytail_r"(ComponentPose=@n28, Alpha=1, Translation="0,0,0", Rotation="0,-30,0", Scale="1,1,1") {
    rotationMode=BMM_Additive
    rotationSpace=BCS_BoneSpace
    scaleMode=BMM_Additive
    scaleSpace=BCS_BoneSpace
    translationMode=BMM_Additive
    translationSpace=BCS_BoneSpace
    alphaBoolBlend {blendInTime=0.2, blendOption=Linear, blendOutTime=0.2}
    alphaScaleBiasClamp {bInterpResult=true}
    boneToModify {boneName=base_ponytail_r}
    rotation {yaw=-30}
   }
   n30 = "Transform (Modify) Bone - Bone: ponytail_ASW_02_r"!(ComponentPose=@n29, bAlphaBoolEnabled:bool=True, Translation="5,-10,0", Rotation="0,0,0", Scale="1,1,1") -> (Pose) {
    alphaInputType=Bool
    rotationMode=BMM_Additive
    rotationSpace=BCS_BoneSpace
    translationMode=BMM_Additive
    translationSpace=BCS_BoneSpace
    alphaBoolBlend {blendInTime=0.2, blendOption=Linear, blendOutTime=0.2}
    alphaScaleBiasClamp {bInterpResult=true, interpSpeedDecreasing=7, interpSpeedIncreasing=7}
    boneToModify {boneName=ponytail_ASW_02_r}
    translation=<5 -10>
   }
   n31 = "Transform (Modify) Bone - Bone: ponytail_ASW_02_l"!(ComponentPose=@n30, bAlphaBoolEnabled:bool=True, Translation="-5,-10,0", Rotation="0,0,0", Scale="1,1,1") -> (Pose) {
    alphaInputType=Bool
    rotationMode=BMM_Additive
    rotationSpace=BCS_BoneSpace
    translationMode=BMM_Additive
    translationSpace=BCS_BoneSpace
    alphaBoolBlend {blendInTime=0.2, blendOption=Linear, blendOutTime=0.2}
    alphaScaleBiasClamp {bInterpResult=true, interpSpeedDecreasing=7, interpSpeedIncreasing=7}
    boneToModify {boneName=ponytail_ASW_02_l}
    translation=<-5 -10>
   }
   n32 = "Transform (Modify) Bone - Bone: ponytail_ASW_02_r"(ComponentPose=@n31, bAlphaBoolEnabled=True, Translation="0,-10,0", Rotation="0,0,0", Scale="1,1,1") {
    alphaInputType=Bool
    rotationMode=BMM_Additive
    rotationSpace=BCS_BoneSpace
    translationMode=BMM_Additive
    translationSpace=BCS_BoneSpace
    alphaBoolBlend {blendInTime=0.2, blendOption=Linear, blendOutTime=0.2}
    alphaScaleBiasClamp {bInterpResult=true, interpSpeedDecreasing=7, interpSpeedIncreasing=7}
    boneToModify {boneName=ponytail_ASW_02_r}
    translation {y=-10}
   }
   n33 = "Transform (Modify) Bone - Bone: ponytail_ASW_03_r"!(ComponentPose=@n32, bAlphaBoolEnabled:bool=True, Translation="0,-10,0", Rotation="0,0,0", Scale="1,1,1") -> (Pose) {
    alphaInputType=Bool
    rotationMode=BMM_Additive
    rotationSpace=BCS_BoneSpace
    translationMode=BMM_Additive
    translationSpace=BCS_BoneSpace
    alphaBoolBlend {blendInTime=0.2, blendOption=Linear, blendOutTime=0.2}
    alphaScaleBiasClamp {bInterpResult=true, interpSpeedDecreasing=7, interpSpeedIncreasing=7}
    boneToModify {boneName=ponytail_ASW_03_r}
    translation {y=-10}
   }
   n34 = "Transform (Modify) Bone - Bone: ponytail_ASW_01_l"(ComponentPose=@n33, Alpha=1, Translation="-10,-5,0", Rotation="0,0,0", Scale="1,1,1") {
    rotationMode=BMM_Additive
    rotationSpace=BCS_BoneSpace
    translationMode=BMM_Additive
    translationSpace=BCS_BoneSpace
    alphaBoolBlend {blendInTime=0.2, blendOption=Linear, blendOutTime=0.2}
    alphaScaleBiasClamp {bInterpResult=true, interpSpeedIncreasing=6}
    boneToModify {boneName=ponytail_ASW_01_l}
    translation=<-10 -5>
   }
   n35 = "Transform (Modify) Bone - Bone: ponytail_ASW_02_l"!(ComponentPose=@n34, Alpha:float=1, Translation="-26,-15,0", Rotation="0,0,0", Scale="1,1,1") -> (Pose) {
    rotationMode=BMM_Additive
    rotationSpace=BCS_BoneSpace
    translationMode=BMM_Additive
    translationSpace=BCS_BoneSpace
    alphaBoolBlend {blendInTime=0.2, blendOption=Linear, blendOutTime=0.2}
    alphaScaleBiasClamp {bInterpResult=true, interpSpeedIncreasing=6}
    boneToModify {boneName=ponytail_ASW_02_l}
    translation=<-26 -15>
   }
   n36 = "Transform (Modify) Bone - Bone: ponytail_ASW_03_l"(ComponentPose=@n35, Alpha=1, Translation="-26,-15,0", Rotation="-10,0,0", Scale="1,1,1") {
    rotationMode=BMM_Additive
    rotationSpace=BCS_BoneSpace
    translationMode=BMM_Additive
    translationSpace=BCS_BoneSpace
    alphaBoolBlend {blendInTime=0.2, blendOption=Linear, blendOutTime=0.2}
    alphaScaleBiasClamp {bInterpResult=true, interpSpeedIncreasing=6}
    boneToModify {boneName=ponytail_ASW_03_l}
    rotation {pitch=-10}
    translation=<-26 -15>
   }
   n37 = "Transform (Modify) Bone - Bone: ponytail_ASW_01_r"(ComponentPose=@n36, Alpha=1, Translation="10,-5,0", Rotation="0,0,0", Scale="1,1,1") {
    rotationMode=BMM_Additive
    rotationSpace=BCS_BoneSpace
    translationMode=BMM_Additive
    translationSpace=BCS_BoneSpace
    alphaBoolBlend {blendInTime=0.2, blendOption=Linear, blendOutTime=0.2}
    alphaScaleBiasClamp {bInterpResult=true, interpSpeedIncreasing=6}
    boneToModify {boneName=ponytail_ASW_01_r}
    translation=<10 -5>
   }
   n38 = "Transform (Modify) Bone - Bone: ponytail_ASW_02_r"!(ComponentPose=@n37, Alpha:float=1, Translation="20,-20,0", Rotation="0,0,0", Scale="1,1,1") -> (Pose) {
    rotationMode=BMM_Additive
    rotationSpace=BCS_BoneSpace
    translationMode=BMM_Additive
    translationSpace=BCS_BoneSpace
    alphaBoolBlend {blendInTime=0.2, blendOption=Linear, blendOutTime=0.2}
    alphaScaleBiasClamp {bInterpResult=true, interpSpeedIncreasing=6}
    boneToModify {boneName=ponytail_ASW_02_r}
    translation=<20 -20>
   }
   n39 = "Transform (Modify) Bone - Bone: ponytail_ASW_03_r"!(ComponentPose=@n38, Alpha:float=1, Translation="20,-20,0", Rotation="0,0,0", Scale="1,1,1") -> (Pose) {
    rotationMode=BMM_Additive
    rotationSpace=BCS_BoneSpace
    translationMode=BMM_Additive
    translationSpace=BCS_BoneSpace
    alphaBoolBlend {blendInTime=0.2, blendOption=Linear, blendOutTime=0.2}
    alphaScaleBiasClamp {bInterpResult=true, interpSpeedIncreasing=6}
    boneToModify {boneName=ponytail_ASW_03_r}
    translation=<20 -20>
   }
   n40 = "Transform (Modify) Bone - Bone: ponytail_ASW_01_r"(ComponentPose=@n39, Alpha=1, Translation="10,0,0", Rotation="0,0,0", Scale="1,1,1") {
    rotationMode=BMM_Additive
    rotationSpace=BCS_BoneSpace
    translationMode=BMM_Additive
    translationSpace=BCS_BoneSpace
    alphaBoolBlend {blendInTime=0.2, blendOption=Linear, blendOutTime=0.2}
    alphaScaleBiasClamp {bInterpResult=true, interpSpeedIncreasing=6}
    boneToModify {boneName=ponytail_ASW_01_r}
    translation {x=10}
   }
   n41 = "Transform (Modify) Bone - Bone: ponytail_ASW_02_r"(ComponentPose=@n40, Alpha=1, Translation="10.000,-20.893,-0.000", Rotation="-0.005,0.002,-8.230", Scale="1,1,1") {
    rotationMode=BMM_Additive
    rotationSpace=BCS_BoneSpace
    translationMode=BMM_Additive
    translationSpace=BCS_BoneSpace
    alphaBoolBlend {blendInTime=0.2, blendOption=Linear, blendOutTime=0.2}
    alphaScaleBiasClamp {bInterpResult=true, interpSpeedIncreasing=6}
    boneToModify {boneName=ponytail_ASW_02_r}
    rotation=<rot -0.005 0.002 -8.23>
    translation=<10 -20.893>
   }
   n42 = "Transform (Modify) Bone - Bone: ponytail_ASW_03_r"(ComponentPose=@n41, Alpha=1, Translation="12.604,-13.162,-0.868", Rotation="-0.010,0.004,-31.250", Scale="1,1,1") {
    rotationMode=BMM_Additive
    rotationSpace=BCS_BoneSpace
    translationMode=BMM_Additive
    translationSpace=BCS_BoneSpace
    alphaBoolBlend {blendInTime=0.2, blendOption=Linear, blendOutTime=0.2}
    alphaScaleBiasClamp {bInterpResult=true, interpSpeedIncreasing=6}
    boneToModify {boneName=ponytail_ASW_03_r}
    rotation=<rot -0.01 0.004 -31.25>
    translation=<12.604 -13.162 -0.868>
   }
   n43 = "Transform (Modify) Bone - Bone: dyn_ponytail_03_r"(ComponentPose=@n42, Alpha=1, Translation="0,10,0", Rotation="0,0,0", Scale="1,1,1") {
    rotationMode=BMM_Additive
    rotationSpace=BCS_BoneSpace
    translationMode=BMM_Additive
    translationSpace=BCS_BoneSpace
    alphaBoolBlend {blendInTime=0.2, blendOption=Linear, blendOutTime=0.2}
    alphaScaleBiasClamp {bInterpResult=true, interpSpeedIncreasing=6}
    boneToModify {boneName=dyn_ponytail_03_r}
    translation {y=10}
   }
   n44 = "Transform (Modify) Bone - Bone: ponytail_ASW_02_l"!(ComponentPose=@n43, bAlphaBoolEnabled:bool=True, Translation="0,-30,0", Rotation="0,0,0", Scale="1,1,1") -> (Pose) {
    alphaInputType=Bool
    rotationMode=BMM_Additive
    rotationSpace=BCS_BoneSpace
    translationMode=BMM_Additive
    translationSpace=BCS_BoneSpace
    alphaBoolBlend {blendInTime=0.2, blendOption=Linear, blendOutTime=0.2}
    alphaScaleBiasClamp {bInterpResult=true, interpSpeedIncreasing=6}
    boneToModify {boneName=ponytail_ASW_02_l}
    translation {y=-30}
   }
   n45 = "Transform (Modify) Bone - Bone: ponytail_ASW_02_r"!(ComponentPose=@n44, bAlphaBoolEnabled:bool=True, Translation="0,-30,0", Rotation="0,0,0", Scale="1,1,1") -> (Pose) {
    alphaInputType=Bool
    rotationMode=BMM_Additive
    rotationSpace=BCS_BoneSpace
    translationMode=BMM_Additive
    translationSpace=BCS_BoneSpace
    alphaBoolBlend {blendInTime=0.2, blendOption=Linear, blendOutTime=0.2}
    alphaScaleBiasClamp {bInterpResult=true, interpSpeedIncreasing=6}
    boneToModify {boneName=ponytail_ASW_02_r}
    translation {y=-30}
   }
   n46 = "Transform (Modify) Bone - Bone: ponytail_ASW_03_l"!(ComponentPose=@n45, bAlphaBoolEnabled:bool=True, Translation="0,-20,0", Rotation="0,0,0", Scale="1,1,1") -> (Pose) {
    alphaInputType=Bool
    rotationMode=BMM_Additive
    rotationSpace=BCS_BoneSpace
    translationMode=BMM_Additive
    translationSpace=BCS_BoneSpace
    alphaBoolBlend {blendInTime=0.2, blendOption=Linear, blendOutTime=0.2}
    alphaScaleBiasClamp {bInterpResult=true, interpSpeedIncreasing=6}
    boneToModify {boneName=ponytail_ASW_03_l}
    translation {y=-20}
   }
   n47 = "Transform (Modify) Bone - Bone: ponytail_ASW_03_r"!(ComponentPose=@n46, bAlphaBoolEnabled:bool=True, Translation="0,-20,0", Rotation="0,0,0", Scale="1,1,1") -> (Pose) {
    alphaInputType=Bool
    rotationMode=BMM_Additive
    rotationSpace=BCS_BoneSpace
    translationMode=BMM_Additive
    translationSpace=BCS_BoneSpace
    alphaBoolBlend {blendInTime=0.2, blendOption=Linear, blendOutTime=0.2}
    alphaScaleBiasClamp {bInterpResult=true, interpSpeedIncreasing=6}
    boneToModify {boneName=ponytail_ASW_03_r}
    translation {y=-20}
   }
   Output Pose(Result=Blend Poses by bool(BlendPose_0=Component To Local(ComponentPose=@n14), BlendPose_1=Component To Local(ComponentPose=@n47), bActiveValue=Get bRigidBodyEnabled()) {blendType=Linear}) {layerGroup=None}
   n48 = "Constraint - dyn_ponytail_01_l (P)"(ComponentPose=Local To Component(LocalPose=Use cached pose 'StartCache'()), Alpha=1, ConstraintWeights_0=1) {
    alphaBoolBlend {blendOption=Linear}
    boneToModify {boneName=dyn_ponytail_01_l}
    constraintSetup [{targetBone {boneName=neck_02}}]
    constraintWeights [1]
   }
   n49 = "Constraint - dyn_ponytail_02_l (P)"(ComponentPose=@n48, Alpha=Get Alpha_NotMeleePickaxe() -> (Alpha_NotMeleePickaxe:double), ConstraintWeights_0=1) {
    alphaBoolBlend {blendOption=Linear}
    alphaScaleBiasClamp {bInterpResult=true, interpSpeedDecreasing=5, interpSpeedIncreasing=5}
    boneToModify {boneName=dyn_ponytail_02_l}
    constraintSetup [{targetBone {boneName=upperarm_l}}]
    constraintWeights [1]
   }
   n50 = "Constraint - dyn_ponytail_02_l (R)"(ComponentPose=@n49, Alpha=Get Alpha_MeleePickaxe() -> (Alpha_MeleePickaxe:double), ConstraintWeights_0=1) {
    alphaBoolBlend {blendOption=Linear}
    alphaScaleBiasClamp {bInterpResult=true, interpSpeedDecreasing=5, interpSpeedIncreasing=5}
    boneToModify {boneName=dyn_ponytail_02_l}
    constraintSetup [{transformType=Rotation, targetBone {boneName=spine_05}}]
    constraintWeights [1]
   }
   n51 = "Constraint - dyn_ponytail_03_l (P)"(ComponentPose=@n50, Alpha=1, ConstraintWeights_0=1) {
    alphaBoolBlend {blendOption=Linear}
    boneToModify {boneName=dyn_ponytail_03_l}
    constraintSetup [{targetBone {boneName=spine_05}}]
    constraintWeights [1]
   }
   n52 = "Constraint - dyn_ponytail_01_r (P)"(ComponentPose=@n51, Alpha=1, ConstraintWeights_0=1) {
    alphaBoolBlend {blendOption=Linear}
    boneToModify {boneName=dyn_ponytail_01_r}
    constraintSetup [{targetBone {boneName=neck_02}}]
    constraintWeights [1]
   }
   n53 = "Constraint - dyn_ponytail_02_r (P)"(ComponentPose=@n52, Alpha=1, ConstraintWeights_0=1) {
    alphaBoolBlend {blendOption=Linear}
    boneToModify {boneName=dyn_ponytail_02_r}
    constraintSetup [{targetBone {boneName=upperarm_r}}]
    constraintWeights [1]
   }
   n54 = "Constraint - dyn_ponytail_03_r (P)"(ComponentPose=@n53, Alpha=1, ConstraintWeights_0=1) {
    alphaBoolBlend {blendOption=Linear}
    boneToModify {boneName=dyn_ponytail_03_r}
    constraintSetup [{targetBone {boneName=spine_05}}]
    constraintWeights [1]
   }
   n55 = "Transform (Modify) Bone - Bone: base_ponytail_l"!(ComponentPose=@n54, bAlphaBoolEnabled:bool=True, Translation="0.636,4.361,1.164", Rotation="2.307,-1.774,1.684", Scale="1,1,1") -> (Pose) {
    alphaInputType=Bool
    rotationMode=BMM_Additive
    rotationSpace=BCS_BoneSpace
    scaleMode=BMM_Additive
    scaleSpace=BCS_BoneSpace
    translationMode=BMM_Additive
    translationSpace=BCS_BoneSpace
    alphaBoolBlend {blendInTime=0.2, blendOption=Linear, blendOutTime=0.2}
    boneToModify {boneName=base_ponytail_l}
    rotation=<rot 2.307 -1.774 1.684>
    translation=<0.636 4.361 1.164>
   }
   n56 = "Transform (Modify) Bone - Bone: base_ponytail_r"!(ComponentPose=@n55, bAlphaBoolEnabled:bool=True, Translation="0.137,0.000,-1.550", Rotation="8.225,-5.982,-0.856", Scale="1,1,1") -> (Pose) {
    alphaInputType=Bool
    rotationMode=BMM_Additive
    rotationSpace=BCS_BoneSpace
    scaleMode=BMM_Additive
    scaleSpace=BCS_BoneSpace
    translationMode=BMM_Additive
    translationSpace=BCS_BoneSpace
    alphaBoolBlend {blendInTime=0.2, blendOption=Linear, blendOutTime=0.2}
    boneToModify {boneName=base_ponytail_r}
    rotation=<rot 8.225 -5.9819 -0.856>
    translation=<0.137 0 -1.5504>
   }
   n57 = "Transform (Modify) Bone - Bone: dyn_ponytail_03_l"(ComponentPose=@n56, Alpha=1, Translation="1.108,-5,0", Rotation="0,-27,0", Scale="1,1,1") {
    rotationMode=BMM_Additive
    rotationSpace=BCS_BoneSpace
    scaleMode=BMM_Additive
    scaleSpace=BCS_BoneSpace
    translationMode=BMM_Additive
    translationSpace=BCS_BoneSpace
    alphaBoolBlend {blendOption=Linear}
    alphaScaleBiasClamp {bInterpResult=true, interpSpeedDecreasing=5, interpSpeedIncreasing=5}
    boneToModify {boneName=dyn_ponytail_03_l}
    rotation {yaw=-27}
    translation=<1.108 -5>
   }
   n58 = "Transform (Modify) Bone - Bone: dyn_ponytail_02_l"(ComponentPose=@n57, Alpha=1, Translation="-10.651,9.707,-0.732", Rotation="0,0,0", Scale="1,1,1") {
    rotationMode=BMM_Additive
    rotationSpace=BCS_BoneSpace
    scaleMode=BMM_Additive
    scaleSpace=BCS_BoneSpace
    translationMode=BMM_Additive
    translationSpace=BCS_BoneSpace
    alphaBoolBlend {blendOption=Linear}
    alphaScaleBiasClamp {bInterpResult=true, interpSpeedDecreasing=1}
    boneToModify {boneName=dyn_ponytail_02_l}
    translation=<-10.651 9.707 -0.732>
   }
   n59 = "Transform (Modify) Bone - Bone: base_ponytail_l"!(ComponentPose=@n58, Alpha:float=1, Translation="0.248,2.999,-0.000", Rotation="0,0,0", Scale="1,1,1") -> (Pose) {
    rotationMode=BMM_Additive
    rotationSpace=BCS_BoneSpace
    scaleMode=BMM_Additive
    scaleSpace=BCS_BoneSpace
    translationMode=BMM_Additive
    translationSpace=BCS_BoneSpace
    alphaBoolBlend {blendInTime=0.2, blendOption=Linear, blendOutTime=0.2}
    alphaScaleBiasClamp {bInterpResult=true, interpSpeedDecreasing=5, interpSpeedIncreasing=5}
    boneToModify {boneName=base_ponytail_l}
    translation=<0.2476 2.999 0>
   }
   n60 = "Transform (Modify) Bone - Bone: dyn_ponytail_03_l"!(ComponentPose=@n59, bAlphaBoolEnabled:bool=True, Translation="-4.646,-16.496,-0.052", Rotation="0.000,-9.597,-0.000", Scale="1,1,1") -> (Pose) {
    alphaInputType=Bool
    rotationMode=BMM_Additive
    rotationSpace=BCS_BoneSpace
    scaleMode=BMM_Additive
    scaleSpace=BCS_BoneSpace
    translationMode=BMM_Additive
    translationSpace=BCS_BoneSpace
    alphaBoolBlend {blendInTime=0.2, blendOption=Linear, blendOutTime=0.2}
    alphaScaleBiasClamp {bInterpResult=true, interpSpeedDecreasing=5, interpSpeedIncreasing=5}
    boneToModify {boneName=dyn_ponytail_03_l}
    rotation=<rot 0 -9.597 0>
    translation=<-4.646 -16.496 -0.052>
   }
   n61 = "Transform (Modify) Bone - Bone: dyn_ponytail_03_r"!(ComponentPose=@n60, bAlphaBoolEnabled:bool=True, Translation="0,5,0", Rotation="17.695,0,-13.055", Scale="1,1,1") -> (Pose) {
    alphaInputType=Bool
    rotationMode=BMM_Additive
    rotationSpace=BCS_BoneSpace
    scaleMode=BMM_Additive
    scaleSpace=BCS_BoneSpace
    translationMode=BMM_Additive
    translationSpace=BCS_BoneSpace
    alphaBoolBlend {blendInTime=0.2, blendOption=Linear, blendOutTime=0.2}
    alphaScaleBiasClamp {bInterpResult=true, interpSpeedDecreasing=5, interpSpeedIncreasing=5}
    boneToModify {boneName=dyn_ponytail_03_r}
    rotation {pitch=17.695, roll=-13.055}
    translation {y=5}
   }
   n62 = "Transform (Modify) Bone - Bone: dyn_ponytail_03_l"(ComponentPose=@n61, bAlphaBoolEnabled=True, Translation="-0.096,-0.572,-0.000", Rotation="3.434,-9.644,-0.154", Scale="1,1,1") {
    alphaInputType=Bool
    rotationMode=BMM_Additive
    rotationSpace=BCS_BoneSpace
    scaleMode=BMM_Additive
    scaleSpace=BCS_BoneSpace
    translationMode=BMM_Additive
    translationSpace=BCS_BoneSpace
    alphaBoolBlend {blendInTime=0.2, blendOption=Linear, blendOutTime=0.2}
    alphaScaleBiasClamp {bInterpResult=true, interpSpeedDecreasing=5, interpSpeedIncreasing=5}
    boneToModify {boneName=dyn_ponytail_03_l}
    rotation=<rot 3.434 -9.6439 -0.154>
    translation=<-0.096 -0.572>
   }
   n63 = "Transform (Modify) Bone - Bone: base_ponytail_l"!(ComponentPose=@n62, bAlphaBoolEnabled:bool=True, Translation="0.376,-0.637,0.140", Rotation="16.922,-30.239,-4.451", Scale="1,1,1") -> (Pose) {
    alphaInputType=Bool
    rotationMode=BMM_Additive
    rotationSpace=BCS_BoneSpace
    scaleMode=BMM_Additive
    scaleSpace=BCS_BoneSpace
    translationMode=BMM_Additive
    translationSpace=BCS_BoneSpace
    alphaBoolBlend {blendInTime=0.1, blendOption=Linear, blendOutTime=0.2}
    alphaScaleBiasClamp {bInterpResult=true}
    boneToModify {boneName=base_ponytail_l}
    rotation=<rot 16.922 -30.2389 -4.451>
    translation=<0.376 -0.637 0.14>
   }
   n64 = "Transform (Modify) Bone - Bone: base_ponytail_r"(ComponentPose=@n63, bAlphaBoolEnabled=True, Translation="0.669,2.881,-0.000", Rotation="-0.000,-18.297,-0.000", Scale="1,1,1") {
    alphaInputType=Bool
    rotationMode=BMM_Additive
    rotationSpace=BCS_BoneSpace
    scaleMode=BMM_Additive
    scaleSpace=BCS_BoneSpace
    translationMode=BMM_Additive
    translationSpace=BCS_BoneSpace
    alphaBoolBlend {blendInTime=0.1, blendOption=Linear, blendOutTime=0.2}
    alphaScaleBiasClamp {bInterpResult=true}
    boneToModify {boneName=base_ponytail_r}
    rotation=<rot 0 -18.297 0>
    translation=<0.669 2.881>
   }
   n65 = "Transform (Modify) Bone - Bone: dyn_ponytail_02_l"(ComponentPose=@n64, Alpha=1, Translation="-9.814,-13.114,12.450", Rotation="9.728,-4.560,0.228", Scale="1,1,1") {
    rotationMode=BMM_Additive
    rotationSpace=BCS_BoneSpace
    scaleMode=BMM_Additive
    scaleSpace=BCS_BoneSpace
    translationMode=BMM_Additive
    translationSpace=BCS_BoneSpace
    alphaBoolBlend {blendInTime=0.2, blendOption=Linear, blendOutTime=0.2}
    alphaScaleBiasClamp {bInterpResult=true, interpSpeedDecreasing=8}
    boneToModify {boneName=dyn_ponytail_02_l}
    rotation=<rot 9.728 -4.56 0.228>
    translation=<-9.814 -13.114 12.45>
   }
   n66 = "Transform (Modify) Bone - Bone: base_ponytail_l"(ComponentPose=@n65, bAlphaBoolEnabled=True, Translation="0,0,0", Rotation="0.000,-11.152,-0.000", Scale="1,1,1") {
    alphaInputType=Bool
    rotationMode=BMM_Additive
    rotationSpace=BCS_BoneSpace
    scaleMode=BMM_Additive
    scaleSpace=BCS_BoneSpace
    translationMode=BMM_Additive
    translationSpace=BCS_BoneSpace
    alphaBoolBlend {blendInTime=0.5, blendOption=Linear, blendOutTime=0.2}
    alphaScaleBiasClamp {bInterpResult=true, interpSpeedDecreasing=5}
    boneToModify {boneName=base_ponytail_l}
    rotation=<rot 0 -11.1519 0>
   }
   n67 = "Transform (Modify) Bone - Bone: dyn_ponytail_03_r"!(ComponentPose=@n66, Alpha:float=1, Translation="1.305,9.219,-3.717", Rotation="2.814,-25.067,-0.454", Scale="1,1,1") -> (Pose) {
    rotationMode=BMM_Additive
    rotationSpace=BCS_BoneSpace
    scaleMode=BMM_Additive
    scaleSpace=BCS_BoneSpace
    translationMode=BMM_Additive
    translationSpace=BCS_BoneSpace
    alphaBoolBlend {blendInTime=0.5, blendOption=Linear, blendOutTime=0.2}
    alphaScaleBiasClamp {bInterpResult=true, interpSpeedDecreasing=5, interpSpeedIncreasing=5}
    boneToModify {boneName=dyn_ponytail_03_r}
    rotation=<rot 2.814 -25.067 -0.454>
    translation=<1.305 9.219 -3.717>
   }
   n68 = "Transform (Modify) Bone - Bone: dyn_ponytail_03_l"!(ComponentPose=@n67, Alpha:float=1, Translation="-0.000,-5.055,0.000", Rotation="4.811,-24.543,-2.188", Scale="1,1,1") -> (Pose) {
    rotationMode=BMM_Additive
    rotationSpace=BCS_BoneSpace
    scaleMode=BMM_Additive
    scaleSpace=BCS_BoneSpace
    translationMode=BMM_Additive
    translationSpace=BCS_BoneSpace
    alphaBoolBlend {blendInTime=0.5, blendOption=Linear, blendOutTime=0.2}
    alphaScaleBiasClamp {bInterpResult=true, interpSpeedDecreasing=5}
    boneToModify {boneName=dyn_ponytail_03_l}
    rotation=<rot 4.8111 -24.543 -2.188>
    translation=<0 -5.0554 0>
   }
   n69 = "Transform (Modify) Bone - Bone: base_ponytail_l"!(ComponentPose=@n68, Alpha:float=1, Translation="-0.126,-3.950,0.117", Rotation="-34.324,-0.113,-5.106", Scale="1,1,1") -> (Pose) {
    rotationMode=BMM_Additive
    rotationSpace=BCS_BoneSpace
    scaleMode=BMM_Additive
    scaleSpace=BCS_BoneSpace
    translationMode=BMM_Additive
    translationSpace=BCS_BoneSpace
    alphaBoolBlend {blendInTime=0.2, blendOption=Linear, blendOutTime=0.2}
    alphaScaleBiasClamp {bInterpResult=true}
    boneToModify {boneName=base_ponytail_l}
    rotation=<rot -34.324 -0.113 -5.106>
    translation=<-0.126 -3.95 0.117>
   }
   n70 = RigidBody(ComponentPose=@n69, Alpha=Get Alpha_NotSkydiving() -> (Alpha_NotSkydiving:double)) {
    overridePhysicsAsset=&/F_MED_JadeTowelHope_FaceAcc_Physics_SpecificBackpack
    alphaBoolBlend {blendOption=Linear}
    baseBoneRef {boneName=spine_04}
    componentAppliedLinearAccClamp=<3000 3000 3000>
    componentLinearAccScale=<0.3 0.3 0.3>
    componentLinearVelScale=<0.3 0.3 0.3>
    simSpaceSettings {worldAlpha=0.3}
   }
   Save cached pose 'RBBacpackCache'(Pose=Component To Local(ComponentPose=@n70)) {cacheName=RBBacpackCache}
   n71 = RigidBody!(
    ComponentPose=Local To Component(LocalPose=Use cached pose 'RBBacpackCache'())
    Alpha=Get Alpha_Skydiving() -> (Alpha_Skydiving:double)
    OverrideWorldGravity:Vector=Get GravityOverride()
    ExternalForce
    SimSpaceSettings:Sim Space Settings Structure=(WorldAlpha=0.5,VelocityScaleZ=1,DampingAlpha=1,MaxLinearVelocity=10000,MaxAngularVelocity=10000,MaxLinearAcceleration=10000,MaxAngularAcceleration=10000,ExternalLinearDragV=(X=0,Y=0,Z=0),ExternalLinearVelocity=(X=0,Y=0,Z=0),ExternalAngularVelocity=(X=0,Y=0,Z=0))
   ) -> (Pose) {
    bOverrideWorldGravity=true
    overridePhysicsAsset=&F_MED_JadeTowelHope_FaceAcc_Physics_SpecificBackpack
    alphaBoolBlend {blendOption=Linear}
    baseBoneRef {boneName=spine_04}
    componentAppliedLinearAccClamp=<3000 3000 3000>
    componentLinearAccScale=<0.3 0.3 0.3>
    componentLinearVelScale=<0.3 0.3 0.3>
    overrideWorldGravity {z=-980}
    simSpaceSettings {worldAlpha=0.5}
   }
   n72 = "Transform (Modify) Bone - Bone: base_ponytail_l"!(
    ComponentPose=Local To Component(LocalPose=Use cached pose 'RBBacpackCache'())
    bAlphaBoolEnabled:bool=True
    Translation="-4.346,-4.505,1.407"
    Rotation="-0.000,-0.000,17.315"
    Scale="1,1,1"
   ) -> (Pose) {
    alphaInputType=Bool
    rotationMode=BMM_Additive
    rotationSpace=BCS_BoneSpace
    scaleMode=BMM_Additive
    scaleSpace=BCS_BoneSpace
    translationMode=BMM_Additive
    translationSpace=BCS_BoneSpace
    alphaBoolBlend {blendInTime=0.2, blendOption=Linear, blendOutTime=0.4}
    alphaScaleBiasClamp {bInterpResult=true, interpSpeedDecreasing=5}
    boneToModify {boneName=base_ponytail_l}
    rotation {roll=17.315}
    translation=<-4.346 -4.505 1.407>
   }
   n73 = "Transform (Modify) Bone - Bone: base_ponytail_r"(ComponentPose=@n72, bAlphaBoolEnabled=True, Translation="1.086,-0.766,-2.000", Rotation="0.002,0.010,15.663", Scale="1,1,1") {
    alphaInputType=Bool
    rotationMode=BMM_Additive
    rotationSpace=BCS_BoneSpace
    scaleMode=BMM_Additive
    scaleSpace=BCS_BoneSpace
    translationMode=BMM_Additive
    translationSpace=BCS_BoneSpace
    alphaBoolBlend {blendInTime=0.2, blendOption=Linear, blendOutTime=0.4}
    alphaScaleBiasClamp {bInterpResult=true, interpSpeedDecreasing=5}
    boneToModify {boneName=base_ponytail_r}
    rotation=<rot 0.002 0.01 15.663>
    translation=<1.086 -0.766 -2>
   }
   n74 = RigidBody(
    ComponentPose=@n73
    OverrideWorldGravity=Get GravityOverride()
    SimSpaceSettings=(WorldAlpha=0.5,VelocityScaleZ=1,DampingAlpha=1,MaxLinearVelocity=10000,MaxAngularVelocity=10000,MaxLinearAcceleration=10000,MaxAngularAcceleration=10000,ExternalLinearDragV=(X=0,Y=0,Z=0),ExternalLinearVelocity=(X=0,Y=0,Z=0),ExternalAngularVelocity=(X=0,Y=0,Z=0))
   ) {
    bOverrideWorldGravity=true
    overridePhysicsAsset=&/F_MED_JadeTowelHope_FaceAcc_Physics
    alphaBoolBlend {blendOption=Linear}
    baseBoneRef {boneName=spine_04}
    componentAppliedLinearAccClamp=<3000 3000 3000>
    componentLinearAccScale=<0.3 0.3 0.3>
    componentLinearVelScale=<0.3 0.3 0.3>
    overrideWorldGravity {z=-980}
    simSpaceSettings {worldAlpha=0.5}
   }
   n75 = Blend Poses by bool(
    BlendPose_0=Component To Local(ComponentPose=@n74)
    BlendPose_1=Component To Local(ComponentPose=@n71)
    BlendTime_0=0.2
    BlendTime_1=0.5
    bActiveValue=Get bIsSkydiveFloating()
   ) {blendType=Linear}
   n76 = RigidBody!(ComponentPose=Local To Component(LocalPose=Use cached pose 'RBBacpackCache'()), Alpha=Get Alpha_NotSkydiving(), ExternalForce) -> (Pose) {
    overridePhysicsAsset=&/F_MED_JadeTowelHope_FaceAcc_Physics_Backpack
    alphaBoolBlend {blendOption=Linear}
    baseBoneRef {boneName=spine_04}
    componentAppliedLinearAccClamp=<3000 3000 3000>
    componentLinearAccScale=<0.3 0.3 0.3>
    componentLinearVelScale=<0.3 0.3 0.3>
    simSpaceSettings {worldAlpha=0.3}
   }
   n77 = RigidBody!(
    ComponentPose=@n76
    Alpha=Get Alpha_Skydiving()
    OverrideWorldGravity:Vector=Get GravityOverride()
    ExternalForce
    SimSpaceSettings:Sim Space Settings Structure=(WorldAlpha=0.5,VelocityScaleZ=1,DampingAlpha=1,MaxLinearVelocity=10000,MaxAngularVelocity=10000,MaxLinearAcceleration=10000,MaxAngularAcceleration=10000,ExternalLinearDragV=(X=0,Y=0,Z=0),ExternalLinearVelocity=(X=0,Y=0,Z=0),ExternalAngularVelocity=(X=0,Y=0,Z=0))
   ) -> (Pose) {
    bOverrideWorldGravity=true
    overridePhysicsAsset=&F_MED_JadeTowelHope_FaceAcc_Physics_SpecificBackpack
    alphaBoolBlend {blendOption=Linear}
    baseBoneRef {boneName=spine_04}
    componentAppliedLinearAccClamp=<3000 3000 3000>
    componentLinearAccScale=<0.3 0.3 0.3>
    componentLinearVelScale=<0.3 0.3 0.3>
    overrideWorldGravity {z=-980}
    simSpaceSettings {worldAlpha=0.5}
   }
   n78 = Component To Local(ComponentPose=@n77)
   n79 = Get bIsPlayingEmote()
   n80 = "Transform (Modify) Bone - Bone: base_ponytail_r"(
    ComponentPose=Local To Component(LocalPose=Use cached pose 'StartCache'())
    bAlphaBoolEnabled=True
    Translation="0,0,0"
    Rotation="-42.65,0,0"
    Scale="1,1,1"
   ) {
    alphaInputType=Bool
    rotationMode=BMM_Additive
    rotationSpace=BCS_BoneSpace
    scaleMode=BMM_Additive
    scaleSpace=BCS_BoneSpace
    translationMode=BMM_Additive
    translationSpace=BCS_BoneSpace
    alphaBoolBlend {blendInTime=0.2, blendOption=Linear, blendOutTime=0.2}
    boneToModify {boneName=base_ponytail_r}
    rotation {pitch=-42.65}
   }
   n81 = "Transform (Modify) Bone - Bone: base_ponytail_l"(ComponentPose=@n80, bAlphaBoolEnabled=True, Translation="0,0,0", Rotation="-27.105,0.000,0.000", Scale="1,1,1") {
    alphaInputType=Bool
    rotationMode=BMM_Additive
    rotationSpace=BCS_BoneSpace
    scaleMode=BMM_Additive
    scaleSpace=BCS_BoneSpace
    translationMode=BMM_Additive
    translationSpace=BCS_BoneSpace
    alphaBoolBlend {blendInTime=0.2, blendOption=Linear, blendOutTime=0.2}
    boneToModify {boneName=base_ponytail_l}
    rotation=<rot -27.105 0 0>
   }
   n82 = "Transform (Modify) Bone - Bone: base_ponytail_r"(ComponentPose=@n81, bAlphaBoolEnabled=True, Translation="0.665,2.993,0.000", Rotation="-0.000,-14.431,0.000", Scale="1,1,1") {
    alphaInputType=Bool
    rotationMode=BMM_Additive
    rotationSpace=BCS_BoneSpace
    scaleMode=BMM_Additive
    scaleSpace=BCS_BoneSpace
    translationMode=BMM_Additive
    translationSpace=BCS_BoneSpace
    alphaBoolBlend {blendInTime=0.2, blendOption=Linear, blendOutTime=0.2}
    boneToModify {boneName=base_ponytail_r}
    rotation=<rot 0 -14.431 0>
    translation=<0.665 2.993>
   }
   n83 = "Transform (Modify) Bone - Bone: base_ponytail_l"!(ComponentPose=@n82, Alpha:float=1, Translation="0.911,-9.322,-1.573", Rotation="0.032,-21.980,0.016", Scale="1,1,1") -> (Pose) {
    rotationMode=BMM_Additive
    rotationSpace=BCS_BoneSpace
    scaleMode=BMM_Additive
    scaleSpace=BCS_BoneSpace
    translationMode=BMM_Additive
    translationSpace=BCS_BoneSpace
    alphaBoolBlend {blendInTime=0.2, blendOption=Linear, blendOutTime=0.2}
    alphaScaleBiasClamp {bInterpResult=true}
    boneToModify {boneName=base_ponytail_l}
    rotation=<rot 0.032 -21.98 0.016>
    translation=<0.911 -9.322 -1.573>
   }
   n84 = "Transform (Modify) Bone - Bone: base_ponytail_l"!(ComponentPose=@n83, bAlphaBoolEnabled:bool=True, Translation="0,0,0", Rotation="5.685,-91.761,-7.333", Scale="1,1,1") -> (Pose) {
    alphaInputType=Bool
    rotationMode=BMM_Additive
    rotationSpace=BCS_BoneSpace
    scaleMode=BMM_Additive
    scaleSpace=BCS_BoneSpace
    translationMode=BMM_Additive
    translationSpace=BCS_BoneSpace
    alphaBoolBlend {blendInTime=0.1, blendOption=Linear, blendOutTime=0.2}
    alphaScaleBiasClamp {bInterpResult=true}
    boneToModify {boneName=base_ponytail_l}
    rotation=<rot 5.685 -91.761 -7.333>
   }
   n85 = "Transform (Modify) Bone - Bone: base_ponytail_r"(ComponentPose=@n84, bAlphaBoolEnabled=True, Translation="0,0,0", Rotation="6.001,-67.933,-4.321", Scale="1,1,1") {
    alphaInputType=Bool
    rotationMode=BMM_Additive
    rotationSpace=BCS_BoneSpace
    scaleMode=BMM_Additive
    scaleSpace=BCS_BoneSpace
    translationMode=BMM_Additive
    translationSpace=BCS_BoneSpace
    alphaBoolBlend {blendInTime=0.1, blendOption=Linear, blendOutTime=0.2}
    alphaScaleBiasClamp {bInterpResult=true}
    boneToModify {boneName=base_ponytail_r}
    rotation=<rot 6.001 -67.933 -4.321>
   }
   n86 = RigidBody!(ComponentPose=@n85, Alpha=Get Alpha_NotSkydiving(), OverrideWorldGravity=Get GravityOverride(), ExternalForce) -> (Pose) {
    bAllowPoseScalePropagation=false
    overridePhysicsAsset=&F_MED_JadeTowelHope_FaceAcc_Physics
    alphaBoolBlend {blendOption=Linear}
    baseBoneRef {boneName=spine_04}
    componentAppliedLinearAccClamp=<3000 3000 3000>
    componentLinearAccScale=<0.3 0.3 0.3>
    componentLinearVelScale=<0.3 0.3 0.3>
    simSpaceSettings {worldAlpha=0.3}
   }
   n87 = Component To Local(ComponentPose=@n86)
   n88 = "Transform (Modify) Bone - Bone: base_ponytail_l"(
    ComponentPose=Local To Component(LocalPose=Use cached pose 'StartCache'())
    bAlphaBoolEnabled=True
    Translation="0,0,0"
    Rotation="3.579,-49.466,-1.658"
    Scale="1,1,1"
   ) {
    alphaInputType=Bool
    rotationMode=BMM_Additive
    rotationSpace=BCS_BoneSpace
    scaleMode=BMM_Additive
    scaleSpace=BCS_BoneSpace
    translationMode=BMM_Additive
    translationSpace=BCS_BoneSpace
    alphaBoolBlend {blendInTime=0.2, blendOption=Linear, blendOutTime=0.2}
    boneToModify {boneName=base_ponytail_l}
    rotation=<rot 3.579 -49.466 -1.658>
   }
   n89 = "Transform (Modify) Bone - Bone: base_ponytail_r"(ComponentPose=@n88, bAlphaBoolEnabled=True, Translation="0,0,0", Rotation="0.000,-87.542,-0.000", Scale="1,1,1") {
    alphaInputType=Bool
    rotationMode=BMM_Additive
    rotationSpace=BCS_BoneSpace
    scaleMode=BMM_Additive
    scaleSpace=BCS_BoneSpace
    translationMode=BMM_Additive
    translationSpace=BCS_BoneSpace
    alphaBoolBlend {blendInTime=0.2, blendOption=Linear, blendOutTime=0.2}
    boneToModify {boneName=base_ponytail_r}
    rotation=<rot 0 -87.5419 0>
   }
   n90 = "Transform (Modify) Bone - Bone: base_ponytail_l"!(ComponentPose=@n89, Alpha:float=1, Translation="-0.126,-3.950,0.117", Rotation="-34.324,-0.113,-5.106", Scale="1,1,1") -> (Pose) {
    rotationMode=BMM_Additive
    rotationSpace=BCS_BoneSpace
    scaleMode=BMM_Additive
    scaleSpace=BCS_BoneSpace
    translationMode=BMM_Additive
    translationSpace=BCS_BoneSpace
    alphaBoolBlend {blendInTime=0.2, blendOption=Linear, blendOutTime=0.2}
    alphaScaleBiasClamp {bInterpResult=true}
    boneToModify {boneName=base_ponytail_l}
    rotation=<rot -34.324 -0.113 -5.106>
    translation=<-0.126 -3.95 0.117>
   }
   n91 = RigidBody!(ComponentPose=@n90, Alpha=Get Alpha_NotSkydiving(), ExternalForce) -> (Pose) {
    overridePhysicsAsset=&/F_MED_JadeTowelHope_FaceAcc_Physics_Cape
    alphaBoolBlend {blendOption=Linear}
    baseBoneRef {boneName=spine_04}
    componentAppliedLinearAccClamp=<3000 3000 3000>
    componentLinearAccScale=<0.3 0.3 0.3>
    componentLinearVelScale=<0.3 0.3 0.3>
    simSpaceSettings {worldAlpha=0.5}
   }
   n92 = Component To Local(ComponentPose=@n91)
   n93 = Get Is Wearing Cape() -> (Is Wearing Cape:bool)
   n94 = RigidBody!(
    ComponentPose=Local To Component(LocalPose=Blend Poses by bool(BlendPose_0=@n92, BlendPose_1=@n87, bActiveValue=@n93) {blendType=Linear})
    Alpha=Get Alpha_Skydiving()
    OverrideWorldGravity:Vector=Get GravityOverride()
    ExternalForce
    SimSpaceSettings:Sim Space Settings Structure=(WorldAlpha=0.5,VelocityScaleZ=1,DampingAlpha=1,MaxLinearVelocity=10000,MaxAngularVelocity=10000,MaxLinearAcceleration=10000,MaxAngularAcceleration=10000,ExternalLinearDragV=(X=0,Y=0,Z=0),ExternalLinearVelocity=(X=0,Y=0,Z=0),ExternalAngularVelocity=(X=0,Y=0,Z=0))
   ) -> (Pose) {
    bOverrideWorldGravity=true
    overridePhysicsAsset=&F_MED_JadeTowelHope_FaceAcc_Physics
    alphaBoolBlend {blendOption=Linear}
    baseBoneRef {boneName=spine_04}
    componentAppliedLinearAccClamp=<3000 3000 3000>
    componentLinearAccScale=<0.3 0.3 0.3>
    componentLinearVelScale=<0.3 0.3 0.3>
    overrideWorldGravity {z=-980}
    simSpaceSettings {worldAlpha=0.5}
   }
   n95 = RigidBody!(ComponentPose=Local To Component(LocalPose=Use cached pose 'StartCache'()), Alpha, OverrideWorldGravity=Get GravityOverride(), ExternalForce) -> (Pose) {
    bOverrideWorldGravity=true
    overridePhysicsAsset=&/F_MED_JadeTowelHope_FaceAcc_FrontEnd_Physics
    alphaBoolBlend {blendOption=Linear}
    baseBoneRef {boneName=spine_04}
    componentAppliedLinearAccClamp=<3000 3000 3000>
    componentLinearAccScale=<0.3 0.3 0.3>
    componentLinearVelScale=<0.3 0.3 0.3>
    simSpaceSettings {worldAlpha=0.3}
   }
   n96 = Blend Poses by bool(
    BlendPose_0=Component To Local(ComponentPose=@n95)
    BlendPose_1=Blend Poses by bool(
     BlendPose_0=Blend Poses by bool(BlendPose_0=@n78, BlendPose_1=@n75, bActiveValue=@n79) {blendType=Linear}
     BlendPose_1=Component To Local(ComponentPose=@n94)
     bActiveValue=Get Alpha_backPack() -> (Alpha_backPack:bool)
    ) {blendType=Linear}
    bActiveValue=Get frontendFix() -> (frontendFix:bool)
   ) {blendType=Linear}
   n97 = "Constraint - dyn_ponytail_01_l (P)"(ComponentPose=Local To Component(LocalPose=Use cached pose 'StartCache'()), Alpha=1, ConstraintWeights_0=1) {
    alphaBoolBlend {blendOption=Linear}
    boneToModify {boneName=dyn_ponytail_01_l}
    constraintSetup [{targetBone {boneName=neck_02}}]
    constraintWeights [1]
   }
   n98 = "Constraint - dyn_ponytail_02_l (P)"(ComponentPose=@n97, Alpha=1, ConstraintWeights_0=1) {
    alphaBoolBlend {blendOption=Linear}
    alphaScaleBiasClamp {bInterpResult=true, interpSpeedDecreasing=5, interpSpeedIncreasing=5}
    boneToModify {boneName=dyn_ponytail_02_l}
    constraintSetup [{targetBone {boneName=upperarm_l}}]
    constraintWeights [1]
   }
   n99 = "Constraint - dyn_ponytail_03_l (P)"(ComponentPose=@n98, Alpha=1, ConstraintWeights_0=1) {
    alphaBoolBlend {blendOption=Linear}
    boneToModify {boneName=dyn_ponytail_03_l}
    constraintSetup [{targetBone {boneName=spine_05}}]
    constraintWeights [1]
   }
   n100 = "Constraint - dyn_ponytail_01_r (P)"(ComponentPose=@n99, Alpha=1, ConstraintWeights_0=1) {
    alphaBoolBlend {blendOption=Linear}
    boneToModify {boneName=dyn_ponytail_01_r}
    constraintSetup [{targetBone {boneName=neck_02}}]
    constraintWeights [1]
   }
   n101 = "Constraint - dyn_ponytail_02_r (P)"(ComponentPose=@n100, Alpha=1, ConstraintWeights_0=1) {
    alphaBoolBlend {blendOption=Linear}
    boneToModify {boneName=dyn_ponytail_02_r}
    constraintSetup [{targetBone {boneName=upperarm_r}}]
    constraintWeights [1]
   }
   n102 = "Constraint - dyn_ponytail_03_r (P)"(ComponentPose=@n101, Alpha=1, ConstraintWeights_0=1) {
    alphaBoolBlend {blendOption=Linear}
    boneToModify {boneName=dyn_ponytail_03_r}
    constraintSetup [{targetBone {boneName=spine_05}}]
    constraintWeights [1]
   }
   n103 = "Transform (Modify) Bone - Bone: dyn_ponytail_03_r"!(ComponentPose=@n102, bAlphaBoolEnabled:bool=True, Translation="0,2,0", Rotation="-15,0,0", Scale="1,1,1") -> (Pose) {
    alphaInputType=Bool
    rotationMode=BMM_Additive
    rotationSpace=BCS_BoneSpace
    scaleMode=BMM_Additive
    scaleSpace=BCS_BoneSpace
    translationMode=BMM_Additive
    translationSpace=BCS_BoneSpace
    alphaBoolBlend {blendInTime=0.2, blendOption=Linear, blendOutTime=0.2}
    boneToModify {boneName=dyn_ponytail_03_r}
    rotation {pitch=-15}
    translation {y=2}
   }
   n104 = "Transform (Modify) Bone - Bone: dyn_ponytail_03_r"(ComponentPose=@n103, bAlphaBoolEnabled=True, Translation="0,5,0", Rotation="0,0,0", Scale="1,1,1") {
    alphaInputType=Bool
    rotationMode=BMM_Additive
    rotationSpace=BCS_BoneSpace
    scaleMode=BMM_Additive
    scaleSpace=BCS_BoneSpace
    translationMode=BMM_Additive
    translationSpace=BCS_BoneSpace
    alphaBoolBlend {blendInTime=0.2, blendOption=Linear, blendOutTime=0.2}
    boneToModify {boneName=dyn_ponytail_03_r}
    translation {y=5}
   }
   n105 = "Transform (Modify) Bone - Bone: dyn_ponytail_03_l"!(ComponentPose=@n104, bAlphaBoolEnabled:bool=True, Translation="0,0,0", Rotation="-15,0,0", Scale="1,1,1") -> (Pose) {
    alphaInputType=Bool
    rotationMode=BMM_Additive
    rotationSpace=BCS_BoneSpace
    scaleMode=BMM_Additive
    scaleSpace=BCS_BoneSpace
    translationMode=BMM_Additive
    translationSpace=BCS_BoneSpace
    alphaBoolBlend {blendInTime=0.2, blendOption=Linear, blendOutTime=0.2}
    boneToModify {boneName=dyn_ponytail_03_l}
    rotation {pitch=-15}
   }
   n106 = RigidBody!(ComponentPose=@n105, Alpha, ExternalForce) -> (Pose) {
    overridePhysicsAsset=&F_MED_JadeTowelHope_FaceAcc_Physics_SpecificBackpack
    alphaBoolBlend {blendOption=Linear}
    baseBoneRef {boneName=spine_04}
    componentAppliedLinearAccClamp=<3000 3000 3000>
    componentLinearAccScale=<0.3 0.3 0.3>
    componentLinearVelScale=<0.3 0.3 0.3>
    simSpaceSettings {worldAlpha=0.3}
   }
   n107 = "Transform (Modify) Bone - Bone: dyn_ponytail_04_r"(ComponentPose=@n106, Alpha=1, Translation="0,0,0", Rotation="-30,0,0", Scale="1,1,1") {
    rotationMode=BMM_Additive
    rotationSpace=BCS_BoneSpace
    scaleMode=BMM_Additive
    scaleSpace=BCS_BoneSpace
    translationMode=BMM_Additive
    translationSpace=BCS_BoneSpace
    alphaBoolBlend {blendInTime=0.2, blendOption=Linear, blendOutTime=0.2}
    boneToModify {boneName=dyn_ponytail_04_r}
    rotation {pitch=-30}
   }
   n108 = Component To Local(ComponentPose=@n107)
   n109 = Get bIsDBNO()
   n110 = RigidBody(ComponentPose=Local To Component(LocalPose=Blend Poses by bool(BlendPose_0=@n108, BlendPose_1=@n96, bActiveValue=@n109) {blendType=Linear})) {
    bOverrideWorldGravity=true
    overridePhysicsAsset=&/F_MED_JadeTowelHope_FaceAcc_PhysicsFront
    simulationSpace=BaseBoneSpace
    alphaBoolBlend {blendOption=Linear}
    baseBoneRef {boneName=head}
    componentAppliedLinearAccClamp=<3000 3000 3000>
    componentLinearAccScale=<0.3 0.3 0.3>
    componentLinearVelScale=<0.3 0.3 0.3>
    overrideWorldGravity {z=-980}
    simSpaceSettings {worldAlpha=1}
   }
   Save cached pose 'RigidBodyCache'(Pose=Component To Local(ComponentPose=@n110)) {cacheName=RigidBodyCache}
   n111 = Layered blend per bone(
    BasePose=Copy Pose From Mesh() {bUseAttachedParent=true}
    BlendPoses_0=Copy Pose From Mesh(SourceMeshComponent=Get Body_Mesh()) {bUseAttachedParent=true}
   ) {layerSetup [{branchFilters [
       {boneName=clavicle_l}
       {boneName=upperarm_l}
       {boneName=clavicle_r}
       {boneName=upperarm_r}
       {boneName=calf_l}
       {boneName=calf_r}
       {boneName=thigh_r}
       {boneName=thigh_l}
      ]}]}
   n112 = "Transform (Modify) Bone - Bone: base_ponytail_r"!(ComponentPose=Local To Component(LocalPose=@n111), Alpha:float=1, Translation="4.438,-0.715,1.225", Rotation="0,0,0", Scale="1,1,1") -> (Pose) {
    rotationMode=BMM_Additive
    rotationSpace=BCS_BoneSpace
    scaleMode=BMM_Additive
    scaleSpace=BCS_BoneSpace
    translationMode=BMM_Additive
    translationSpace=BCS_BoneSpace
    alphaBoolBlend {blendOption=Linear}
    alphaScaleBiasClamp {bInterpResult=true, interpSpeedDecreasing=5}
    boneToModify {boneName=base_ponytail_r}
    translation=<4.4383 -0.7148 1.225>
   }
   n113 = "Transform (Modify) Bone - Bone: base_ponytail_r"(ComponentPose=@n112, Alpha=1, Translation="1.715,0,0.907", Rotation="0,0,0", Scale="1,1,1") {
    rotationMode=BMM_Additive
    rotationSpace=BCS_BoneSpace
    scaleMode=BMM_Additive
    scaleSpace=BCS_BoneSpace
    translationMode=BMM_Additive
    translationSpace=BCS_BoneSpace
    alphaBoolBlend {blendOption=Linear}
    alphaScaleBiasClamp {bInterpResult=true, interpSpeedDecreasing=5}
    boneToModify {boneName=base_ponytail_r}
    translation {x=1.715, z=0.907}
   }
   n114 = "Transform (Modify) Bone - Bone: base_ponytail_l"(ComponentPose=@n113, Alpha=1, Translation="-4.973,-0.000,-0.000", Rotation="0.000,0.000,-13.851", Scale="1,1,1") {
    rotationMode=BMM_Additive
    rotationSpace=BCS_BoneSpace
    scaleMode=BMM_Additive
    scaleSpace=BCS_BoneSpace
    translationMode=BMM_Additive
    translationSpace=BCS_BoneSpace
    alphaBoolBlend {blendOption=Linear}
    alphaScaleBiasClamp {bInterpResult=true}
    boneToModify {boneName=base_ponytail_l}
    rotation=<rot 0 0 -13.851>
    translation {x=-4.9734}
   }
   n115 = "Transform (Modify) Bone - Bone: base_ponytail_r"(ComponentPose=@n114, Alpha=1, Translation="5.359,-1.984,-0.000", Rotation="2.077,14.865,5.151", Scale="1,1,1") {
    rotationMode=BMM_Additive
    rotationSpace=BCS_BoneSpace
    scaleMode=BMM_Additive
    scaleSpace=BCS_BoneSpace
    translationMode=BMM_Additive
    translationSpace=BCS_BoneSpace
    alphaBoolBlend {blendOption=Linear}
    alphaScaleBiasClamp {bInterpResult=true}
    boneToModify {boneName=base_ponytail_r}
    rotation=<rot 2.077 14.865 5.151>
    translation=<5.359 -1.984>
   }
   n116 = "Transform (Modify) Bone - Bone: base_ponytail_l"!(ComponentPose=@n115, bAlphaBoolEnabled:bool=True, Translation="0,0,0", Rotation="8.117,-19.194,22.083", Scale="1,1,1") -> (Pose) {
    alphaInputType=Bool
    rotationMode=BMM_Additive
    rotationSpace=BCS_BoneSpace
    scaleMode=BMM_Additive
    scaleSpace=BCS_BoneSpace
    translationMode=BMM_Additive
    translationSpace=BCS_BoneSpace
    alphaBoolBlend {blendInTime=0.1, blendOption=Linear, blendOutTime=0.2}
    alphaScaleBiasClamp {bInterpResult=true}
    boneToModify {boneName=base_ponytail_l}
    rotation=<rot 8.117 -19.194 22.083>
   }
   n117 = "Transform (Modify) Bone - Bone: base_ponytail_l"!(ComponentPose=@n116, Alpha:float=1, Translation="-7.254,-5.232,0", Rotation="0,-25.373,-26.642", Scale="1,1,1") -> (Pose) {
    rotationMode=BMM_Additive
    rotationSpace=BCS_BoneSpace
    scaleMode=BMM_Additive
    scaleSpace=BCS_BoneSpace
    translationMode=BMM_Additive
    translationSpace=BCS_BoneSpace
    alphaBoolBlend {blendInTime=0.1, blendOption=Linear, blendOutTime=0.2}
    alphaScaleBiasClamp {bInterpResult=true}
    boneToModify {boneName=base_ponytail_l}
    rotation {roll=-26.642, yaw=-25.373}
    translation=<-7.254 -5.232>
   }
   Save cached pose 'StartCache'(Pose=Component To Local(ComponentPose=@n117)) {cacheName=StartCache}
  ]}
 EventGraph {nodes [
   Event BlueprintUpdateAnimation()
   > n1 = Sequence!(execute) -> (then_0, then_1, then_2:Exec, then_3:Exec, then_4:Exec, then_5:Exec, then_6:Exec, then_7:Exec, then_8:Exec, then_9:Exec, then_10:Exec)
   > n2 = Cast To PlayerPawn_Athena(Object=TryGetPawnOwner())
   > Set Body_Mesh(Body_Mesh=GetSkeletalMeshForPartType(self=@n2))
   n3 = Sequence!(execute=@n1.then_1) -> (then_0, then_1, then_2, then_3, then_4, then_5, then_6, then_7, then_8, then_9, then_10, then_11:Exec, then_12:Exec)
   > n4 = Gravity Override State Function DEPRECATED(
    CharacterPartAnimInstance=Self-Reference()
    GameTimeInSeconds=GetGameTimeInSeconds()
    GravityOverrideParams=Get Gravity Override Params()
    RBAN Params=Get RBAN Params()
   )
   > Set GravityOverride(GravityOverride=@n4)
   > Set RBAN Settings Out(RBAN Settings Out:RBANSetting Struct Structure=@n4.RBAN SettingsOut) -> (Output_Get:RBANSetting Struct Structure)
   n5 = GetOwnerSkeletalMeshComponent()
   n6 = BreakRotator(InRot=BreakTransform(InTransform=GetDeltaTransformFromRefPose(self=@n5, BoneName=neck_02, BaseName=spine_05)).Rotation)
   n7 = Branch(execute=@n3.then_1, Condition=AND Boolean(A="float < float"(A=@n6.Yaw, B=-30), B=Get bIsCrouching()))
   > Set Alpha_HeadRotateUpCrouchExtrem(Alpha_HeadRotateUpCrouchExtrem:double=1) -> (Output_Get:double)
   Set Alpha_HeadRotateUpCrouchExtrem(execute=@n7.else, Alpha_HeadRotateUpCrouchExtrem=0)
   n8 = Get bIsParachuteOpened()
   n9 = Get bIsSkyDiving()
   n10 = OR Boolean!(A=Get bIsPlayingEmote(), B=Get bIsSwimSprinting() -> (bIsSwimSprinting:bool), C:bool=Get bIsMoving2D()) -> (ReturnValue)
   n11 = GetDeltaTransformFromRefPose(self=GetOwnerSkeletalMeshComponent(), BoneName=upperarm_r, BaseName=spine_05)
   n12 = "float < float"(A=BreakRotator(InRot=BreakTransform(InTransform=@n11).Rotation).Yaw, B=-10)
   n13 = Branch(
    execute=@n3.then_11
    Condition=AND Boolean!(A=AND Boolean!(A=@n10, B=@n12) -> (ReturnValue), B=NOT Boolean(A=@n8), C:bool=NOT Boolean(A=@n9)) -> (ReturnValue)
   )
   > Set Alpha_DetectedArmBack_R(Alpha_DetectedArmBack_R:double=1) -> (Output_Get:double)
   Set Alpha_DetectedArmBack_R(execute=@n13.else, Alpha_DetectedArmBack_R=0)
   n14 = Get bRigidBodyEnabled()
   n15 = Get bIsSkyDiving()
   n16 = GetOwnerSkeletalMeshComponent()
   n17 = BreakRotator(InRot=BreakTransform(InTransform=GetDeltaTransformFromRefPose(self=@n16, BoneName=clavicle_r, BaseName=spine_05)).Rotation)
   n18 = Branch(
    execute=@n3.then_12
    Condition=AND Boolean!(A="float < float"(A=@n17.Yaw, B=-10), B=Get bIsInFrontEnd(), C:bool=NOT Boolean(A=@n15), D:bool=NOT Boolean(A=@n14)) -> (ReturnValue)
   )
   > Set Alpha_DetectedClavicleBack_R(Alpha_DetectedClavicleBack_R:double=1) -> (Output_Get:double)
   Set Alpha_DetectedClavicleBack_R(execute=@n18.else, Alpha_DetectedClavicleBack_R=0)
   n19 = Get bIsSkyDiving()
   n20 = Get bIsPlayingEmote()
   n21 = Get bIsTacticalSprinting()
   n22 = GetOwnerSkeletalMeshComponent()
   n23 = BreakRotator(InRot=BreakTransform(InTransform=GetDeltaTransformFromRefPose(self=@n22, BoneName=upperarm_l, BaseName=spine_05)).Rotation)
   n24 = Branch(
    execute=@n3.then_7
    Condition=AND Boolean(A="float < float"(A=@n23.Yaw, B=1.5), B=NOT Boolean(A=@n19), C=NOT Boolean(A=@n21), D=NOT Boolean(A=@n20))
   )
   > Set Alpha_DetectedArmRotBackIdle(Alpha_DetectedArmRotBackIdle:double=1) -> (Output_Get:double)
   Set Alpha_DetectedArmRotBackIdle(execute=@n24.else, Alpha_DetectedArmRotBackIdle=0)
   n25 = GetOwnerSkeletalMeshComponent()
   n26 = BreakRotator(InRot=BreakTransform(InTransform=GetDeltaTransformFromRefPose(self=@n25, BoneName=upperarm_l, BaseName=spine_05)).Rotation)
   n27 = Branch(
    execute=@n3.then_9
    Condition=AND Boolean!(
     A="float > float"(A=@n26.Pitch, B=10)
     B=Get Test_DetectedRotAvocadoEmote() -> (Test_DetectedRotAvocadoEmote:bool)
     C=Get bRigidBodyEnabled()
    ) -> (ReturnValue)
   )
   > Set Alpha_DetectedArmRotBackEmoteL(Alpha_DetectedArmRotBackEmoteL:double=1) -> (Output_Get:double)
   Set Alpha_DetectedArmRotBackEmoteL(execute=@n27.else, Alpha_DetectedArmRotBackEmoteL=0)
   n28 = Get bIsSkyDiving()
   n29 = GetOwnerSkeletalMeshComponent()
   n30 = BreakRotator(InRot=BreakTransform(InTransform=GetDeltaTransformFromRefPose(self=@n29, BoneName=upperarm_l, BaseName=spine_05)).Rotation)
   n31 = Branch(execute=@n3.then_10, Condition=AND Boolean(A="float < float"(A=@n30.Yaw, B=-10), B=@n10, C=NOT Boolean(A=@n28)))
   > Set Alpha_DetectedArmBack_L(Alpha_DetectedArmBack_L:double=1) -> (Output_Get:double)
   Set Alpha_DetectedArmBack_L(execute=@n31.else, Alpha_DetectedArmBack_L=0)
   n32 = GetOwnerSkeletalMeshComponent()
   n33 = BreakRotator(InRot=BreakTransform(InTransform=GetDeltaTransformFromRefPose(self=@n32, BoneName=head, BaseName=pelvis)).Rotation)
   n34 = Branch(execute=@n3.then_2, Condition=AND Boolean!(A="float < float"(A=@n33.Yaw, B=-30), B=Get bIsCrouching()) -> (ReturnValue))
   > Set Alpha_HeadRotateUpCrouching(Alpha_HeadRotateUpCrouching:double=1) -> (Output_Get:double)
   Set Alpha_HeadRotateUpCrouching(execute=@n34.else, Alpha_HeadRotateUpCrouching=0)
   n35 = GetOwnerSkeletalMeshComponent()
   n36 = BreakRotator(InRot=BreakTransform(InTransform=GetDeltaTransformFromRefPose(self=@n35, BoneName=head, BaseName=pelvis)).Rotation)
   n37 = Branch(execute=@n3.then_4, Condition=AND Boolean(A="float > float"(A=@n36, B=15), B=Get Alpha_backPack()))
   > Set Alpha_HeadRotLeft(Alpha_HeadRotLeft:double=1) -> (Output_Get:double)
   Set Alpha_HeadRotLeft(execute=@n37.else, Alpha_HeadRotLeft=0)
   n38 = GetOwnerSkeletalMeshComponent()
   Set Test_DetectedRotAvocadoEmote(
    execute=@n3.then_8
    Test_DetectedRotAvocadoEmote:bool=Montage_IsPlaying(self=GetAnimInstance(self=@n38), Montage=/BRCosmetics/Animation/Game/MainPlayer/Emotes/AvocadoSeal/CMF/Emote_AvocadoSeal_CMF_M)
   ) -> (Output_Get:bool)
   n39 = GetOwnerSkeletalMeshComponent()
   n40 = BreakRotator(InRot=BreakTransform(InTransform=GetDeltaTransformFromRefPose(self=@n39, BoneName=neck_01, BaseName=pelvis)).Rotation)
   > n41 = Branch(Condition=AND Boolean!(A="float > float"(A=@n40, B=14), B=Get Test_DetectedRotAvocadoEmote(), C:bool=Get bRigidBodyEnabled()) -> (ReturnValue))
   > Set Alpha_DetectedAvocadoEmote(Alpha_DetectedAvocadoEmote:double=1) -> (Output_Get:double)
   Set Alpha_DetectedAvocadoEmote(execute=@n41.else, Alpha_DetectedAvocadoEmote=0)
   n42 = GetOwnerSkeletalMeshComponent()
   n43 = BreakRotator(InRot=BreakTransform(InTransform=GetDeltaTransformFromRefPose(self=@n42, BoneName=head, BaseName=pelvis)).Rotation)
   n44 = Branch(execute=@n3.then_3, Condition=AND Boolean!(A="float < float"(A=@n43.Yaw, B=-30), B=Get bIsPlayingMeleeAnim()) -> (ReturnValue))
   > Set Alpha_HeadRotateUpMelee(Alpha_HeadRotateUpMelee:double=1) -> (Output_Get:double)
   Set Alpha_HeadRotateUpMelee(execute=@n44.else, Alpha_HeadRotateUpMelee=0)
   n45 = GetOwnerSkeletalMeshComponent()
   n46 = BreakRotator(InRot=BreakTransform(InTransform=GetDeltaTransformFromRefPose(self=@n45, BoneName=spine_05, BaseName=pelvis)).Rotation)
   n47 = Branch(
    execute=@n3.then_5
    Condition=AND Boolean!(A="float > float"(A=@n46, B=12), B=Get bIsPlayingMeleeAnim(), C:bool=Get Selected_DetectedPickaxe() -> (Selected_DetectedPickaxe:bool)) -> (ReturnValue)
   )
   > Set Alpha_CollisionHairAndArm L(Alpha_CollisionHairAndArm L:double=1) -> (Output_Get:double)
   Set Alpha_CollisionHairAndArm L(execute=@n47.else, Alpha_CollisionHairAndArm L=0)
   n48 = Get bIsPlayingEmote()
   n49 = Get bIsSkyDiving()
   n50 = Get bIsInFrontEnd()
   n51 = GetOwnerSkeletalMeshComponent()
   n52 = BreakRotator(InRot=BreakTransform(InTransform=GetDeltaTransformFromRefPose(self=@n51, BoneName=upperarm_r, BaseName=spine_05)).Rotation)
   n53 = Branch(
    execute=@n3.then_6
    Condition=AND Boolean!(A="float < float"(A=@n52.Yaw, B=-9), B=NOT Boolean(A=@n49), C=NOT Boolean(A=@n48), D:bool=NOT Boolean(A=@n50)) -> (ReturnValue)
   )
   > Set Alpha_CollisionHairAndArm R(Alpha_CollisionHairAndArm R:double=1) -> (Output_Get:double)
   Set Alpha_CollisionHairAndArm R(execute=@n53.else, Alpha_CollisionHairAndArm R=0)
   n54 = Get Is Wearing Cape()
   n55 = Get isWearingBackbling()
   n56 = Get Alpha_backPack()
   Set frontendFix(
    execute=@n1.then_10
    frontendFix:bool=AND Boolean(A=NOT Boolean(A=@n54), B=NOT Boolean(A=@n55), C=NOT Boolean(A=@n56), D=Get bIsInFrontEnd())
   ) -> (Output_Get:bool)
   n57 = CheckForBackbling()
   n58 = Set isWearingBackbling(execute=@n1.then_2, isWearingBackbling=@n57)
   > n59 = Set Is Wearing Cape(Is Wearing Cape=OR Boolean!(A=@n57.bIsBackblingACape, B=@n57.bIsBackblingLong) -> (ReturnValue))
   > Set Alpha_backPack(Alpha_backPack:bool=AND Boolean!(A=@n58, B=NOT Boolean(A=@n59)) -> (ReturnValue)) -> (Output_Get:bool)
   n60 = Branch(execute=@n1.then_6, Condition=AND Boolean(A=Get bIsPlayingMeleeAnim(), B=Get Selected_DetectedPickaxe()))
   > Set Alpha_MeleePickaxe(Alpha_MeleePickaxe:double=1) -> (Output_Get:double)
   > Set Alpha_NotMeleePickaxe(Alpha_NotMeleePickaxe:double=0) -> (Output_Get:double)
   Set Alpha_MeleePickaxe(execute=@n60.else, Alpha_MeleePickaxe=0)
   > Set Alpha_NotMeleePickaxe(Alpha_NotMeleePickaxe=1)
   n61 = Branch(execute=@n1.then_9, Condition=AND Boolean(A=Get bIsJumping(), B=Get bIsTacticalSprinting()))
   > Set Alpha_TacticalSprintingJump(Alpha_TacticalSprintingJump:double=1) -> (Output_Get:double)
   Set Alpha_TacticalSprintingJump(execute=@n61.else, Alpha_TacticalSprintingJump=0)
   n62 = Get bIsCrouching()
   n63 = Get bIsSkyDiving()
   n64 = Get bIsTacticalSprinting()
   n65 = Branch(
    execute=@n1.then_4
    Condition=AND Boolean!(
     A=Get bIsMoving2D()
     B=NOT Boolean(A=@n64)
     C:bool=Get Alpha_backPack()
     D:bool=NOT Boolean(A=@n62)
     E:bool=NOT Boolean(A=@n63)
     F:bool=Get Selected_DetectedPickaxe()
    ) -> (ReturnValue)
   )
   > Set Alpha_Moving2d Pickaxe(Alpha_Moving2d Pickaxe:double=1) -> (Output_Get:double)
   Set Alpha_Moving2d Pickaxe(execute=@n65.else, Alpha_Moving2d Pickaxe=0)
   n66 = Branch(execute=@n1.then_5, Condition=AND Boolean!(A=Get bIsJumping(), B=Get bIsTacticalSprinting(), C=Get Alpha_backPack()) -> (ReturnValue))
   > Set Alpha_TacticalSprintingJumpBackpack(Alpha_TacticalSprintingJumpBackpack:double=1) -> (Output_Get:double)
   Set Alpha_TacticalSprintingJumpBackpack(execute=@n66.else, Alpha_TacticalSprintingJumpBackpack=0)
   n67 = Get bIsDBNOCarrying() -> (bIsDBNOCarrying:bool)
   n68 = Get bIsDBNOCarried()
   n69 = Branch(execute=@n1.then_7, Condition=AND Boolean(A=Get bIsDBNO(), B=NOT Boolean(A=@n68), C=NOT Boolean(A=@n67)))
   > Set Alpha_OnlyDBNOCarried(Alpha_OnlyDBNOCarried:double=1) -> (Output_Get:double)
   Set Alpha_OnlyDBNOCarried(execute=@n69.else, Alpha_OnlyDBNOCarried=0)
   n70 = Branch(execute=@n1.then_3, Condition=Get bIsSkyDiving())
   > Set Alpha_Skydiving(Alpha_Skydiving:double=1) -> (Output_Get:double)
   > Set Alpha_NotSkydiving(Alpha_NotSkydiving:double=0) -> (Output_Get:double)
   Set Alpha_Skydiving(execute=@n70.else, Alpha_Skydiving=0)
   > Set Alpha_NotSkydiving(Alpha_NotSkydiving=1)
   n71 = Branch(execute=@n1.then_8, Condition=NOT Boolean(A=Get bRigidBodyEnabled()))
   > Set RB_Disable(RB_Disable:double=1) -> (Output_Get:double)
   Set RB_Disable(execute=@n71.else, RB_Disable=0)
   Event OnWeaponChanged()
   > n72 = Cast To FortPlayerAnimInstance(Object=GetAnimInstance(self=GetOwnerSkeletalMeshComponent()))
   > Set Selected_DetectedPickaxe(Selected_DetectedPickaxe:bool=false) -> (Output_Get:bool)
   > n73 = Switch on EFortWeaponCoreAnimation!(execute, Selection:EFortWeaponCoreAnimation Enum=Get WeaponCoreAnim(self=@n72)) -> (
    Melee:Exec
    Pistol:Exec
    Shotgun:Exec
    PaperBlueprint:Exec
    Rifle:Exec
    MeleeOneHand:Exec
    MachinePistol:Exec
    RocketLauncher:Exec
    GrenadeLauncher:Exec
    GoingCommando:Exec
    AssaultRifle:Exec
    TacticalShotgun:Exec
    SniperRifle:Exec
    TrapPlacement:Exec
    ShoulderLauncher:Exec
    AbilityDecoTool:Exec
    Crossbow:Exec
    C4:Exec
    RemoteControl:Exec
    DualWield:Exec
    AR_BullPup:Exec
    AR_ForwardGrip:Exec
    MedPackPaddles:Exec
    SMG_P90:Exec
    AR_DrumGun:Exec
    Consumable_Small:Exec
    Consumable_Large:Exec
    Balloon:Exec
    MountedTurret:Exec
    CreativeTool:Exec
    ExplosiveBow:Exec
    AshtonIndigo:Exec
    AshtonChicago:Exec
    MeleeDualWield:Exec
    Unarmed:Exec
   )
   Set Selected_DetectedPickaxe(execute=[@n73.Melee, @n73.MeleeDualWield], Selected_DetectedPickaxe=1)
   n74 = "float > float"(B=100)
   n75 = Get bIsSkydiveFloating()
   AND Boolean!(A=NOT Boolean(A=Get bIsParachuteOpened()), B=OR Boolean(A=AND Boolean(A=@n75, B=@n74), B=false)) -> (ReturnValue)
  ]}
]
variables [
 Alpha_backPack {type=boolean}
 Alpha_CollisionHairAndArm L {type=number}
 Alpha_CollisionHairAndArm R {type=number}
 Alpha_DetectedArmBack_L {type=number}
 Alpha_DetectedArmBack_R {type=number}
 Alpha_DetectedArmRotBackEmoteL {type=number}
 Alpha_DetectedArmRotBackIdle {type=number}
 Alpha_DetectedAvocadoEmote {type=number}
 Alpha_DetectedClavicleBack_R {type=number}
 Alpha_DetectedSpecificBackpack {type=boolean}
 Alpha_HeadRotateUpCrouchExtrem {type=number}
 Alpha_HeadRotateUpCrouching {type=number}
 Alpha_HeadRotateUpMelee {type=number}
 Alpha_HeadRotLeft {type=number}
 Alpha_MeleePickaxe {type=number}
 Alpha_Moving2d Pickaxe {type=number}
 Alpha_NotMeleePickaxe {type=number}
 Alpha_NotSkydiving {type=number}
 Alpha_OnlyDBNOCarried {type=number}
 Alpha_Skydiving {type=number}
 Alpha_TacticalSprintingJump {type=number}
 Alpha_TacticalSprintingJumpBackpack {type=number}
 backblingItemDef {type=AthenaBackpackItemDefinition, default=null}
 Body_Mesh {type=SkeletalMeshComponent}
 currentBackpack {type=AthenaBackpackItemDefinition, default=null}
 frontendFix {type=boolean}
 Gravity Override Params {type=GravityOverrideParamsStruct}
 GravityOverride {type=Vector}
 Is Wearing Cape {type=boolean}
 isWearingBackbling {type=boolean}
 RB_Disable {type=number}
 RBAN Params {
  type=RBANParamsStruct
  default {
   dBNO {simSpaceSettings {
     maxAngularAcceleration=1000
     maxAngularVelocity=3
     maxLinearAcceleration=300
     maxLinearVelocity=700
     velocityScaleZ=0
     externalLinearDragV=<0 0.5 0>
    }}
   dBNO_Carried {simSpaceSettings {
     maxAngularAcceleration=1000
     maxAngularVelocity=3
     maxLinearAcceleration=300
     maxLinearVelocity=700
     velocityScaleZ=0
     externalLinearDragV=<0 0.5 0>
    }}
   emoteOrMelee {simSpaceSettings {
     maxAngularAcceleration=1000
     maxAngularVelocity=3
     maxLinearAcceleration=300
     maxLinearVelocity=700
     velocityScaleZ=0
     externalLinearDragV=<0 0.5 0>
    }}
   front_End {simSpaceSettings {
     maxAngularAcceleration=1000
     maxAngularVelocity=3
     maxLinearAcceleration=300
     maxLinearVelocity=700
     velocityScaleZ=0
     externalLinearDragV=<0 0.5 0>
    }}
   onGround_Crouch_Moving {simSpaceSettings {
     maxAngularAcceleration=1000
     maxAngularVelocity=3
     maxLinearAcceleration=300
     maxLinearVelocity=700
     velocityScaleZ=0
     externalLinearDragV=<0 0.5 0>
    }}
   onGround_Falling {simSpaceSettings {
     maxAngularAcceleration=1000
     maxAngularVelocity=3
     maxLinearAcceleration=300
     maxLinearVelocity=700
     velocityScaleZ=0
     externalLinearDragV=<0 0.5 0>
    }}
   onGround_Moving {simSpaceSettings {
     maxAngularAcceleration=1000
     maxAngularVelocity=3
     maxLinearAcceleration=300
     maxLinearVelocity=700
     velocityScaleZ=0
     externalLinearDragV=<0 0.5 0>
    }}
   onGround_Standing {simSpaceSettings {
     maxAngularAcceleration=1000
     maxAngularVelocity=3
     maxLinearAcceleration=300
     maxLinearVelocity=700
     velocityScaleZ=0
     externalLinearDragV=<0 0.5 0>
    }}
   rail_Grinding {simSpaceSettings {
     maxAngularAcceleration=1000
     maxAngularVelocity=3
     maxLinearAcceleration=300
     maxLinearVelocity=700
     velocityScaleZ=0
     worldAlpha=0.1
     externalLinearDragV=<0 0.5 0>
    }}
   riding_Moving {simSpaceSettings {
     maxAngularAcceleration=1000
     maxAngularVelocity=3
     maxLinearAcceleration=300
     maxLinearVelocity=700
     velocityScaleZ=0
     externalLinearDragV=<0 0.5 0>
    }}
   riding_Moving_Targeting {simSpaceSettings {
     maxAngularAcceleration=1000
     maxAngularVelocity=3
     maxLinearAcceleration=300
     maxLinearVelocity=700
     velocityScaleZ=0
     externalLinearDragV=<0 0.5 0>
    }}
   riding_Standing {simSpaceSettings {
     maxAngularAcceleration=1000
     maxAngularVelocity=3
     maxLinearAcceleration=300
     maxLinearVelocity=700
     velocityScaleZ=0
     externalLinearDragV=<0 0.5 0>
    }}
   skydive_DivingDown {simSpaceSettings {
     maxAngularAcceleration=1000
     maxAngularVelocity=3
     maxLinearAcceleration=300
     maxLinearVelocity=700
     velocityScaleZ=0
     externalLinearDragV=<0 0.5 0>
    }}
   skydive_FloatingDown {
    componentLinearAccScale=<0 0.5 0>
    componentLinearVelScale=<0 0.5 0>
    simSpaceSettings {
     maxAngularAcceleration=1000
     maxAngularVelocity=3
     maxLinearAcceleration=300
     maxLinearVelocity=700
     velocityScaleZ=0
     externalLinearDragV=<0 0.5 0>
    }
   }
   skydive_GoingUp {simSpaceSettings {
     maxAngularAcceleration=1000
     maxAngularVelocity=3
     maxLinearAcceleration=300
     maxLinearVelocity=700
     velocityScaleZ=0
     externalLinearDragV=<0 0.5 0>
    }}
   skydive_Parachuting {
    componentLinearAccScale=<0 0.1 0.3>
    componentLinearVelScale=<0 0.1 0.3>
    simSpaceSettings {
     maxAngularAcceleration=1000
     maxAngularVelocity=3
     maxLinearAcceleration=300
     maxLinearVelocity=700
     velocityScaleZ=0
     worldAlpha=0.1
     externalLinearDragV=<0 0.5 0>
    }
   }
   skydive_Surfing {simSpaceSettings {
     maxAngularAcceleration=1000
     maxAngularVelocity=3
     maxLinearAcceleration=300
     maxLinearVelocity=700
     velocityScaleZ=0
     externalLinearDragV=<0 0.5 0>
    }}
   swimming {simSpaceSettings {
     maxAngularAcceleration=1000
     maxAngularVelocity=3
     maxLinearAcceleration=300
     maxLinearVelocity=700
     velocityScaleZ=0
     externalLinearDragV=<0 0.5 0>
    }}
   tactical_Sprint {simSpaceSettings {
     maxAngularAcceleration=1000
     maxAngularVelocity=3
     maxLinearAcceleration=300
     maxLinearVelocity=700
     velocityScaleZ=0
     externalLinearDragV=<0 0.5 0>
    }}
   ziplining {simSpaceSettings {
     maxAngularAcceleration=1000
     maxAngularVelocity=3
     maxLinearAcceleration=300
     maxLinearVelocity=700
     velocityScaleZ=0
     externalLinearDragV=<0 0.5 0>
    }}
  }
 }
 RBAN Settings Out {
  type=RBANSettingStruct
  default {
   componentLinearAccScale=<0 0 0>
   componentLinearVelScale=<0 0 0>
   simSpaceSettings {
    dampingAlpha=1
    maxAngularAcceleration=1000
    maxAngularVelocity=3
    maxLinearAcceleration=300
    maxLinearVelocity=700
    velocityScaleZ=0
    worldAlpha=1
    externalAngularVelocity=<0 0 0>
    externalLinearDragV=<0 0.5 0>
    externalLinearVelocity=<0 0 0>
   }
  }
 }
 Selected_DetectedPickaxe {type=boolean}
 Test_DetectedRotAvocadoEmote {type=boolean}
 typhoonBackblingSkelMesh {type=SkeletalMesh, default=null}
]
